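# Physics asset for the TameEagle female medium player body cosmetic, covering the pelvis, thighs and skirt.
# Capsule bodies form a three-link front-center skirt chain: the first link locks to its base bone, and the next two have tight angular limits.
# Both thighs attach to the pelvis with limited joints, while the tiny front-center and back-left skirt base bodies attach with free rotation.
# The back-left base has no chain bodies below it.
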PhysicsAsset /BRCosmetics/Characters/Player/Female/Medium/Bodies/F_MED_TameEagle/Meshes/F_MED_TameEagle_Physics
bodies {
 base_skirt_bk_l [{shapeType=Capsule, length=0.505, radius=0.505, shapeName=base_skirt_bk_l_capsule, rotation=<rot 0 -90 0>}]
 base_skirt_fr_c [{shapeType=Capsule, length=0.505, radius=0.505, shapeName=base_skirt_fr_c_capsule, rotation=<rot 0 -90 0>}]
 dyn_skirt_fr_c_01 [{shapeType=Capsule, length=5, radius=3, shapeName=dyn_skirt_fr_c_01_capsule, rotation=<rot 0 0 90>}]
 dyn_skirt_fr_c_02 [{shapeType=Capsule, length=10, radius=4, shapeName=dyn_skirt_fr_c_02_capsule, center=<0 0 2.333>, rotation=<rot 0 0 90>}]
 dyn_skirt_fr_c_03 [{shapeType=Capsule, length=10, radius=4, shapeName=dyn_skirt_fr_c_03_capsule, rotation=<rot 0 0 90>}]
 pelvis [{shapeType=Capsule, length=5, radius=10}]
 thigh_l [{shapeType=Capsule, length=34, radius=8, center=<-22.9837 0.3669 0.0921>, rotation=<rot -0.1219 -93.0124 -92.483>}]
 thigh_r [{shapeType=Capsule, length=34, radius=8, center=<22.9666 -0.7001 -0.0023>, rotation=<rot -0.1869 -92.7275 -93.8082>}]
}
constraints [
 {bone1Name=thigh_l, bone2Name=pelvis, swing1LimitDegrees=45, swing2LimitDegrees=45, twistLimitDegrees=45}
 {bone1Name=thigh_r, bone2Name=pelvis, swing1LimitDegrees=45, swing2LimitDegrees=45, twistLimitDegrees=45}
 {
  bone1Name=base_skirt_bk_l
  bone2Name=pelvis
  swing1LimitDegrees=45
  swing1Motion=Free
  swing2LimitDegrees=45
  swing2Motion=Free
  twistLimitDegrees=45
  twistMotion=Free
 }
 {
  bone1Name=base_skirt_fr_c
  bone2Name=pelvis
  swing1LimitDegrees=45
  swing1Motion=Free
  swing2LimitDegrees=45
  swing2Motion=Free
  twistLimitDegrees=45
  twistMotion=Free
 }
 {
  bone1Name=dyn_skirt_fr_c_01
  bone2Name=base_skirt_fr_c
  swing1LimitDegrees=5
  swing1Motion=Locked
  swing2LimitDegrees=5
  swing2Motion=Locked
  twistLimitDegrees=0
  twistMotion=Locked
 }
 {bone1Name=dyn_skirt_fr_c_02, bone2Name=dyn_skirt_fr_c_01, swing1LimitDegrees=2, swing2LimitDegrees=5, twistLimitDegrees=2}
 {bone1Name=dyn_skirt_fr_c_03, bone2Name=dyn_skirt_fr_c_02, swing1LimitDegrees=10, swing2LimitDegrees=5, twistLimitDegrees=2}
]
