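# Control Rig for the TameEagle female body's face accessory part, posing three long hair chains (left, center, right) against the body.
# Control nulls follow weighted blends of head, clavicle, spine and pelvis bones, feeding B-spline IK and twist distribution along each chain.
# Corrective offsets respond to head, neck, spine, clavicle and arm poses through swing/twist delta and spherical pose readers, plus gameplay variables like rifle aim pitch.
# Several alternative nodes are left unconnected.
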ControlRigBlueprint /BRCosmetics/Characters/Player/Female/Medium/Bodies/F_MED_TameEagle/Meshes/Parts/F_MED_TameEagle_FaceAcc_CtrlRig
graphs [
 CollapseNode_ContainedGraph {nodes [
   RigVMFunctionEntryNode()
   > &RigUnit_ParentConstraint(
    AdvancedSettings=(RotationOrderForFilter=XZY)
    Child=(Type=Null,Name="dyn_L_hair_1_ctrl_null")
    Filter=(TranslationFilter=(bX=True,bY=True,bZ=True),RotationFilter=(bX=True,bY=True,bZ=True),ScaleFilter=(bX=False,bY=False,bZ=False))
    Parents=((Item=(Type=Bone,Name="head"),Weight=1))
    Weight=1
   )
   > &RigUnit_ParentConstraint(
    AdvancedSettings=(RotationOrderForFilter=XZY)
    Child=(Type=Null,Name="dyn_C_hair_1_ctrl_null")
    Filter=(TranslationFilter=(bX=True,bY=True,bZ=True),RotationFilter=(bX=True,bY=True,bZ=True),ScaleFilter=(bX=False,bY=False,bZ=False))
    Parents=((Item=(Type=Bone,Name="head"),Weight=1))
    Weight=1
   )
   > &RigUnit_ParentConstraint(
    AdvancedSettings=(RotationOrderForFilter=XZY)
    Child=(Type=Null,Name="dyn_R_hair_1_ctrl_null")
    Filter=(TranslationFilter=(bX=True,bY=True,bZ=True),RotationFilter=(bX=True,bY=True,bZ=True),ScaleFilter=(bX=False,bY=False,bZ=False))
    Parents=((Item=(Type=Bone,Name="head"),Weight=1))
    Weight=1
   )
   > &RigUnit_ParentConstraint(
    AdvancedSettings=(RotationOrderForFilter=XZY)
    Child=(Type=Null,Name="dyn_L_hair_2_ctrl_null")
    Filter=(TranslationFilter=(bX=True,bY=True,bZ=True),RotationFilter=(bX=True,bY=True,bZ=True),ScaleFilter=(bX=False,bY=False,bZ=False))
    Parents=((Item=(Type=Bone,Name="head"),Weight=0.5),(Item=(Type=Bone,Name="spine_05"),Weight=0.5))
    Weight=1
   )
   > &RigUnit_ParentConstraint(
    AdvancedSettings=(RotationOrderForFilter=XZY)
    Child=(Type=Null,Name="dyn_C_hair_2_ctrl_null")
    Filter=(TranslationFilter=(bX=True,bY=True,bZ=True),RotationFilter=(bX=True,bY=True,bZ=True),ScaleFilter=(bX=False,bY=False,bZ=False))
    Parents=((Item=(Type=Bone,Name="head"),Weight=0.5),(Item=(Type=Bone,Name="spine_05"),Weight=0.5))
    Weight=1
   )
   > &RigUnit_ParentConstraint(
    AdvancedSettings=(RotationOrderForFilter=XZY)
    Child=(Type=Null,Name="dyn_R_hair_2_ctrl_null")
    Filter=(TranslationFilter=(bX=True,bY=True,bZ=True),RotationFilter=(bX=True,bY=True,bZ=True),ScaleFilter=(bX=False,bY=False,bZ=False))
    Parents=((Item=(Type=Bone,Name="head"),Weight=0.5),(Item=(Type=Bone,Name="spine_05"),Weight=0.5))
    Weight=1
   )
   > &RigUnit_ParentConstraint(
    AdvancedSettings=(RotationOrderForFilter=XZY)
    Child=(Type=Null,Name="dyn_L_hair_3_ctrl_null")
    Filter=(TranslationFilter=(bX=True,bY=True,bZ=True),RotationFilter=(bX=True,bY=True,bZ=True),ScaleFilter=(bX=False,bY=False,bZ=False))
    Parents=((Item=(Type=Bone,Name="upperarm_twist_01_l"),Weight=0.05),(Item=(Type=Bone,Name="spine_05"),Weight=0.35),(Item=(Type=Bone,Name="clavicle_l"),Weight=0.6))
    Weight=1
   )
   > &RigUnit_ParentConstraint(
    AdvancedSettings=(RotationOrderForFilter=XZY)
    Child=(Type=Null,Name="dyn_C_hair_3_ctrl_null")
    Filter=(TranslationFilter=(bX=True,bY=True,bZ=True),RotationFilter=(bX=True,bY=True,bZ=True),ScaleFilter=(bX=False,bY=False,bZ=False))
    Parents=((Item=(Type=Bone,Name="clavicle_l"),Weight=0.05),(Item=(Type=Bone,Name="clavicle_r"),Weight=0.05),(Item=(Type=Bone,Name="spine_05"),Weight=0.9))
    Weight=1
   )
   > &RigUnit_ParentConstraint(
    AdvancedSettings=(RotationOrderForFilter=XZY)
    Child=(Type=Null,Name="dyn_R_hair_3_ctrl_null")
    Filter=(TranslationFilter=(bX=True,bY=True,bZ=True),RotationFilter=(bX=True,bY=True,bZ=True),ScaleFilter=(bX=False,bY=False,bZ=False))
    Parents=((Item=(Type=Bone,Name="upperarm_twist_01_r"),Weight=0.05),(Item=(Type=Bone,Name="spine_05"),Weight=0.35),(Item=(Type=Bone,Name="clavicle_r"),Weight=0.6))
    Weight=1
   )
   > &RigUnit_ParentConstraint(
    AdvancedSettings=(RotationOrderForFilter=XZY)
    Child=(Type=Null,Name="dyn_L_hair_4_ctrl_null")
    Filter=(TranslationFilter=(bX=True,bY=True,bZ=True),RotationFilter=(bX=True,bY=True,bZ=True),ScaleFilter=(bX=False,bY=False,bZ=False))
    Parents="((Item=(Type=Bone,Name=\"spine_02\"),Weight=0.0014),(Item=(Type=Bone,Name=\"spine_03\"),Weight=0.0235),(Item=(Type=Bone,Name=\"spine_04\"),Weight=0.0379),(Item=(Type=Bone,Name=\"upperarm_twist_01_l\"),Weight=0.0733),(Item=(Type=Bone,Name=\"spine_05\"),Weight=0.3523) ...(truncated)"
    Weight=1
   )
   > &RigUnit_ParentConstraint(
    AdvancedSettings=(RotationOrderForFilter=XZY)
    Child=(Type=Null,Name="dyn_C_hair_4_ctrl_null")
    Filter=(TranslationFilter=(bX=True,bY=True,bZ=True),RotationFilter=(bX=True,bY=True,bZ=True),ScaleFilter=(bX=False,bY=False,bZ=False))
    Parents="((Item=(Type=Bone,Name=\"spine_01\"),Weight=0.0001),(Item=(Type=Bone,Name=\"spine_02\"),Weight=0.0068),(Item=(Type=Bone,Name=\"clavicle_l\"),Weight=0.063),(Item=(Type=Bone,Name=\"clavicle_r\"),Weight=0.0705),(Item=(Type=Bone,Name=\"spine_03\"),Weight=0.1021),(Item=( ...(truncated)"
    Weight=1
   )
   > &RigUnit_ParentConstraint(
    AdvancedSettings=(RotationOrderForFilter=XZY)
    Child=(Type=Null,Name="dyn_R_hair_4_ctrl_null")
    Filter=(TranslationFilter=(bX=True,bY=True,bZ=True),RotationFilter=(bX=True,bY=True,bZ=True),ScaleFilter=(bX=False,bY=False,bZ=False))
    Parents="((Item=(Type=Bone,Name=\"spine_02\"),Weight=0.0014),(Item=(Type=Bone,Name=\"spine_03\"),Weight=0.0235),(Item=(Type=Bone,Name=\"spine_04\"),Weight=0.038),(Item=(Type=Bone,Name=\"upperarm_twist_01_r\"),Weight=0.0733),(Item=(Type=Bone,Name=\"spine_05\"),Weight=0.3528), ...(truncated)"
    Weight=1
   )
   > &RigUnit_ParentConstraint(
    AdvancedSettings=(RotationOrderForFilter=XZY)
    Child=(Type=Null,Name="dyn_L_hair_5_ctrl_null")
    Filter=(TranslationFilter=(bX=True,bY=True,bZ=True),RotationFilter=(bX=True,bY=True,bZ=True),ScaleFilter=(bX=False,bY=False,bZ=False))
    Parents=((Item=(Type=Bone,Name="clavicle_l"),Weight=0.1),(Item=(Type=Bone,Name="spine_04"),Weight=0.15),(Item=(Type=Bone,Name="spine_05"),Weight=0.15),(Item=(Type=Bone,Name="spine_02"),Weight=0.2),(Item=(Type=Bone,Name="spine_03"),Weight=0.4))
    Weight=1
   )
   > &RigUnit_ParentConstraint(
    AdvancedSettings=(RotationOrderForFilter=XZY)
    Child=(Type=Null,Name="dyn_C_hair_5_ctrl_null")
    Filter=(TranslationFilter=(bX=True,bY=True,bZ=True),RotationFilter=(bX=True,bY=True,bZ=True),ScaleFilter=(bX=False,bY=False,bZ=False))
    Parents=((Item=(Type=Bone,Name="spine_04"),Weight=0.15),(Item=(Type=Bone,Name="spine_05"),Weight=0.1),(Item=(Type=Bone,Name="spine_02"),Weight=0.25),(Item=(Type=Bone,Name="spine_03"),Weight=0.5))
    Weight=1
   )
   > &RigUnit_ParentConstraint(
    AdvancedSettings=(RotationOrderForFilter=XZY)
    Child=(Type=Null,Name="dyn_R_hair_5_ctrl_null")
    Filter=(TranslationFilter=(bX=True,bY=True,bZ=True),RotationFilter=(bX=True,bY=True,bZ=True),ScaleFilter=(bX=False,bY=False,bZ=False))
    Parents=((Item=(Type=Bone,Name="clavicle_r"),Weight=0.1),(Item=(Type=Bone,Name="spine_04"),Weight=0.15),(Item=(Type=Bone,Name="spine_05"),Weight=0.15),(Item=(Type=Bone,Name="spine_02"),Weight=0.2),(Item=(Type=Bone,Name="spine_03"),Weight=0.4))
    Weight=1
   )
   > &RigUnit_ParentConstraint(
    AdvancedSettings=(RotationOrderForFilter=XZY)
    Child=(Type=Null,Name="dyn_L_hair_6_ctrl_null")
    Filter=(TranslationFilter=(bX=True,bY=True,bZ=True),RotationFilter=(bX=True,bY=True,bZ=True),ScaleFilter=(bX=False,bY=False,bZ=False))
    Parents=((Item=(Type=Bone,Name="spine_02"),Weight=0.075),(Item=(Type=Bone,Name="spine_01"),Weight=0.425),(Item=(Type=Bone,Name="pelvis"),Weight=0.5))
    Weight=1
   )
   > &RigUnit_ParentConstraint(
    AdvancedSettings=(RotationOrderForFilter=XZY)
    Child=(Type=Null,Name="dyn_C_hair_6_ctrl_null")
    Filter=(TranslationFilter=(bX=True,bY=True,bZ=True),RotationFilter=(bX=True,bY=True,bZ=True),ScaleFilter=(bX=False,bY=False,bZ=False))
    Parents=((Item=(Type=Bone,Name="spine_02"),Weight=0.075),(Item=(Type=Bone,Name="spine_01"),Weight=0.425),(Item=(Type=Bone,Name="pelvis"),Weight=0.5))
    Weight=1
   )
   > &RigUnit_ParentConstraint(
    AdvancedSettings=(RotationOrderForFilter=XZY)
    Child=(Type=Null,Name="dyn_R_hair_6_ctrl_null")
    Filter=(TranslationFilter=(bX=True,bY=True,bZ=True),RotationFilter=(bX=True,bY=True,bZ=True),ScaleFilter=(bX=False,bY=False,bZ=False))
    Parents=((Item=(Type=Bone,Name="spine_02"),Weight=0.075),(Item=(Type=Bone,Name="spine_01"),Weight=0.425),(Item=(Type=Bone,Name="pelvis"),Weight=0.5))
    Weight=1
   )
   > &RigUnit_ParentConstraint(
    AdvancedSettings=(RotationOrderForFilter=XZY)
    Child=(Type=Null,Name="dyn_L_hair_7_ctrl_null")
    Filter=(TranslationFilter=(bX=True,bY=True,bZ=True),RotationFilter=(bX=True,bY=True,bZ=True),ScaleFilter=(bX=False,bY=False,bZ=False))
    Parents=((Item=(Type=Bone,Name="spine_01"),Weight=0.02),(Item=(Type=Bone,Name="pelvis"),Weight=0.98))
    Weight=1
   )
   > &RigUnit_ParentConstraint(
    AdvancedSettings=(RotationOrderForFilter=XZY)
    Child=(Type=Null,Name="dyn_C_hair_7_ctrl_null")
    Filter=(TranslationFilter=(bX=True,bY=True,bZ=True),RotationFilter=(bX=True,bY=True,bZ=True),ScaleFilter=(bX=False,bY=False,bZ=False))
    Parents=((Item=(Type=Bone,Name="spine_01"),Weight=0.02),(Item=(Type=Bone,Name="pelvis"),Weight=0.98))
    Weight=1
   )
   > &RigUnit_ParentConstraint(
    AdvancedSettings=(RotationOrderForFilter=XZY)
    Child=(Type=Null,Name="dyn_R_hair_7_ctrl_null")
    Filter=(TranslationFilter=(bX=True,bY=True,bZ=True),RotationFilter=(bX=True,bY=True,bZ=True),ScaleFilter=(bX=False,bY=False,bZ=False))
    Parents=((Item=(Type=Bone,Name="spine_01"),Weight=0.02),(Item=(Type=Bone,Name="pelvis"),Weight=0.98))
    Weight=1
   )
   > RigVMFunctionReturnNode()
  ]}
 New Function_ContainedGraph {nodes [
   n1 = RigVMFunctionEntryNode()
   RigVMFunction_ControlFlowBranch = &RigVMFunction_ControlFlowBranch(ExecuteContext=@n1.ExecuteContext, Condition=@n1.debug)
   RigVMFunctionReturnNode(ExecuteContext=@RigVMFunction_ControlFlowBranch.Completed)
   RigUnit_ItemArray = &RigUnit_ItemArray(Items="((Type=Bone,Name=\"spine_01\"),(Type=Bone,Name=\"spine_02\"),(Type=Bone,Name=\"spine_03\"),(Type=Bone,Name=\"spine_04\"),(Type=Bone,Name=\"spine_05\"),(Type=Bone,Name=\"neck_01\"),(Type=Bone,Name=\"neck_02\"),(Type=Bone,Name=\"head\"),(Type=Bone,Name=\"clavicle_l\"),(Type=B ...(truncated)")
   DISPATCH_RigVMDispatch_ArrayIterator = RigVMDispatchNode(Array=@RigUnit_ItemArray.Items, ExecuteContext=@RigVMFunction_ControlFlowBranch.True) {template="DISPATCH_RigVMDispatch_ArrayIterator(in Array,out Element,out Index,out Count,out Ratio)"}
   Concat = /Script/RigVM.RigVMFunction_NameConcat(B=_ctrl, A=@DISPATCH_RigVMDispatch_ArrayIterator.Element.Name) {is=struct RigVMUnitNode, template="Concat::Execute(in A,in B,out Result)"}
   > &RigUnit_SetTransform(
    Item=(Type=Bone,Name="spine_01")
    Weight=1
    bInitial=False
    Item.Name=@DISPATCH_RigVMDispatch_ArrayIterator.Element.Name
    Value=&RigUnit_GetTransform(Item=(Type=Control,Name="spine_01_ctrl"), bInitial=False, Item.Name=@Concat.Result).Transform
   )
   &RigUnit_SetTransform(
    Item=(Type=Bone,Name="pelvis")
    Weight=1
    bInitial=False
    bPropagateToChildren=false
    Value=&RigUnit_GetTransform(Item=(Type=Control,Name="pelvis_ctrl"), bInitial=False).Transform
    ExecutePin=@DISPATCH_RigVMDispatch_ArrayIterator.Completed
   )
  ]}
 New Function_ContainedGraph {nodes [
   RigVMFunctionEntryNode()
   > RigVMCollapseNode()
   > RigVMFunction_Sequence = RigVMAggregateNode()
   TwistBones_2 = RigVMFunctionReferenceNode(
    Axis to Twist=(bX=true,bY=true,bZ=true)
    Bones to Twist=((Type=Bone,Name="dyn_C_hair_3"),(Type=Bone,Name="dyn_C_hair_4"),(Type=Bone,Name="dyn_C_hair_5"),(Type=Bone,Name="dyn_C_hair_6"))
    End=(Type=Bone,Name="neck_01")
    Start=(Type=Bone,Name="pelvis")
    ExecuteContext=@RigVMFunction_Sequence.A
   )
   > TwistBones_2_1 = RigVMFunctionReferenceNode(
    Axis to Twist=(bX=True,bY=true,bZ=true)
    Bones to Twist=((Type=Bone,Name="dyn_L_hair_3"),(Type=Bone,Name="dyn_L_hair_4"),(Type=Bone,Name="dyn_L_hair_5"))
    End=(Type=Bone,Name="neck_01")
    Start=(Type=Bone,Name="pelvis")
   )
   > TwistBones_2_1_1 = RigVMFunctionReferenceNode(
    Axis to Twist=(bX=True,bY=true,bZ=true)
    Bones to Twist=((Type=Bone,Name="dyn_R_hair_3"),(Type=Bone,Name="dyn_R_hair_4"),(Type=Bone,Name="dyn_R_hair_5"))
    End=(Type=Bone,Name="neck_01")
    Start=(Type=Bone,Name="pelvis")
   )
   SplineIK_2 = RigVMFunctionReferenceNode(
    Bones=((Type=Bone,Name="dyn_C_hair_1"),(Type=Bone,Name="dyn_C_hair_2"),(Type=Bone,Name="dyn_C_hair_3"),(Type=Bone,Name="dyn_C_hair_4"),(Type=Bone,Name="dyn_C_hair_5"),(Type=Bone,Name="dyn_C_hair_6"))
    Controls=((Type=Control,Name="dyn_C_hair_1_ctrl"),(Type=Control,Name="dyn_C_hair_2_ctrl"),(Type=Control,Name="dyn_C_hair_3_ctrl"),(Type=Control,Name="dyn_C_hair_4_ctrl"),(Type=Control,Name="dyn_C_hair_5_ctrl"),(Type=Control,Name="dyn_C_hair_6_ctrl"))
    Primary Axis=(X=1,Y=0,Z=0)
    Samples Per Segment=1
    Secondary Spline Direction=(X=0,Y=0,Z=0)
    Spline Mode=BSpline
    Stretch=false
    Up Axis=(X=0,Y=0,Z=0)
    debug=false
    ExecuteContext=@RigVMFunction_Sequence.B
   )
   > SplineIK_2_1 = RigVMFunctionReferenceNode(
    Bones=((Type=Bone,Name="dyn_L_hair_1"),(Type=Bone,Name="dyn_L_hair_2"),(Type=Bone,Name="dyn_L_hair_3"),(Type=Bone,Name="dyn_L_hair_4"),(Type=Bone,Name="dyn_L_hair_5"))
    Controls=((Type=Control,Name="dyn_L_hair_1_ctrl"),(Type=Control,Name="dyn_L_hair_2_ctrl"),(Type=Control,Name="dyn_L_hair_3_ctrl"),(Type=Control,Name="dyn_L_hair_4_ctrl"),(Type=Control,Name="dyn_L_hair_5_ctrl"))
    Primary Axis=(X=1,Y=0,Z=0)
    Samples Per Segment=1
    Secondary Spline Direction=(X=0,Y=0,Z=0)
    Spline Mode=BSpline
    Stretch=false
    Up Axis=(X=0,Y=0,Z=0)
    debug=false
   )
   > SplineIK_2_1_1 = RigVMFunctionReferenceNode(
    Bones=((Type=Bone,Name="dyn_R_hair_1"),(Type=Bone,Name="dyn_R_hair_2"),(Type=Bone,Name="dyn_R_hair_3"),(Type=Bone,Name="dyn_R_hair_4"),(Type=Bone,Name="dyn_R_hair_5"))
    Controls=((Type=Control,Name="dyn_R_hair_1_ctrl"),(Type=Control,Name="dyn_R_hair_2_ctrl"),(Type=Control,Name="dyn_R_hair_3_ctrl"),(Type=Control,Name="dyn_R_hair_4_ctrl"),(Type=Control,Name="dyn_R_hair_5_ctrl"))
    Primary Axis=(X=1,Y=0,Z=0)
    Samples Per Segment=1
    Secondary Spline Direction=(X=0,Y=0,Z=0)
    Spline Mode=BSpline
    Stretch=false
    Up Axis=(X=0,Y=0,Z=0)
    debug=false
   )
   RigVMFunctionReturnNode(ExecuteContext=@RigVMFunction_Sequence.D)
   SplineIK_1 = RigVMFunctionReferenceNode(
    Bones=((Type=Bone,Name="dyn_C_hair_1"),(Type=Bone,Name="dyn_C_hair_2"),(Type=Bone,Name="dyn_C_hair_3"),(Type=Bone,Name="dyn_C_hair_4"),(Type=Bone,Name="dyn_C_hair_5"),(Type=Bone,Name="dyn_C_hair_6"),(Type=Bone,Name="dyn_C_hair_end_c"))
    Controls="((Type=Control,Name=\"dyn_C_hair_1_ctrl\"),(Type=Control,Name=\"dyn_C_hair_2_ctrl\"),(Type=Control,Name=\"dyn_C_hair_3_ctrl\"),(Type=Control,Name=\"dyn_C_hair_4_ctrl\"),(Type=Control,Name=\"dyn_C_hair_5_ctrl\"),(Type=Control,Name=\"dyn_C_hair_6_ctrl\"),(Type=Control,N ...(truncated)"
    ExecuteContext=()
    Primary Axis=(X=1,Y=0,Z=0)
    Samples Per Segment=1
    Secondary Spline Direction=(X=0,Y=0,Z=0)
    Spline Mode=BSpline
    Stretch=false
    Up Axis=(X=0,Y=0,Z=0)
    debug=false
   )
   > SplineIK_2_2 = RigVMFunctionReferenceNode(
    Bones=((Type=Bone,Name="dyn_L_hair_1"),(Type=Bone,Name="dyn_L_hair_2"),(Type=Bone,Name="dyn_L_hair_3"),(Type=Bone,Name="dyn_L_hair_4"),(Type=Bone,Name="dyn_L_hair_5"),(Type=Bone,Name="dyn_L_hair_end_r"))
    Controls=((Type=Control,Name="dyn_L_hair_1_ctrl"),(Type=Control,Name="dyn_L_hair_2_ctrl"),(Type=Control,Name="dyn_L_hair_3_ctrl"),(Type=Control,Name="dyn_L_hair_4_ctrl"),(Type=Control,Name="dyn_L_hair_5_ctrl"),(Type=Control,Name="dyn_L_hair_6_ctrl"))
    Primary Axis=(X=1,Y=0,Z=0)
    Samples Per Segment=1
    Secondary Spline Direction=(X=0,Y=0,Z=0)
    Spline Mode=BSpline
    Stretch=false
    Up Axis=(X=0,Y=0,Z=0)
    debug=false
   )
   > SplineIK_2_1_2 = RigVMFunctionReferenceNode(
    Bones=((Type=Bone,Name="dyn_R_hair_1"),(Type=Bone,Name="dyn_R_hair_2"),(Type=Bone,Name="dyn_R_hair_3"),(Type=Bone,Name="dyn_R_hair_4"),(Type=Bone,Name="dyn_R_hair_5"),(Type=Bone,Name="dyn_R_hair_end_l"))
    Controls=((Type=Control,Name="dyn_R_hair_1_ctrl"),(Type=Control,Name="dyn_R_hair_2_ctrl"),(Type=Control,Name="dyn_R_hair_3_ctrl"),(Type=Control,Name="dyn_R_hair_4_ctrl"),(Type=Control,Name="dyn_R_hair_5_ctrl"),(Type=Control,Name="dyn_R_hair_6_ctrl"))
    Primary Axis=(X=1,Y=0,Z=0)
    Samples Per Segment=1
    Secondary Spline Direction=(X=0,Y=0,Z=0)
    Spline Mode=BSpline
    Stretch=false
    Up Axis=(X=0,Y=0,Z=0)
    debug=false
   )
   TwistBones_1 = RigVMFunctionReferenceNode(
    Axis to Twist=(bX=true,bY=true,bZ=true)
    Bones to Twist=((Type=Bone,Name="dyn_C_hair_3"),(Type=Bone,Name="dyn_C_hair_4"),(Type=Bone,Name="dyn_C_hair_5"),(Type=Bone,Name="dyn_C_hair_6"),(Type=Bone,Name="dyn_C_hair_end_c"))
    End=(Type=Bone,Name="neck_01")
    ExecuteContext=()
    Start=(Type=Bone,Name="pelvis")
   )
   > TwistBones_2_2 = RigVMFunctionReferenceNode(
    Axis to Twist=(bX=True,bY=true,bZ=true)
    Bones to Twist=((Type=Bone,Name="dyn_L_hair_3"),(Type=Bone,Name="dyn_L_hair_4"),(Type=Bone,Name="dyn_L_hair_5"),(Type=Bone,Name="dyn_L_hair_end_r"))
    End=(Type=Bone,Name="neck_01")
    Start=(Type=Bone,Name="pelvis")
   )
   > TwistBones_2_1_2 = RigVMFunctionReferenceNode(
    Axis to Twist=(bX=True,bY=true,bZ=true)
    Bones to Twist=((Type=Bone,Name="dyn_R_hair_3"),(Type=Bone,Name="dyn_R_hair_4"),(Type=Bone,Name="dyn_R_hair_5"),(Type=Bone,Name="dyn_R_hair_end_l"))
    End=(Type=Bone,Name="neck_01")
    Start=(Type=Bone,Name="pelvis")
   )
  ]}
 New Function_ContainedGraph {nodes [
   RigVMFunctionEntryNode()
   > RigVMFunction_Sequence_2 = RigVMAggregateNode()
   n1 = &RigVMFunction_ControlFlowBranch(ExecuteContext=@RigVMFunction_Sequence_2.F, Condition=RigVMVariableNode() {variable=FE}.Value)
   &RigUnit_ModifyTransforms(
    ItemToModify="((Item=(Type=Bone,Name=\"dyn_R_hair_3\"),Transform=(Rotation=(X=0,Y=0,Z=-0.0872,W=0.9962),Translation=(X=0,Y=-2,Z=-1),Scale3D=(X=1,Y=1,Z=1))),(Item=(Type=Bone,Name=\"dyn_L_hair_3\"),Transform=(Rotation=(X=0,Y=0,Z=0.0872,W=0.9962),Translation=(X=0,Y=2,Z=-2),Sca ...(truncated)"
    Weight=1
    WeightMinimum=0
    ExecutePin=@n1.True
   )
   RigVMFunctionReturnNode(ExecuteContext=@RigVMFunction_Sequence_2.G)
   n2 = &RigUnit_SphericalPoseReader(
    ActiveRegionScaleFactors=(PositiveWidth=1,NegativeWidth=1,PositiveHeight=1,NegativeHeight=1)
    ActiveRegionSize=0.1
    Debug=(bDrawDebug=false,bDraw2D=False,bDrawLocalAxes=False,DebugScale=25,DebugSegments=20,DebugThickness=0.25)
    DriverAxis=(X=1,Y=0,Z=0)
    DriverItem=(Type=Bone,Name="upperarm_l")
    FalloffRegionScaleFactors=(PositiveWidth=1,NegativeWidth=1,PositiveHeight=1,NegativeHeight=1)
    FalloffSize=0.3
    FlipHeightScaling=False
    FlipWidthScaling=False
    OptionalParentItem=(Type=Bone,Name="None")
    RotationOffset=(X=-140,Y=50,Z=0)
    ExecutePin=@n15.A
   )
   n3 = &RigUnit_ModifyTransforms(
    ExecutePin=()
    ItemToModify="((Item=(Type=Bone,Name=\"dyn_L_hair_2\"),Transform=(Rotation=(X=0,Y=0,Z=0,W=1),Translation=(X=0,Y=2,Z=0),Scale3D=(X=1,Y=1,Z=1))),(Item=(Type=Bone,Name=\"dyn_L_hair_3\"),Transform=(Rotation=(X=0,Y=0,Z=0,W=1),Translation=(X=0,Y=4,Z=0),Scale3D=(X=1,Y=1,Z=1))),(It ...(truncated)"
    WeightMinimum=0
    Weight=@n2.OutputParam
   )
   n4 = &RigUnit_SphericalPoseReader(
    ActiveRegionScaleFactors=(PositiveWidth=1,NegativeWidth=1,PositiveHeight=1,NegativeHeight=1)
    ActiveRegionSize=0.1
    Debug=(bDrawDebug=false,DebugScale=25,DebugSegments=20,DebugThickness=0.25)
    DriverAxis=(X=-1,Y=0,Z=0)
    DriverItem=(Type=Bone,Name="upperarm_r")
    FalloffRegionScaleFactors=(PositiveWidth=1,NegativeWidth=1,PositiveHeight=1,NegativeHeight=1)
    FalloffSize=0.3
    FlipHeightScaling=false
    FlipWidthScaling=False
    OptionalParentItem=(Type=Bone,Name="None")
    RotationOffset=(X=40,Y=50,Z=0)
    ExecutePin=@n2.ExecutePin
   )
   &RigUnit_ModifyTransforms(
    ItemToModify="((Item=(Type=Bone,Name=\"dyn_R_hair_2\"),Transform=(Rotation=(X=0,Y=0,Z=0,W=1),Translation=(X=0,Y=-2,Z=0),Scale3D=(X=1,Y=1,Z=1))),(Item=(Type=Bone,Name=\"dyn_R_hair_3\"),Transform=(Rotation=(X=0,Y=0,Z=0,W=1),Translation=(X=0,Y=-4,Z=0),Scale3D=(X=1,Y=1,Z=1))),( ...(truncated)"
    WeightMinimum=0
    ExecutePin=@n3.ExecutePin
    Weight=@n4.OutputParam
   )
   n5 = RigVMRerouteNode(Value=())
   n6 = &RigVMFunction_ControlFlowBranch(ExecuteContext=@n5.Value, Condition=RigVMVariableNode() {variable=pub diveFloating}.Value)
   claviclesBackFoward_3 = RigVMFunctionReferenceNode(ExecuteContext=@n6.True)
   n7 = RigVMRerouteNode(Value=())
   RigVMFunction_Sequence_5 = RigVMAggregateNode(ExecuteContext=@n7.Value)
   DeltaRotation_8 = RigVMFunctionReferenceNode(
    Rotation Order=ZYX
    child=neck_01
    debugRotation=false
    max=-23
    min=0
    parent=neck_01_null
    parentType=Null
    primAxis=2
    twistAxis=(X=1,Y=0,Z=0)
    ExecuteContext=@RigVMFunction_Sequence_5.A
   )
   > alterTransform_4_1_1 = RigVMFunctionReferenceNode(
    type=Bone
    transforms="((Bone=\"dyn_C_hair_2\",Transform=(Rotation=(X=0,Y=0,Z=0,W=1),Translation=(X=0,Y=0,Z=0),Scale3D=(X=1,Y=1,Z=1))),(Bone=\"dyn_L_hair_2\",Transform=(Rotation=(X=0,Y=0,Z=0,W=1),Translation=(X=0,Y=0,Z=0),Scale3D=(X=1,Y=1,Z=1))),(Bone=\"dyn_R_hair_2\",Transform=(Rotat ...(truncated)"
    weight=@DeltaRotation_8.outputWeight
   )
   > alterTransform_4_1_1_1 = RigVMFunctionReferenceNode(
    type=Bone
    transforms="((Bone=\"dyn_C_hair_3\",Transform=(Rotation=(X=0,Y=0,Z=0,W=1),Translation=(X=0,Y=0,Z=0),Scale3D=(X=1,Y=1,Z=1))),(Bone=\"dyn_L_hair_3\",Transform=(Rotation=(X=0,Y=0,Z=0,W=1),Translation=(X=0,Y=0,Z=0),Scale3D=(X=1,Y=1,Z=1))),(Bone=\"dyn_R_hair_3\",Transform=(Rotat ...(truncated)"
    weight=@DeltaRotation_8.outputWeight
   )
   > DeltaRotation_1_4 = RigVMFunctionReferenceNode(Rotation Order=ZYX, child=head, debugRotation=false, max=-20, min=0, parent=head_null, parentType=Null, primAxis=2, twistAxis=(X=1,Y=0,Z=0))
   n8 = &RigVMFunction_AnimEvalRichCurve(
    Curve=(EditorCurveData=(Keys=((InterpMode=RCIM_Cubic),(InterpMode=RCIM_Cubic,Time=0.3879,Value=0.174,ArriveTangent=1,LeaveTangent=1),(InterpMode=RCIM_Cubic,TangentMode=RCTM_User,Time=1,Value=1,ArriveTangent=1.9233,LeaveTangent=1.9233))))
    SourceMinimum=0
    TargetMinimum=0
    Value=@DeltaRotation_1_4.outputWeight
   )
   > alterTransform_4_2_3_2 = RigVMFunctionReferenceNode(
    type=Bone
    transforms="((Bone=\"dyn_C_hair_1\",Transform=(Rotation=(X=0,Y=0.0872,Z=0,W=0.9962),Translation=(X=0,Y=0,Z=-3),Scale3D=(X=1,Y=1,Z=1))),(Bone=\"dyn_L_hair_1\",Transform=(Rotation=(X=0,Y=0.0872,Z=0,W=0.9962),Translation=(X=0,Y=0,Z=-3),Scale3D=(X=1,Y=1,Z=1))),(Bone=\"dyn_R_ha ...(truncated)"
    weight=@n8.Result
   )
   > alterTransform_4_1_2_2 = RigVMFunctionReferenceNode(
    type=Bone
    transforms="((Bone=\"dyn_C_hair_2\",Transform=(Rotation=(X=0,Y=0.0872,Z=0,W=0.9962),Translation=(X=0,Y=0,Z=1),Scale3D=(X=1,Y=1,Z=1))),(Bone=\"dyn_L_hair_2\",Transform=(Rotation=(X=0,Y=0.0872,Z=0,W=0.9962),Translation=(X=0,Y=0,Z=-1),Scale3D=(X=1,Y=1,Z=1))),(Bone=\"dyn_R_hai ...(truncated)"
    weight=@n8.Result
   )
   > alterTransform_4_1_3_2 = RigVMFunctionReferenceNode(
    type=Bone
    transforms="((Bone=\"dyn_C_hair_2\",Transform=(Rotation=(X=0,Y=-0.0872,Z=0,W=0.9962),Translation=(X=0,Y=0,Z=-1),Scale3D=(X=1,Y=1,Z=1))),(Bone=\"dyn_L_hair_2\",Transform=(Rotation=(X=0,Y=-0.0872,Z=0,W=0.9962),Translation=(X=0,Y=0,Z=-1),Scale3D=(X=1,Y=1,Z=1))),(Bone=\"dyn_R_ ...(truncated)"
    weight=&RigVMFunction_MathDoubleMul(B=@DeltaRotation_8.outputWeight, A=@n8.Result).Result
   )
   > DeltaRotation_5 = RigVMFunctionReferenceNode(Rotation Order=ZYX, child=head, debugRotation=false, max=60, min=0, parent=head_null, parentType=Null, primAxis=2, twistAxis=(X=1,Y=0,Z=0))
   > alterTransform_4_2_4 = RigVMFunctionReferenceNode(
    type=Bone
    transforms="((Bone=\"dyn_C_hair_2\",Transform=(Rotation=(X=0,Y=0,Z=0,W=1),Translation=(X=0,Y=0,Z=-3),Scale3D=(X=1,Y=1,Z=1))),(Bone=\"dyn_L_hair_2\",Transform=(Rotation=(X=0,Y=0,Z=0,W=1),Translation=(X=0,Y=0,Z=-3),Scale3D=(X=1,Y=1,Z=1))),(Bone=\"dyn_R_hair_2\",Transform=(Rot ...(truncated)"
    weight=@DeltaRotation_5.outputWeight
   )
   > DeltaRotation_1_1_1_1_2 = RigVMFunctionReferenceNode(
    Rotation Order=ZYX
    child=head
    debugRotation=false
    max=20
    min=-20
    onlyTwist=true
    parent=head_null_sp
    parentType=Null
    primAxis=0
    twistAxis=(X=1,Y=0,Z=0)
   )
   > alterTransform_4_2_2_1_1_1_2 = RigVMFunctionReferenceNode(
    type=Bone
    transforms="((Bone=\"dyn_C_hair_1\",Transform=(Rotation=(X=0,Y=0,Z=0,W=1),Translation=(X=0,Y=2,Z=0),Scale3D=(X=1,Y=1,Z=1))),(Bone=\"dyn_R_hair_2\",Transform=(Rotation=(X=0,Y=0,Z=0,W=1),Translation=(X=0,Y=0,Z=-5),Scale3D=(X=1,Y=1,Z=1))),(Bone=\"dyn_R_hair_3\",Transform=(Rota ...(truncated)"
    weight=&RigVMFunction_MathDoubleMul(B=0.5, A=@DeltaRotation_1_1_1_1_2.outputWeight).Result
   )
   > DeltaRotation_1_1_1_2 = RigVMFunctionReferenceNode(
    Rotation Order=ZYX
    child=head
    debugRotation=false
    max=-50
    min=-15
    onlyTwist=true
    parent=head_null_sp
    parentType=Null
    primAxis=0
    twistAxis=(X=1,Y=0,Z=0)
   )
   > alterTransform_4_2_2_1_1_2 = RigVMFunctionReferenceNode(
    type=Bone
    transforms="((Bone=\"dyn_C_hair_1\",Transform=(Rotation=(X=0,Y=0,Z=0,W=1),Translation=(X=0,Y=0,Z=0),Scale3D=(X=1,Y=1,Z=1))),(Bone=\"dyn_R_hair_2\",Transform=(Rotation=(X=-0.2588,Y=0,Z=0,W=0.9659),Translation=(X=6,Y=0,Z=-3),Scale3D=(X=1,Y=1,Z=1))),(Bone=\"dyn_R_hair_3\",Tran ...(truncated)"
    weight=@DeltaRotation_1_1_1_2.outputWeight
   )
   > DeltaRotation_1_2 = RigVMFunctionReferenceNode(
    Rotation Order=ZYX
    child=head
    debugRotation=false
    max=-100
    min=-60
    parent=head_null_sp
    parentType=Null
    primAxis=2
    twistAxis=(X=1,Y=0,Z=0)
   )
   > alterTransform_4_2_2_2 = RigVMFunctionReferenceNode(
    type=Bone
    transforms="((Bone=\"dyn_C_hair_1\",Transform=(Rotation=(X=0,Y=-0.342,Z=0,W=0.9397),Translation=(X=2,Y=0,Z=-2),Scale3D=(X=1,Y=1,Z=1))),(Bone=\"dyn_L_hair_1\",Transform=(Rotation=(X=0,Y=-0.342,Z=0,W=0.9397),Translation=(X=0,Y=0,Z=-2),Scale3D=(X=1,Y=1,Z=1))),(Bone=\"dyn_R_ha ...(truncated)"
    weight=@DeltaRotation_1_2.outputWeight
   )
   > alterTransform_4_2_2_1_2 = RigVMFunctionReferenceNode(
    type=Bone
    transforms="((Bone=\"dyn_C_hair_2\",Transform=(Rotation=(X=0,Y=0,Z=0,W=1),Translation=(X=0,Y=0,Z=2),Scale3D=(X=1,Y=1,Z=1))),(Bone=\"dyn_L_hair_2\",Transform=(Rotation=(X=0,Y=0,Z=0,W=1),Translation=(X=0,Y=0,Z=2),Scale3D=(X=1,Y=1,Z=1))),(Bone=\"dyn_R_hair_2\",Transform=(Rotat ...(truncated)"
    weight=@DeltaRotation_1_2.outputWeight
   )
   DeltaRotation_7 = RigVMFunctionReferenceNode(
    Rotation Order=ZYX
    child=spine_05
    debugRotation=false
    max=-50
    min=-10
    onlyTwist=true
    parent=spine_01
    parentType=Bone
    twistAxis=(X=1,Y=0,Z=0)
    ExecuteContext=@RigVMFunction_Sequence_5.B
   )
   n9 = &RigVMFunction_AnimEvalRichCurve(
    Curve="(EditorCurveData=(Keys=((InterpMode=RCIM_Cubic,TangentMode=RCTM_Auto,TangentWeightMode=RCTWM_WeightedNone,Time=0.0004,Value=0.0068,ArriveTangent=0,ArriveTangentWeight=0,LeaveTangent=0,LeaveTangentWeight=0),(InterpMode=RCIM_Cubic,TangentMode=RCTM_User,Tange ...(truncated)"
    SourceMinimum=0
    TargetMinimum=0
    Value=@DeltaRotation_7.outputWeight
   )
   > alterTransform_4_2_1 = RigVMFunctionReferenceNode(
    type=Bone
    transforms="((Bone=\"dyn_L_hair_2\",Transform=(Rotation=(X=0,Y=0,Z=0,W=1),Translation=(X=0,Y=-1,Z=0),Scale3D=(X=1,Y=1,Z=1))),(Bone=\"dyn_L_hair_3\",Transform=(Rotation=(X=0,Y=0,Z=0,W=1),Translation=(X=0,Y=-1,Z=0),Scale3D=(X=1,Y=1,Z=1))),(Bone=\"dyn_L_hair_4\",Transform=(Rot ...(truncated)"
    weight=@n9.Result
   )
   > DeltaRotation_6 = RigVMFunctionReferenceNode(
    Rotation Order=ZYX
    child=spine_05
    debugRotation=false
    max=50
    min=10
    onlyTwist=true
    parent=spine_01
    parentType=Bone
    twistAxis=(X=1,Y=0,Z=0)
   )
   n10 = &RigVMFunction_AnimEvalRichCurve(
    Curve=(EditorCurveData=(Keys=((InterpMode=RCIM_Cubic),(InterpMode=RCIM_Cubic,TangentMode=RCTM_User,Time=0.5279,Value=0.4056,ArriveTangent=1.3758,LeaveTangent=1.3758),(InterpMode=RCIM_Cubic,Time=1,Value=1))))
    SourceMinimum=0
    TargetMinimum=0
    Value=@DeltaRotation_6.outputWeight
   )
   > alterTransform_4_2_1_2 = RigVMFunctionReferenceNode(
    type=Bone
    transforms="((Bone=\"dyn_R_hair_2\",Transform=(Rotation=(X=0,Y=0,Z=0,W=1),Translation=(X=0,Y=-2,Z=0),Scale3D=(X=1,Y=1,Z=1))),(Bone=\"dyn_R_hair_3\",Transform=(Rotation=(X=0,Y=0.1736,Z=0,W=0.9848),Translation=(X=0,Y=0,Z=-2),Scale3D=(X=1,Y=1,Z=1))),(Bone=\"dyn_R_hair_4\",Tran ...(truncated)"
    weight=@n10.Result
   )
   n11 = &RigVMFunction_ControlFlowBranch(Condition=RigVMVariableNode() {variable=pub isTargeting}.Value, ExecuteContext=@RigVMFunction_Sequence_5.C)
   n12 = RigVMVariableNode() {variable=pub aimPitch}
   n13 = &RigVMFunction_AnimEvalRichCurve(
    Curve=(EditorCurveData=(Keys=((InterpMode=RCIM_Cubic),(InterpMode=RCIM_Cubic,TangentMode=RCTM_User,Time=0.4302,Value=0.2017,ArriveTangent=0.8571,LeaveTangent=0.8571),(Time=1,Value=1))))
    SourceMinimum=0
    TargetMinimum=0
    Value=&RigVMFunction_MathDoubleMul(
     B=&RigVMFunction_MathDoubleRemap(SourceMaximum=0, SourceMinimum=88, TargetMaximum=0, TargetMinimum=1, Value=@n12.Value).Result
     A=RigVMVariableNode() {variable=pub isAssaultRifleEquipped}.Value
    ).Result
   )
   &RigUnit_ModifyTransforms(
    ItemToModify=((Item=(Type=Bone,Name="base_R_hair_1"),Transform=(Rotation=(X=0,Y=0,Z=0,W=1),Translation=(X=0,Y=3,Z=-1.5),Scale3D=(X=1,Y=1,Z=1))))
    WeightMinimum=0
    Weight=@n13.Result
    ExecutePin=@n11.True
   )
   > alterTransform_7 = RigVMFunctionReferenceNode(
    type=Bone
    transforms=((Bone="dyn_R_hair_1",Transform=(Rotation=(X=-0.0872,Y=0,Z=0,W=0.9962),Translation=(X=0,Y=0,Z=0),Scale3D=(X=1,Y=1,Z=1))))
    weight=@n13.Result
   )
   n14 = &RigVMFunction_ControlFlowBranch(ExecuteContext=(), Condition=RigVMVariableNode() {variable=pub disableUpperLimbsCorrection}.Value)
   n15 = &RigVMFunction_Sequence(ExecuteContext=@n14.False)
   claviclesBackFoward_2 = RigVMFunctionReferenceNode(ExecuteContext=@n15.B)
   > n16 = &RigUnit_SphericalPoseReader(
    ActiveRegionScaleFactors=(PositiveWidth=1,NegativeWidth=1,PositiveHeight=1,NegativeHeight=1)
    ActiveRegionSize=0.1
    Debug=(bDrawDebug=false,bDraw2D=False,bDrawLocalAxes=False,DebugScale=25,DebugSegments=20,DebugThickness=0.25)
    DriverAxis=(X=1,Y=0,Z=0)
    DriverItem=(Type=Bone,Name="clavicle_l")
    FalloffRegionScaleFactors=(PositiveWidth=0.5,NegativeWidth=0.5,PositiveHeight=0.5,NegativeHeight=0.5)
    FalloffSize=0.3
    FlipHeightScaling=False
    FlipWidthScaling=False
    OptionalParentItem=(Type=Bone,Name="None")
    RotationOffset=(X=-45,Y=6,Z=60)
   )
   > alterTransform_6 = RigVMFunctionReferenceNode(
    type=Bone
    transforms=((Bone="dyn_L_hair_2",Transform=(Rotation=(X=0,Y=0,Z=0,W=1),Translation=(X=0,Y=3,Z=-1),Scale3D=(X=1,Y=1,Z=1))),(Bone="dyn_L_hair_3",Transform=(Rotation=(X=0,Y=0,Z=0,W=1),Translation=(X=0,Y=6,Z=-1),Scale3D=(X=1,Y=1,Z=1))))
    weight=@n16.OutputParam
   )
   > n17 = &RigUnit_SphericalPoseReader(
    ActiveRegionScaleFactors=(PositiveWidth=1,NegativeWidth=1,PositiveHeight=1,NegativeHeight=1)
    ActiveRegionSize=0.1
    Debug=(bDrawDebug=false,bDraw2D=False,bDrawLocalAxes=False,DebugScale=25,DebugSegments=20,DebugThickness=0.25)
    DriverAxis=(X=-1,Y=0,Z=0)
    DriverItem=(Type=Bone,Name="clavicle_r")
    FalloffRegionScaleFactors=(PositiveWidth=0.5,NegativeWidth=0.5,PositiveHeight=0.5,NegativeHeight=0.5)
    FalloffSize=0.3
    FlipHeightScaling=False
    FlipWidthScaling=False
    OptionalParentItem=(Type=Bone,Name="None")
    RotationOffset=(X=45,Y=174,Z=30)
   )
   > alterTransform_5_2 = RigVMFunctionReferenceNode(
    type=Bone
    transforms=((Bone="dyn_R_hair_2",Transform=(Rotation=(X=0,Y=0,Z=0,W=1),Translation=(X=0,Y=-3,Z=-1),Scale3D=(X=1,Y=1,Z=1))),(Bone="dyn_R_hair_3",Transform=(Rotation=(X=0,Y=0,Z=0,W=1),Translation=(X=0,Y=-5,Z=-1),Scale3D=(X=1,Y=1,Z=1))))
    weight=@n17.OutputParam
   )
   > DeltaRotation_9 = RigVMFunctionReferenceNode(
    Rotation Order=ZXY
    child=clavicle_r
    debugRotation=false
    max=-15
    parent=clavicle_r_null
    parentType=Null
    primAxis=1
    twistAxis=(X=1,Y=0,Z=0)
   )
   > alterTransform_4_2_1_2_2 = RigVMFunctionReferenceNode(
    type=Bone
    transforms=((Bone="dyn_R_hair_2",Transform=(Rotation=(X=0,Y=0,Z=0,W=1),Translation=(X=-1,Y=0,Z=-1),Scale3D=(X=1,Y=1,Z=1))),(Bone="dyn_R_hair_3",Transform=(Rotation=(X=0,Y=0,Z=0,W=1),Translation=(X=-4,Y=-1,Z=-1),Scale3D=(X=1,Y=1,Z=1))))
    weight=@DeltaRotation_9.outputWeight
   )
   > DeltaRotation_4_2 = RigVMFunctionReferenceNode(
    Rotation Order=ZXY
    child=clavicle_l
    debugRotation=false
    max=-10
    parent=clavicle_l_null
    parentType=Null
    primAxis=1
    twistAxis=(X=1,Y=0,Z=0)
   )
   > alterTransform_4_2_1_2_1_2 = RigVMFunctionReferenceNode(
    type=Bone
    transforms=((Bone="dyn_L_hair_2",Transform=(Rotation=(X=0,Y=0,Z=0,W=1),Translation=(X=-1,Y=0,Z=-1),Scale3D=(X=1,Y=1,Z=1))),(Bone="dyn_L_hair_3",Transform=(Rotation=(X=0,Y=0,Z=0,W=1),Translation=(X=-4,Y=1,Z=-1),Scale3D=(X=1,Y=1,Z=1))))
    weight=@DeltaRotation_4_2.outputWeight
   )
   n18 = &RigVMFunction_ControlFlowBranch(ExecuteContext=(), Condition=RigVMVariableNode() {variable=Cape}.Value)
   &RigUnit_ModifyTransforms(
    ItemToModify="((Item=(Type=Bone,Name=\"dyn_L_hair_5\"),Transform=(Rotation=(X=0,Y=-0.1736,Z=0,W=0.9848),Translation=(X=0,Y=0,Z=0),Scale3D=(X=1,Y=1,Z=1))),(Item=(Type=Bone,Name=\"dyn_C_hair_5\"),Transform=(Rotation=(X=0,Y=-0.1736,Z=0,W=0.9848),Translation=(X=0,Y=0,Z=0),Scale ...(truncated)"
    Weight=1
    WeightMinimum=0
    ExecutePin=@n18.True
   )
   n19 = &RigUnit_SphericalPoseReader(
    ActiveRegionScaleFactors=(PositiveWidth=1,NegativeWidth=1,PositiveHeight=1,NegativeHeight=1)
    ActiveRegionSize=0.1
    Debug=(bDrawDebug=true,bDraw2D=False,bDrawLocalAxes=False,DebugScale=25,DebugSegments=20,DebugThickness=0.25)
    DriverAxis=(X=1,Y=0,Z=0)
    DriverItem=(Type=Bone,Name="dyn_C_hair_5")
    ExecutePin=()
    FalloffRegionScaleFactors=(PositiveWidth=1,NegativeWidth=1,PositiveHeight=1,NegativeHeight=1)
    FalloffSize=0.2
    FlipHeightScaling=False
    FlipWidthScaling=False
    OptionalParentItem=(Type=Bone,Name="None")
    RotationOffset=(X=180,Y=70,Z=0)
   )
   > &RigUnit_ModifyTransforms(
    ItemToModify=((Item=(Type=Bone,Name="dyn_C_hair_spike"),Transform=(Rotation=(X=0,Y=0.0872,Z=0,W=0.9962),Translation=(X=-2,Y=0,Z=0),Scale3D=(X=1,Y=1,Z=1))))
    WeightMinimum=0
    Weight=@n19.OutputParam
   )
   n20 = &RigVMFunction_AlphaInterp(
    InterpSpeedDecreasing=2
    InterpSpeedIncreasing=10
    Scale=1
    bClampResult=true
    bMapRange=true
    Value=RigVMVariableNode() {variable=Tsprint}.Value
   )
   > &RigUnit_ModifyTransforms(
    ItemToModify="((Item=(Type=Bone,Name=\"dyn_R_hair_3\"),Transform=(Rotation=(X=0,Y=0.0872,Z=0,W=0.9962),Translation=(X=0,Y=5,Z=-2),Scale3D=(X=1,Y=1,Z=1))),(Item=(Type=Bone,Name=\"dyn_L_hair_3\"),Transform=(Rotation=(X=0,Y=0.0872,Z=0,W=0.9962),Translation=(X=0,Y=2,Z=-3),Scale ...(truncated)"
    WeightMinimum=0
    Weight=@n20.Result
   )
   n21 = &RigVMFunction_AlphaInterp(
    InterpSpeedDecreasing=2
    InterpSpeedIncreasing=8
    Scale=1
    bClampResult=true
    bMapRange=true
    Value=RigVMVariableNode() {variable=Dive}.Value
   )
   > &RigUnit_ModifyTransforms(
    ItemToModify="((Item=(Type=Bone,Name=\"dyn_R_hair_3\"),Transform=(Rotation=(X=0,Y=0,Z=0,W=1),Translation=(X=0,Y=-5,Z=0),Scale3D=(X=1,Y=1,Z=1))),(Item=(Type=Bone,Name=\"dyn_L_hair_3\"),Transform=(Rotation=(X=0,Y=0,Z=0,W=1),Translation=(X=0,Y=5,Z=0),Scale3D=(X=1,Y=1,Z=1))),(I ...(truncated)"
    WeightMinimum=0
    Weight=@n21.Result
   )
   alterTransform_4 = RigVMFunctionReferenceNode(
    type=Bone
    transforms="((Bone=\"dyn_R_hair_2\",Transform=(Rotation=(X=0,Y=0,Z=0,W=1),Translation=(X=0,Y=-2,Z=0),Scale3D=(X=1,Y=1,Z=1))),(Bone=\"dyn_R_hair_3\",Transform=(Rotation=(X=0,Y=0,Z=0,W=1),Translation=(X=0,Y=-4,Z=0),Scale3D=(X=1,Y=1,Z=1))),(Bone=\"dyn_R_hair_4\",Transform=(Rot ...(truncated)"
    weight=@n4.OutputParam
    ExecuteContext=@alterTransform_1.ExecuteContext
   )
   alterTransform_1 = RigVMFunctionReferenceNode(
    type=Bone
    transforms="((Bone=\"dyn_L_hair_2\",Transform=(Rotation=(X=0,Y=0,Z=0,W=1),Translation=(X=0,Y=2,Z=0),Scale3D=(X=1,Y=1,Z=1))),(Bone=\"dyn_L_hair_3\",Transform=(Rotation=(X=0,Y=0,Z=0,W=1),Translation=(X=0,Y=4,Z=0),Scale3D=(X=1,Y=1,Z=1))),(Bone=\"dyn_L_hair_4\",Transform=(Rotat ...(truncated)"
    weight=@n2.OutputParam
    ExecuteContext=@n4.ExecutePin
   )
  ]}
 New Function_ContainedGraph {nodes [
   n1 = RigVMFunctionEntryNode()
   > DISPATCH_RigVMDispatch_ArrayIterator = RigVMDispatchNode(Array=@n1.transforms) {template="DISPATCH_RigVMDispatch_ArrayIterator(in Array,out Element,out Index,out Count,out Ratio)"}
   Multiply = &RigVMFunction_MathTransformMul(
    A=@DISPATCH_RigVMDispatch_ArrayIterator.Element.Transform
    B=&RigUnit_GetTransform(
     Item=(Type=Bone,Name="dyn_L_hair_3")
     Space=LocalSpace
     bInitial=False
     Item.Name=@DISPATCH_RigVMDispatch_ArrayIterator.Element.Bone
     Item.Type=@n1.type
    ).Transform
   ) {template="Multiply::Execute(in A,in B,out Result)"}
   > &RigUnit_SetTransform(
    Item=(Type=Bone,Name="dyn_L_hair_3")
    Space=LocalSpace
    bInitial=false
    Value=@Multiply.Result
    bPropagateToChildren=@n1.inheritsTransform
    Weight=@n1.weight
    Item.Name=@DISPATCH_RigVMDispatch_ArrayIterator.Element.Bone
    Item.Type=@n1.type
   )
   RigVMFunctionReturnNode(ExecuteContext=@DISPATCH_RigVMDispatch_ArrayIterator.Completed)
   &RigUnit_OffsetTransformForItem(
    ExecutePin=()
    Item=(Type=Bone,Name="None")
    OffsetTransform=(Rotation=(X=0,Y=0,Z=0,W=1),Translation=(X=0,Y=0,Z=0),Scale3D=(X=1,Y=1,Z=1))
    Weight=1
    bPropagateToChildren=True
   )
  ]}
 New Function_ContainedGraph {nodes [
   n1 = RigVMFunctionEntryNode()
   RigVMFunction_MathBoolToInteger = &RigVMFunction_MathBoolToInteger(Value=@n1.onlyTwist)
   RigVMFunction_MathQuaternionSwingTwist = &RigVMFunction_MathQuaternionSwingTwist(
    TwistAxis=@n1.twistAxis
    Input=&RigUnit_GetRelativeTransformForItem(
     Child=(Type=Bone,Name="None")
     Parent=(Type=Bone,Name="None")
     bChildInitial=False
     bParentInitial=False
     Child.Name=@n1.child
     Parent.Name=@n1.parent
     Parent.Type=@n1.parentType
    ).RelativeTransform.Rotation
   )
   RigVMFunction_MathQuaternionToEuler = &RigVMFunction_MathQuaternionToEuler(
    Value=RigVMDispatchNode(
     Values=((),())
     Index=@RigVMFunction_MathBoolToInteger.Result
     Values.0=@RigVMFunction_MathQuaternionSwingTwist.Swing
     Values.1=@RigVMFunction_MathQuaternionSwingTwist.Twist
    ) {template="DISPATCH_RigVMDispatch_SelectInt32(in Index,in Values,out Result)"}.Result
    RotationOrder=@n1.Rotation Order
   )
   > DISPATCH_RigVMDispatch_Print = RigVMDispatchNode(ScreenColor=(R=1,G=1,B=1,A=1), ScreenDuration=1.5, Enabled=@n1.debugRotation, Value=@RigVMFunction_MathQuaternionToEuler.Result) {template="DISPATCH_RigVMDispatch_Print(in Prefix,in Value,in Enabled,in ScreenDuration,in ScreenColor)"}
   Clamp = &RigVMFunction_MathIntClamp(Value=@n1.primAxis)
   DISPATCH_RigVMDispatch_SelectInt32 = RigVMDispatchNode(
    Values=(0,0,0)
    Index=@Clamp.Result
    Values.0=@RigVMFunction_MathQuaternionToEuler.Result.X
    Values.1=@RigVMFunction_MathQuaternionToEuler.Result.Y
    Values.2=@RigVMFunction_MathQuaternionToEuler.Result.Z
   ) {template="DISPATCH_RigVMDispatch_SelectInt32(in Index,in Values,out Result)"}
   Remap = &RigVMFunction_MathDoubleRemap(
    TargetMaximum=1
    TargetMinimum=0
    bClamp=False
    Value=@DISPATCH_RigVMDispatch_SelectInt32.Result
    SourceMaximum=@n1.max
    SourceMinimum=@n1.min
   )
   > RigVMFunctionReturnNode(outputWeight=&RigVMFunction_MathDoubleClamp(Value=@Remap.Result).Result)
  ]}
 New Function_ContainedGraph {nodes [
   RigVMFunctionEntryNode()
   > n1 = &RigUnit_SphericalPoseReader(
    ActiveRegionScaleFactors=(PositiveWidth=0.5,NegativeWidth=0.5,PositiveHeight=0.5,NegativeHeight=0.5)
    ActiveRegionSize=0.18
    Debug=(bDrawDebug=false,bDraw2D=False,bDrawLocalAxes=False,DebugScale=25,DebugSegments=20,DebugThickness=0.25)
    DriverAxis=(X=1,Y=0,Z=0)
    DriverItem=(Type=Bone,Name="clavicle_l")
    FalloffRegionScaleFactors=(PositiveWidth=0.5,NegativeWidth=0.5,PositiveHeight=0.5,NegativeHeight=0.5)
    FalloffSize=0.4
    FlipHeightScaling=False
    FlipWidthScaling=False
    OptionalParentItem=(Type=Bone,Name="None")
    RotationOffset=(X=0,Y=90,Z=130)
   )
   > alterTransform_4 = RigVMFunctionReferenceNode(
    type=Bone
    transforms="((Bone=\"dyn_L_hair_3\",Transform=(Rotation=(X=0,Y=0,Z=0,W=1),Translation=(X=0,Y=0,Z=-5),Scale3D=(X=1,Y=1,Z=1))),(Bone=\"dyn_L_hair_4\",Transform=(Rotation=(X=0,Y=0,Z=0,W=1),Translation=(X=0,Y=4,Z=-1),Scale3D=(X=1,Y=1,Z=1))),(Bone=\"dyn_L_hair_2\",Transform=(Rot ...(truncated)"
    weight=&RigVMFunction_MathDoubleMul(A=@n1.OutputParam, B=RigVMVariableNode() {variable=left}.Value).Result
   )
   > n2 = &RigUnit_SphericalPoseReader(
    ActiveRegionScaleFactors=(PositiveWidth=0.5,NegativeWidth=0.5,PositiveHeight=0.5,NegativeHeight=0.5)
    ActiveRegionSize=0.18
    Debug=(bDrawDebug=false,bDraw2D=False,bDrawLocalAxes=False,DebugScale=25,DebugSegments=20,DebugThickness=0.25)
    DriverAxis=(X=-1,Y=0,Z=0)
    DriverItem=(Type=Bone,Name="clavicle_r")
    FalloffRegionScaleFactors=(PositiveWidth=0.5,NegativeWidth=0.5,PositiveHeight=0.5,NegativeHeight=0.5)
    FalloffSize=0.4
    FlipHeightScaling=False
    FlipWidthScaling=False
    OptionalParentItem=(Type=Bone,Name="None")
    RotationOffset=(X=0,Y=-90,Z=130)
   )
   > alterTransform_5 = RigVMFunctionReferenceNode(
    type=Bone
    transforms="((Bone=\"dyn_R_hair_3\",Transform=(Rotation=(X=0,Y=0,Z=0,W=1),Translation=(X=0,Y=0,Z=-5),Scale3D=(X=1,Y=1,Z=1))),(Bone=\"dyn_R_hair_4\",Transform=(Rotation=(X=0,Y=0,Z=0,W=1),Translation=(X=0,Y=-4,Z=-1),Scale3D=(X=1,Y=1,Z=1))),(Bone=\"dyn_R_hair_2\",Transform=(Ro ...(truncated)"
    weight=&RigVMFunction_MathDoubleMul(A=@n2.OutputParam, B=RigVMVariableNode() {variable=right}.Value).Result
   )
   > n3 = &RigUnit_SphericalPoseReader(
    ActiveRegionScaleFactors=(PositiveWidth=0.5,NegativeWidth=0.5,PositiveHeight=0.5,NegativeHeight=0.1)
    ActiveRegionSize=0.18
    Debug=(bDrawDebug=false,bDraw2D=False,bDrawLocalAxes=False,DebugScale=25,DebugSegments=20,DebugThickness=0.25)
    DriverAxis=(X=1,Y=0,Z=0)
    DriverItem=(Type=Bone,Name="clavicle_l")
    FalloffRegionScaleFactors=(PositiveWidth=0.5,NegativeWidth=0.5,PositiveHeight=0.5,NegativeHeight=0.1)
    FalloffSize=0.4
    FlipHeightScaling=False
    FlipWidthScaling=False
    OptionalParentItem=(Type=Bone,Name="None")
    RotationOffset=(X=0,Y=90,Z=170)
   )
   > RigVMVariableNode(Value=@n3.OutputParam) {variable=left}
   > alterTransform_7 = RigVMFunctionReferenceNode(
    type=Bone
    transforms="((Bone=\"dyn_L_hair_3\",Transform=(Rotation=(X=0,Y=0,Z=0,W=1),Translation=(X=0,Y=0,Z=-3),Scale3D=(X=1,Y=1,Z=1))),(Bone=\"dyn_L_hair_4\",Transform=(Rotation=(X=0,Y=0,Z=0,W=1),Translation=(X=0,Y=5,Z=0),Scale3D=(X=1,Y=1,Z=1))),(Bone=\"dyn_L_hair_offshoot\",Transfor ...(truncated)"
    weight=@n3.OutputParam
   )
   > n4 = &RigUnit_SphericalPoseReader(
    ActiveRegionScaleFactors=(PositiveWidth=0.5,NegativeWidth=0.5,PositiveHeight=0.1,NegativeHeight=0.5)
    ActiveRegionSize=0.18
    Debug=(bDrawDebug=true,bDraw2D=False,bDrawLocalAxes=False,DebugScale=25,DebugSegments=20,DebugThickness=0.25)
    DriverAxis=(X=-1,Y=0,Z=0)
    DriverItem=(Type=Bone,Name="clavicle_r")
    FalloffRegionScaleFactors=(PositiveWidth=0.5,NegativeWidth=0.5,PositiveHeight=0.1,NegativeHeight=0.5)
    FalloffSize=0.4
    FlipHeightScaling=False
    FlipWidthScaling=False
    OptionalParentItem=(Type=Bone,Name="None")
    RotationOffset=(X=0,Y=-90,Z=178)
   )
   > RigVMVariableNode(Value=@n4.OutputParam) {variable=right}
   n5 = &RigVMFunction_AlphaInterp(
    InterpSpeedDecreasing=5
    InterpSpeedIncreasing=5
    Scale=1
    bClampResult=true
    bMapRange=true
    Value=&RigVMFunction_MathDoubleMul(A=@n4.OutputParam, B=RigVMVariableNode() {variable=Push Hair}.Value).Result
   )
   > alterTransform_2 = RigVMFunctionReferenceNode(
    type=Bone
    transforms=((Bone="dyn_R_hair_3",Transform=(Rotation=(X=0,Y=0,Z=0,W=1),Translation=(X=0,Y=0,Z=-3),Scale3D=(X=1,Y=1,Z=1))),(Bone="dyn_R_hair_4",Transform=(Rotation=(X=0,Y=0,Z=0,W=1),Translation=(X=0,Y=-5,Z=0),Scale3D=(X=1,Y=1,Z=1))))
    weight=@n5.Result
   )
   > RigVMFunctionReturnNode()
   alterTransform_6 = RigVMFunctionReferenceNode(
    type=Bone
    ExecuteContext=()
    transforms="((Bone=\"dyn_R_hair_3\",Transform=(Rotation=(X=0,Y=0,Z=0,W=1),Translation=(X=0,Y=0,Z=-3),Scale3D=(X=1,Y=1,Z=1))),(Bone=\"dyn_R_hair_4\",Transform=(Rotation=(X=0,Y=0,Z=0,W=1),Translation=(X=0,Y=0,Z=-3),Scale3D=(X=1,Y=1,Z=1))),(Bone=\"dyn_R_hair_offshoot\",Transfo ...(truncated)"
   )
   Multiply = &RigVMFunction_MathDoubleMul(A=1, B=RigVMVariableNode() {variable=Push Hair}.Value)
   > alterTransform_3 = RigVMFunctionReferenceNode(
    type=Bone
    transforms=((Bone="dyn_R_hair_4",Transform=(Rotation=(X=0,Y=0,Z=0,W=1),Translation=(X=0,Y=-5,Z=3),Scale3D=(X=1,Y=1,Z=1))),(Bone="dyn_R_hair_offshoot",Transform=(Rotation=(X=0.1736,Y=0,Z=0,W=0.9848),Translation=(X=0,Y=0,Z=0),Scale3D=(X=1,Y=1,Z=1))))
    weight=@Multiply.Result
   )
   alterTransform_1 = RigVMFunctionReferenceNode(
    type=Bone
    ExecuteContext=()
    transforms="((Bone=\"dyn_R_hair_3\",Transform=(Rotation=(X=0,Y=0,Z=0,W=1),Translation=(X=0,Y=0,Z=-3),Scale3D=(X=1,Y=1,Z=1))),(Bone=\"dyn_R_hair_4\",Transform=(Rotation=(X=0,Y=0,Z=0,W=1),Translation=(X=0,Y=-5,Z=0),Scale3D=(X=1,Y=1,Z=1))),(Bone=\"dyn_R_hair_offshoot\",Transfo ...(truncated)"
   )
  ]}
 RigVMFunctionLibrary {nodes [
   DeltaRotation = RigVMCollapseNode(ExecuteContext=(), Rotation Order=ZYX, twistAxis=(X=1,Y=0,Z=0))
   Hair Adjustments = RigVMCollapseNode(ExecuteContext=())
   alterTransform = RigVMCollapseNode(type=Bone, ExecuteContext=(), transforms=())
   claviclesBackFoward = RigVMCollapseNode(ExecuteContext=())
   controllers = RigVMCollapseNode(ExecuteContext=(), debug=False)
   splineIK = RigVMCollapseNode(ExecuteContext=())
  ]}
 RigVMFunction_Sequence_1_ContainedGraph {nodes [
   n1 = RigVMFunctionEntryNode()
   RigVMFunction_Sequence = &RigVMFunction_Sequence(ExecuteContext=@n1.ExecuteContext)
   n2 = &RigVMFunction_Sequence(ExecuteContext=@RigVMFunction_Sequence.B)
   RigVMFunctionReturnNode(A=@RigVMFunction_Sequence.A, B=@n2.A, C=@n2.B)
  ]}
 RigVMFunction_Sequence_1_ContainedGraph {nodes [
   n1 = RigVMFunctionEntryNode()
   RigVMFunction_Sequence = &RigVMFunction_Sequence(ExecuteContext=@n1.ExecuteContext)
   n2 = &RigVMFunction_Sequence(ExecuteContext=@RigVMFunction_Sequence.B)
   n3 = &RigVMFunction_Sequence(ExecuteContext=@n2.B)
   RigVMFunctionReturnNode(A=@RigVMFunction_Sequence.A, B=@n2.A, C=@n3.A, D=@n3.B)
  ]}
 RigVMFunction_Sequence_1_ContainedGraph {nodes [
   n1 = RigVMFunctionEntryNode()
   RigVMFunction_Sequence = &RigVMFunction_Sequence(ExecuteContext=@n1.ExecuteContext)
   n2 = &RigVMFunction_Sequence(ExecuteContext=@RigVMFunction_Sequence.B)
   n3 = &RigVMFunction_Sequence(ExecuteContext=@n2.B)
   n4 = &RigVMFunction_Sequence(ExecuteContext=@n3.B)
   n5 = &RigVMFunction_Sequence(ExecuteContext=@n4.B)
   n6 = &RigVMFunction_Sequence(ExecuteContext=@n5.B)
   RigVMFunctionReturnNode(A=@RigVMFunction_Sequence.A, B=@n2.A, C=@n3.A, D=@n4.A, E=@n5.A, F=@n6.A, G=@n6.B)
  ]}
 RigVMFunction_Sequence_3_1_ContainedGraph {nodes [
   n1 = RigVMFunctionEntryNode()
   n2 = &RigVMFunction_Sequence(ExecuteContext=@n1.ExecuteContext)
   n3 = &RigVMFunction_Sequence(ExecuteContext=@n2.B)
   RigVMFunctionReturnNode(A=@n2.A, B=@n3.A, C=@n3.B)
  ]}
 RigVMModel {nodes [
   RigUnit_BeginExecution = &RigUnit_BeginExecution()
   splineIK = RigVMFunctionReferenceNode(ExecuteContext=@RigUnit_BeginExecution.ExecutePin)
   > Hair Adjustments = RigVMFunctionReferenceNode()
   &RigUnit_ModifyTransforms(
    ExecutePin=()
    ItemToModify="((Item=(Type=Bone,Name=\"dyn_R_hair_offshoot\"),Transform=(Rotation=(X=0,Y=0,Z=0,W=1),Translation=(X=4000,Y=0,Z=0),Scale3D=(X=1,Y=1,Z=1))),(Item=(Type=Bone,Name=\"dyn_L_hair_offshoot\"),Transform=(Rotation=(X=0,Y=0,Z=0,W=1),Translation=(X=40000,Y=0,Z=0),Scale3 ...(truncated)"
    Weight=1
    WeightMinimum=0
   )
   controllers = RigVMFunctionReferenceNode(ExecuteContext=(), debug=RigVMVariableNode() {variable=pub disablecontrolRig}.Value)
  ]}
 RigVMModel Rig_4 {nodes [
   n1 = &RigUnit_PrepareForExecution()
   RigVMFunction_Sequence = RigVMAggregateNode(ExecuteContext=@n1.ExecutePin)
   n2 = &RigUnit_ItemArray(Items=((Type=Control,Name="dyn_L_hair_1_ctrl"),(Type=Control,Name="dyn_L_hair_2_ctrl"),(Type=Control,Name="dyn_L_hair_3_ctrl"),(Type=Control,Name="dyn_L_hair_4_ctrl"),(Type=Control,Name="dyn_L_hair_5_ctrl"),(Type=Control,Name="dyn_L_hair_6_ctrl")))
   RigVMVariableNode(
    Value=/Script/ControlRig.RigUnit_HierarchyGetPoseItemArray(ElementType=All, Initial=False, ItemsToGet=@n2.Items) {is=struct RigVMUnitNode}.Pose
    ExecuteContext=@RigVMFunction_Sequence.A
   ) {variable=l_dyn_hair_pose}
   RigUnit_ItemArray = &RigUnit_ItemArray(Items=((Type=Control,Name="dyn_R_hair_1_ctrl"),(Type=Control,Name="dyn_R_hair_2_ctrl"),(Type=Control,Name="dyn_R_hair_3_ctrl"),(Type=Control,Name="dyn_R_hair_4_ctrl"),(Type=Control,Name="dyn_R_hair_5_ctrl"),(Type=Control,Name="dyn_R_hair_6_ctrl")))
   RigVMVariableNode(
    Value=&RigUnit_HierarchyGetPoseItemArray(ElementType=All, Initial=False, ItemsToGet=@RigUnit_ItemArray.Items).Pose
    ExecuteContext=@RigVMFunction_Sequence.B
   ) {variable=r_dyn_hair_pose}
   n3 = &RigUnit_ItemArray(Items=((Type=Control,Name="dyn_C_hair_1_ctrl"),(Type=Control,Name="dyn_C_hair_2_ctrl"),(Type=Control,Name="dyn_C_hair_3_ctrl"),(Type=Control,Name="dyn_C_hair_4_ctrl"),(Type=Control,Name="dyn_C_hair_5_ctrl"),(Type=Control,Name="dyn_C_hair_6_ctrl")))
   RigVMVariableNode(
    Value=&RigUnit_HierarchyGetPoseItemArray(ElementType=All, Initial=False, ItemsToGet=@n3.Items).Pose
    ExecuteContext=@RigVMFunction_Sequence.C
   ) {variable=c_dyn_hair_pose}
  ]}
]
hierarchy [
 root {loc=<0 0 0>, rot=<0 0 0 1>}
 pelvis {parent=root, loc=<0 -0.5301 98.0571>, rot=<0.0127 0.707 -0.0127 -0.707>}
 spine_01 {parent=pelvis, loc=<6.1208 -0.9328 0>, rot=<0 0 0.034 -0.9994>}
 spine_02 {parent=spine_01, loc=<6.5996 -0.2095 0>, rot=<0 0 0.0238 -0.9997>}
 spine_03 {parent=spine_02, loc=<6.4914 -0.1237 0>, rot=<0 0 0.0149 -0.9999>}
 spine_04 {parent=spine_03, loc=<6.9598 0.0794 0>, rot=<0 0 -0.0132 -0.9999>}
 spine_05 {parent=spine_04, loc=<6.9987 0 0>, rot=<0 0 -0.0015 -1>}
 neck_01 {parent=spine_05, loc=<11.3138 0.1529 0>, rot=<0 0 -0.2003 -0.9797>}
 neck_02 {parent=neck_01, loc=<4.5627 0 0>, rot=<0 0 0.0504 -0.9987>}
 head {parent=neck_02, loc=<5.3914 0 0>, rot=<0 0 0.115 -0.9934>}
 base_C_hair_1 {parent=head, loc=<6.9966 -5.405 0>, rot=<0.0967 -0.7005 -0.7005 0.0967>}
 dyn_C_hair_1 {parent=base_C_hair_1, loc=<0 0 0>, rot=<0 0 0 1>}
 dyn_C_hair_2 {parent=dyn_C_hair_1, loc=<9.7484 0 0>, rot=<0 0.0014 0 -1>}
 dyn_C_hair_3 {parent=dyn_C_hair_2, loc=<8.6126 0 -0.096>, rot=<0 0.0468 0 -0.9989>}
 dyn_C_hair_4 {parent=dyn_C_hair_3, loc=<10.4133 0 -0.1891>, rot=<-0.0007 0.0328 0 -0.9995>}
 dyn_C_hair_5 {parent=dyn_C_hair_4, loc=<11.8219 0.0001 0.075>, rot=<0 -0.0213 0 -0.9998>}
 dyn_C_hair_6 {parent=dyn_C_hair_5, loc=<10.2966 0.0029 2.0225>, rot=<0 -0.0064 0 -1>}
 dyn_C_hair_end_r {parent=dyn_C_hair_6, loc=<3.371 2.818 0.5381>, rot=<0 0 0 1>}
 dyn_C_hair_end_l {parent=dyn_C_hair_6, loc=<4.0092 -3.7807 0.6504>, rot=<0 0 0 1>}
 dyn_C_hair_end_c {parent=dyn_C_hair_6, loc=<4.0092 -0.5662 0.6457>, rot=<0 0 0 1>}
 dyn_C_hair_spike {parent=dyn_C_hair_4, loc=<0.7282 4.6595 -3.3941>, rot=<0.0007 -0.0328 0 -0.9995>}
 dyn_C_hair_spike_end {parent=dyn_C_hair_spike, loc=<9.3838 -1.8933 -0.805>, rot=<0 0 0 1>}
 base_L_hair_1 {parent=head, loc=<6.9561 -2.955 -7.9806>, rot=<0.0674 -0.7041 -0.7033 0.0714>}
 dyn_L_hair_1 {parent=base_L_hair_1, loc=<0 0 0>, rot=<0 0 0 1>}
 dyn_L_hair_2 {parent=dyn_L_hair_1, loc=<9.4278 1.1908 -0.0107>, rot=<0.0263 -0.0205 0.023 -0.9992>}
 dyn_L_hair_3 {parent=dyn_L_hair_2, loc=<8.9347 0.0332 -0.5435>, rot=<-0.0219 -0.0946 0.0038 -0.9953>}
 dyn_L_hair_4 {parent=dyn_L_hair_3, loc=<10.9748 -0.9591 0.6932>, rot=<0.0113 0.1227 -0.0219 -0.9921>}
 dyn_L_hair_5 {parent=dyn_L_hair_4, loc=<11.4876 0.7005 0.6936>, rot=<-0.0017 -0.004 0.0116 -0.9999>}
 dyn_L_hair_end_l {parent=dyn_L_hair_5, loc=<10.5625 -1.2254 1.722>, rot=<-0.0074 0.0079 -0.0013 -0.9999>}
 dyn_L_hair_end_r {parent=dyn_L_hair_5, loc=<11.9416 1.6932 1.3287>, rot=<-0.0074 0.0079 -0.0013 -0.9999>}
 dyn_L_hair_offshoot {parent=dyn_L_hair_4, loc=<-0.1841 -3.7401 0.466>, rot=<-0.0017 -0.004 0.0116 -0.9999>}
 dyn_L_hair_offshoot_end {parent=dyn_L_hair_offshoot, loc=<8.719 -2.6556 3.5425>, rot=<0 0 0 1>}
 dyn_L_hair_spike {parent=dyn_L_hair_4, loc=<-0.7332 2.2617 -3.2466>, rot=<0 0 0 1>}
 base_R_hair_1 {parent=head, loc=<6.9598 -2.9541 7.3872>, rot=<0.0699 -0.7036 -0.7036 0.0699>}
 dyn_R_hair_1 {parent=base_R_hair_1, loc=<0 0 0>, rot=<0 0 0 1>}
 dyn_R_hair_2 {parent=dyn_R_hair_1, loc=<9.414 -1.1066 -0.0073>, rot=<0.0004 -0.0174 -0.0218 -0.9996>}
 dyn_R_hair_3 {parent=dyn_R_hair_2, loc=<8.9075 -0.0035 -0.4118>, rot=<-0.0018 -0.0976 -0.0002 -0.9952>}
 dyn_R_hair_4 {parent=dyn_R_hair_3, loc=<11.3185 0.9816 -0.0567>, rot=<-0.0106 0.1235 0.0123 -0.9922>}
 dyn_R_hair_5 {parent=dyn_R_hair_4, loc=<11.5614 -0.4576 0.6685>, rot=<0.0046 -0.0026 -0.0044 -1>}
 dyn_R_hair_end_r {parent=dyn_R_hair_5, loc=<6.1347 2.2969 3.3357>, rot=<0.0087 0.0076 -0.0001 -0.9999>}
 dyn_R_hair_end_l {parent=dyn_R_hair_5, loc=<10.2534 -2.1754 3.825>, rot=<0.0087 0.0076 -0.0001 -0.9999>}
 dyn_R_hair_offshoot {parent=dyn_R_hair_4, loc=<4.7654 5.6224 2.4528>, rot=<0.0046 -0.0026 -0.0044 -1>}
 dyn_R_hair_offshoot_end {parent=dyn_R_hair_offshoot, loc=<7.9614 1.1904 2.2133>, rot=<0.0087 0.0076 -0.0001 -0.9999>}
 dyn_R_hair_spike {parent=dyn_R_hair_3, loc=<5.6419 0.6703 -3.4001>, rot=<-0.0106 0.1235 0.0123 -0.9922>}
 base_LB_bangs_3 {parent=head, loc=<11.0247 2.0952 -4.9884>, rot=<0.0674 -0.7041 -0.7033 0.0714>}
 dyn_LB_bangs_1 {parent=base_LB_bangs_3, loc=<0 0 0>, rot=<0 0 0 1>}
 dyn_LB_bangs_2 {parent=dyn_LB_bangs_1, loc=<5.6162 -1.909 0.0935>, rot=<0.0263 -0.0205 0.023 -0.9992>}
 dyn_LB_bangs_3 {parent=dyn_LB_bangs_2, loc=<5.3764 0.5108 1.3488>, rot=<-0.0209 0.0797 0.0075 -0.9966>}
 base_LF_bangs_2 {parent=head, loc=<10.9892 6.3869 -4.9884>, rot=<0.0674 -0.7041 -0.7033 0.0714>}
 dyn_LF_bangs_1 {parent=base_LF_bangs_2, loc=<0 0 0>, rot=<0 0 0 1>}
 dyn_LF_bangs_2 {parent=dyn_LF_bangs_1, loc=<4.7999 -1.9043 0.0725>, rot=<0.0263 -0.0205 0.023 -0.9992>}
 dyn_LF_bangs_3 {parent=dyn_LF_bangs_2, loc=<4.7771 0.5281 0.5316>, rot=<-0.0209 0.0797 0.0075 -0.9966>}
 dyn_RF_bangs_8 {parent=base_LF_bangs_2, loc=<7.1022 11.3324 4.3954>, rot=<0.0263 -0.0205 0.023 -0.9992>}
 base_RB_bangs_4 {parent=head, loc=<11.0247 2.0952 6.0273>, rot=<0.0674 -0.7041 -0.7033 0.0714>}
 dyn_RB_bangs_1 {parent=base_RB_bangs_4, loc=<0 0 0>, rot=<0 0 0 1>}
 dyn_RB_bangs_2 {parent=dyn_RB_bangs_1, loc=<5.5377 1.9933 0.6194>, rot=<0.0263 -0.0205 0.023 -0.9992>}
 dyn_RB_bangs_3 {parent=dyn_RB_bangs_2, loc=<5.3764 0.5108 1.3488>, rot=<-0.0209 0.0797 0.0075 -0.9966>}
 base_RF_bangs_5 {parent=head, loc=<9.949 7.4671 6.0273>, rot=<0.0674 -0.7041 -0.7033 0.0714>}
 dyn_RF_bangs_1 {parent=base_RF_bangs_5, loc=<0 0 0>, rot=<0 0 0 1>}
 dyn_RF_bangs_2 {parent=dyn_RF_bangs_1, loc=<4.3874 1.8553 0.4004>, rot=<0.0263 -0.0205 0.023 -0.9992>}
 dyn_RF_bangs_3 {parent=dyn_RF_bangs_2, loc=<5.7364 -0.5574 -0.0549>, rot=<-0.0209 0.0797 0.0075 -0.9966>}
 head_null {parent=neck_02, type=Null, loc=<5.3914 0 0>, rot=<0 0 0.115 -0.9934>}
 clavicle_r_null {parent=spine_05, type=Null, loc=<5.4831 -0.1053 3.2946>, rot=<0.7289 0.0465 -0.6829 0.009>}
 clavicle_l_null {parent=spine_05, type=Null, loc=<5.4831 -0.1053 -3.2946>, rot=<0.0465 -0.7289 -0.009 -0.6829>}
 head_null_sp {parent=spine_05, type=Null, loc=<20.6573 3.5489 0>, rot=<0 0 -0.036 -0.9994>}
 neck_01_null {parent=spine_05, type=Null, loc=<11.3138 0.1529 0>, rot=<0 0 -0.2003 -0.9797>}
 clavicle_l {parent=spine_05, loc=<5.4831 -0.1053 -3.2946>, rot=<0.0465 -0.7289 -0.009 -0.6829>}
 upperarm_l {parent=clavicle_l, loc=<9.8658 -1.2136 -0.4105>, rot=<0.0164 -0.3245 -0.0269 -0.9454>}
 clavicle_r {parent=spine_05, loc=<5.4831 -0.1053 3.2946>, rot=<0.7289 0.0465 -0.6829 0.009>}
 upperarm_r {parent=clavicle_r, loc=<-9.8658 1.2136 0.4105>, rot=<0.0164 -0.3245 -0.0269 -0.9454>}
 spine_02_null {parent=pelvis, type=Null, loc=<12.6908 -1.5909 0>, rot=<0 0 -0.0578 0.9983>}
 dyn_R_hair_7_ctrl_null {type=Null, loc=<-9.4953 -14.7292 96.1939>, rot=<0.5577 -0.594 -0.4227 -0.3968>}
 dyn_R_hair_7_ctrl {controlType=EULER_TRANSFORM, parent=dyn_R_hair_7_ctrl_null, type=Control, loc=<0 0 0>, rot=<0 0 0 1>}
 dyn_R_hair_6_ctrl_null {type=Null, loc=<-9.2115 -12.1921 106.644>, rot=<0.5485 -0.5631 -0.4428 -0.4313>}
 dyn_R_hair_6_ctrl {controlType=EULER_TRANSFORM, parent=dyn_R_hair_6_ctrl_null, type=Control, loc=<0 0 0>, rot=<0 0 0 1>}
 dyn_R_hair_5_ctrl_null {type=Null, loc=<-8.9082 -12.8298 117.384>, rot=<0.4781 -0.4917 -0.5218 -0.5073>}
 dyn_R_hair_5_ctrl {controlType=EULER_TRANSFORM, parent=dyn_R_hair_5_ctrl_null, type=Control, loc=<0 0 0>, rot=<0 0 0 1>}
 dyn_R_hair_4_ctrl_null {type=Null, loc=<-8.843 -13.2203 128.151>, rot=<0.4894 -0.4923 -0.5105 -0.5074>}
 dyn_R_hair_4_ctrl {controlType=EULER_TRANSFORM, parent=dyn_R_hair_4_ctrl_null, type=Control, loc=<0 0 0>, rot=<0 0 0 1>}
 dyn_R_hair_3_ctrl_null {type=Null, loc=<-8.3639 -12.0398 138.745>, rot=<-0.515 0.5387 0.482 0.4608>}
 dyn_R_hair_3_ctrl {controlType=EULER_TRANSFORM, parent=dyn_R_hair_3_ctrl_null, type=Control, loc=<0 0 0>, rot=<0 0 0 1>}
 dyn_R_hair_2_ctrl_null {type=Null, loc=<-7.9192 -7.6159 148.554>, rot=<0.5816 -0.6061 -0.3915 -0.3757>}
 dyn_R_hair_2_ctrl {controlType=EULER_TRANSFORM, parent=dyn_R_hair_2_ctrl_null, type=Control, loc=<0 0 0>, rot=<0 0 0 1>}
 dyn_R_hair_1_ctrl_null {type=Null, loc=<-7.9806 -4.7569 158.941>, rot=<-0.564 0.5608 0.4273 0.4298>}
 dyn_R_hair_1_ctrl {controlType=EULER_TRANSFORM, parent=dyn_R_hair_1_ctrl_null, type=Control, loc=<0 0 0>, rot=<0 0 0 1>}
 dyn_L_hair_6_ctrl_null {type=Null, loc=<9.2115 -12.1921 106.6437>, rot=<0.5631 -0.5485 -0.4313 -0.4428>}
 dyn_L_hair_6_ctrl {controlType=EULER_TRANSFORM, parent=dyn_L_hair_6_ctrl_null, type=Control, loc=<0 0 0>, rot=<0 0 0 1>}
 dyn_L_hair_5_ctrl_null {type=Null, loc=<8.9082 -12.8298 117.3842>, rot=<0.4917 -0.4781 -0.5073 -0.5218>}
 dyn_L_hair_5_ctrl {controlType=EULER_TRANSFORM, parent=dyn_L_hair_5_ctrl_null, type=Control, loc=<0 0 0>, rot=<0 0 0 1>}
 dyn_L_hair_4_ctrl_null {type=Null, loc=<8.843 -13.2203 128.1512>, rot=<0.4923 -0.4894 -0.5074 -0.5105>}
 dyn_L_hair_4_ctrl {controlType=EULER_TRANSFORM, parent=dyn_L_hair_4_ctrl_null, type=Control, loc=<0 0 0>, rot=<0 0 0 1>}
 dyn_L_hair_3_ctrl_null {type=Null, loc=<8.364 -12.0398 138.7448>, rot=<-0.5387 0.515 0.4608 0.482>}
 dyn_L_hair_3_ctrl {controlType=EULER_TRANSFORM, parent=dyn_L_hair_3_ctrl_null, type=Control, loc=<0 0 0>, rot=<0 0 0 1>}
 dyn_L_hair_2_ctrl_null {type=Null, loc=<7.9192 -7.6159 148.5537>, rot=<0.6061 -0.5816 -0.3757 -0.3915>}
 dyn_L_hair_2_ctrl {controlType=EULER_TRANSFORM, parent=dyn_L_hair_2_ctrl_null, type=Control, loc=<0 0 0>, rot=<0 0 0 1>}
 dyn_L_hair_1_ctrl_null {type=Null, loc=<7.9806 -4.7569 158.9405>, rot=<0.5608 -0.564 -0.4298 -0.4273>}
 dyn_L_hair_1_ctrl {controlType=EULER_TRANSFORM, parent=dyn_L_hair_1_ctrl_null, type=Control, loc=<0 0 0>, rot=<0 0 0 1>}
 dyn_C_hair_7_ctrl_null {type=Null, loc=<0 -15.0214 96.4145>, rot=<0.5712 -0.5712 -0.4168 -0.4168>}
 dyn_C_hair_7_ctrl {controlType=EULER_TRANSFORM, parent=dyn_C_hair_7_ctrl_null, type=Control, loc=<0 0 0>, rot=<0 0 0 1>}
 dyn_C_hair_6_ctrl_null {type=Null, loc=<0 -13.5663 106.8971>, rot=<0.5333 -0.5333 -0.4644 -0.4644>}
 dyn_C_hair_6_ctrl {controlType=EULER_TRANSFORM, parent=dyn_C_hair_6_ctrl_null, type=Control, loc=<0 0 0>, rot=<0 0 0 1>}
 dyn_C_hair_5_ctrl_null {type=Null, loc=<0 -14.0207 117.4844>, rot=<0.4892 -0.4892 -0.5106 -0.5106>}
 dyn_C_hair_5_ctrl {controlType=EULER_TRANSFORM, parent=dyn_C_hair_5_ctrl_null, type=Control, loc=<0 0 0>, rot=<0 0 0 1>}
 dyn_C_hair_4_ctrl_null {type=Null, loc=<0 -14.4726 128.0719>, rot=<0.4892 -0.4892 -0.5105 -0.5105>}
 dyn_C_hair_4_ctrl {controlType=EULER_TRANSFORM, parent=dyn_C_hair_4_ctrl_null, type=Control, loc=<0 0 0>, rot=<0 0 0 1>}
 dyn_C_hair_3_ctrl_null {type=Null, loc=<0 -13.0257 138.5554>, rot=<-0.5331 0.5331 0.4646 0.4646>}
 dyn_C_hair_3_ctrl {controlType=EULER_TRANSFORM, parent=dyn_C_hair_3_ctrl_null, type=Control, loc=<0 0 0>, rot=<0 0 0 1>}
 dyn_C_hair_2_ctrl_null {type=Null, loc=<0 -9.9416 148.6977>, rot=<0.5681 -0.5681 -0.421 -0.421>}
 dyn_C_hair_2_ctrl {controlType=EULER_TRANSFORM, parent=dyn_C_hair_2_ctrl_null, type=Control, loc=<0 0 0>, rot=<0 0 0 1>}
 dyn_C_hair_1_ctrl_null {type=Null, loc=<0 -7.2072 158.9405>, rot=<-0.5608 0.5608 0.4307 0.4307>}
 dyn_C_hair_1_ctrl {controlType=EULER_TRANSFORM, parent=dyn_C_hair_1_ctrl_null, type=Control, loc=<0 0 0>, rot=<0 0 0 1>}
 spine_01_ctrl {controlType=EULER_TRANSFORM, type=Control, loc=<0 0 0>, rot=<0 0 0 1>}
 spine_02_ctrl {controlType=EULER_TRANSFORM, parent=spine_01_ctrl, type=Control, loc=<0 0 0>, rot=<0 0 0 1>}
 spine_03_ctrl {controlType=EULER_TRANSFORM, parent=spine_02_ctrl, type=Control, loc=<0 0 0>, rot=<0 0 0 1>}
 spine_04_ctrl {controlType=EULER_TRANSFORM, parent=spine_03_ctrl, type=Control, loc=<0 0 0>, rot=<0 0 0 1>}
 spine_05_ctrl {controlType=EULER_TRANSFORM, parent=spine_04_ctrl, type=Control, loc=<0 0 0>, rot=<0 0 0 1>}
 clavicle_r_ctrl {controlType=EULER_TRANSFORM, parent=spine_05_ctrl, type=Control, loc=<0 0 0>, rot=<0 0 0 1>}
 upperarm_r_ctrl {controlType=EULER_TRANSFORM, parent=clavicle_r_ctrl, type=Control, loc=<0 0 0>, rot=<0 0 0 1>}
 clavicle_l_ctrl {controlType=EULER_TRANSFORM, parent=spine_05_ctrl, type=Control, loc=<0 0 0>, rot=<0 0 0 1>}
 upperarm_l_ctrl {controlType=EULER_TRANSFORM, parent=clavicle_l_ctrl, type=Control, loc=<0 0 0>, rot=<0 0 0 1>}
 neck_01_ctrl {controlType=EULER_TRANSFORM, parent=spine_05_ctrl, type=Control, loc=<0 0 0>, rot=<0 0 0 1>}
 neck_02_ctrl {controlType=EULER_TRANSFORM, parent=neck_01_ctrl, type=Control, loc=<0 0 0>, rot=<0 0 0 1>}
 head_ctrl {controlType=EULER_TRANSFORM, parent=neck_02_ctrl, type=Control, loc=<0 0 0>, rot=<0 0 0 1>}
 pelvis_ctrl {controlType=EULER_TRANSFORM, type=Control, loc=<0 0 0>, rot=<0 0 0 1>}
 thigh_r_ctrl {controlType=EULER_TRANSFORM, parent=pelvis_ctrl, type=Control, loc=<0 0 0>, rot=<0 0 0 1>}
 thigh_l_ctrl {controlType=EULER_TRANSFORM, parent=pelvis_ctrl, type=Control, loc=<0 0 0>, rot=<0 0 0 1>}
]
variables [
 Cape {type=bool, default=False}
 Dive {type=double, default=0}
 FE {type=bool, default=False}
 Push Hair {type=double, default=0}
 Tsprint {type=double, default=0}
 c_dyn_hair_controls {type="TArray<FRigElementKey>"}
 c_dyn_hair_pose {type=FRigPose, default=(Elements=,HierarchyTopologyVersion=-1,PoseHash=-1)}
 l_dyn_hair_controls {type="TArray<FRigElementKey>"}
 l_dyn_hair_pose {type=FRigPose, default=(Elements=,HierarchyTopologyVersion=-1,PoseHash=-1)}
 pub aimPitch {type=double, default=0}
 pub disableUpperLimbsCorrection {type=bool, default=False}
 pub disablecontrolRig {type=bool, default=True}
 pub diveFloating {type=bool, default=False}
 pub isAssaultRifleEquipped {type=double, default=0}
 pub isTargeting {type=bool, default=False}
 r_dyn_hair_controls {type="TArray<FRigElementKey>"}
 r_dyn_hair_pose {type=FRigPose, default=(Elements=,HierarchyTopologyVersion=-1,PoseHash=-1)}
]
