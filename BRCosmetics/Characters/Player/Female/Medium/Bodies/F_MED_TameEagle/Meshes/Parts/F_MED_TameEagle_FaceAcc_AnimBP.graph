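# Animation Blueprint for the TameEagle female face-accessory mesh, adding secondary motion to hair, bangs and spikes.
# It copies the parent mesh pose through a Control Rig fed by gameplay states like aiming, sprinting and diving.
# That cached pose feeds Anim Dynamics chains, with constraints tying side strands to center ones in the high-inertia branch.
# Nested bool blends select between rig-only, backpack, high-inertia and other hair-dynamics poses.
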AnimBlueprint /BRCosmetics/Characters/Player/Female/Medium/Bodies/F_MED_TameEagle/Meshes/Parts/F_MED_TameEagle_FaceAcc_AnimBP
target_skeleton=&/F_MED_TameEagle_FaceAcc_Skeleton
graphs [
 AnimGraph {nodes [
   n1 = Blend Poses by bool(
    BlendPose_0=Use cached pose 'High Inert'()
    BlendPose_1=Blend Poses by bool(BlendPose_0=Use cached pose 'HairDyn'(), BlendPose_1=Use cached pose 'Cntrl Rig'()) {blendType=Linear}
    BlendTime_0=1
    BlendTime_1=1
   ) {blendType=CircularIn}
   n2 = Use cached pose 'Backpack'()
   Output Pose(Result=Blend Poses by bool(BlendPose_0=Use cached pose 'Cntrl Rig'(), BlendPose_1=Blend Poses by bool(BlendPose_0=@n2, BlendPose_1=@n1) {blendType=Linear}) {blendType=Linear}) {layerGroup=None}
   n3 = Control Rig!(
    Source=Copy Pose From Mesh(SourceMeshComponent=Get Skeletal_Mesh() -> (Skeletal_Mesh:Skeletal Mesh Component Object Reference)) {bUseAttachedParent=true}
    Alpha
    pub disablecontrolRig:bool=Get disableControlRig() -> (disableControlRig:bool)
    pub disableUpperLimbsCorrection:bool=Get disableUpperLimbsCorrection() -> (disableUpperLimbsCorrection:bool)
    pub diveFloating:bool=Get bIsSkydiveFloating()
    pub isTargeting:bool=Get bIsTargeting()
    pub aimPitch:double=Get AimPitch()
    pub isAssaultRifleEquipped:double=Get isAssaultRifleEquipped float() -> (isAssaultRifleEquipped float:double)
    FE:bool=Get FE() -> (FE:bool)
    Cape:bool=Get isWearingCape() -> (isWearingCape:bool)
    Push Hair:double=Get push hair() -> (push hair:double)
    Tsprint:double=Get TSprint() -> (TSprint:double)
    Dive:double=Get Diving() -> (Diving:double)
   ) -> (Pose) {
    alphaBoolBlend {blendOption=Linear}
    controlRigAssetReference {blueprintRigClass=&/F_MED_TameEagle_FaceAcc_CtrlRig.F_MED_TameEagle_FaceAcc_CtrlRig_C}
   }
   Save cached pose 'Cntrl Rig'(Pose=@n3) {cacheName=Cntrl Rig}
   n4 = Anim Dynamics - Bone: dyn_L_hair_4!(
    ComponentPose=Local To Component(LocalPose=Use cached pose 'Cntrl Rig'())
    bAlphaBoolEnabled:bool=True
    GravityOverride:Vector="0,0,-980"
    AngularSpringConstant:float=20000
    ExternalForce="0,0,0"
   ) -> (Pose) {
    alphaInputType=Bool
    angularDampingOverride=0.999
    angularSpringConstant=20000
    bAngularSpring=true
    bOverrideAngularDamping=true
    bUseGravityOverride=true
    gravityScale=0.35
    simulationSpace=BoneRelative
    alphaBoolBlend {blendInTime=2, blendOption=Cubic}
    boundBone {boneName=dyn_L_hair_4}
    componentAppliedLinearAccClamp=<3000 3000 3000>
    componentLinearAccScale=<0.3 0.4 0.3>
    componentLinearVelScale=<0.3 0.4 0.3>
    gravityOverride {z=-980}
    physicsBodyDefinitions [{
      boundBone {boneName=dyn_L_hair_4}
      boxExtents=<30 15 15>
      constraintSetup {angularLimitsMax=<0 90 0>, angularLimitsMin=<0 0 0>, angularTarget=<1 0 0>}
      localJointOffset=<4 0 0>
     }]
    relativeSpaceBone {boneName=head}
   }
   n5 = Anim Dynamics - Bone: dyn_C_hair_4!(ComponentPose=@n4, bAlphaBoolEnabled:bool=True, GravityOverride:Vector="0,0,-980", AngularSpringConstant:float=20000, ExternalForce="0,0,0") -> (Pose) {
    alphaInputType=Bool
    angularDampingOverride=0.999
    angularSpringConstant=20000
    bAngularSpring=true
    bOverrideAngularDamping=true
    bUseGravityOverride=true
    gravityScale=0.35
    simulationSpace=BoneRelative
    alphaBoolBlend {blendInTime=2, blendOption=Cubic}
    boundBone {boneName=dyn_C_hair_4}
    componentAppliedLinearAccClamp=<3000 3000 3000>
    componentLinearAccScale=<0.3 0.4 0.3>
    componentLinearVelScale=<0.3 0.4 0.3>
    gravityOverride {z=-980}
    physicsBodyDefinitions [{
      boundBone {boneName=dyn_C_hair_4}
      boxExtents=<30 15 15>
      constraintSetup {angularLimitsMax=<0 90 0>, angularLimitsMin=<0 0 0>, angularTarget=<1 0 0>}
      localJointOffset=<4 0 0>
     }]
    relativeSpaceBone {boneName=head}
   }
   n6 = Anim Dynamics - Bone: dyn_R_hair_4!(ComponentPose=@n5, bAlphaBoolEnabled:bool=True, GravityOverride:Vector="0,0,-980", AngularSpringConstant:float=20000, ExternalForce="0,0,0") -> (Pose) {
    alphaInputType=Bool
    angularDampingOverride=0.999
    angularSpringConstant=20000
    bAngularSpring=true
    bOverrideAngularDamping=true
    bUseGravityOverride=true
    gravityScale=0.35
    simulationSpace=BoneRelative
    alphaBoolBlend {blendInTime=2, blendOption=Cubic}
    boundBone {boneName=dyn_R_hair_4}
    componentAppliedLinearAccClamp=<3000 3000 3000>
    componentLinearAccScale=<0.3 0.4 0.3>
    componentLinearVelScale=<0.3 0.4 0.3>
    gravityOverride {z=-980}
    physicsBodyDefinitions [{
      boundBone {boneName=dyn_R_hair_4}
      boxExtents=<30 15 15>
      constraintSetup {angularLimitsMax=<0 90 0>, angularLimitsMin=<0 0 0>, angularTarget=<1 0 0>}
      localJointOffset=<4 0 0>
     }]
    relativeSpaceBone {boneName=head}
   }
   n7 = Anim Dynamics - Bone: dyn_L_hair_5!(ComponentPose=@n6, bAlphaBoolEnabled:bool=True, GravityOverride:Vector="0,0,-980", AngularSpringConstant:float=20000, ExternalForce="0,0,0") -> (Pose) {
    alphaInputType=Bool
    angularDampingOverride=0.999
    angularSpringConstant=20000
    bAngularSpring=true
    bOverrideAngularDamping=true
    bUseGravityOverride=true
    gravityScale=0.35
    simulationSpace=BoneRelative
    alphaBoolBlend {blendInTime=2, blendOption=Cubic}
    boundBone {boneName=dyn_L_hair_5}
    componentAppliedLinearAccClamp=<3000 3000 3000>
    componentLinearAccScale=<0.3 0.4 0.3>
    componentLinearVelScale=<0.3 0.4 0.3>
    gravityOverride {z=-980}
    physicsBodyDefinitions [{
      boundBone {boneName=dyn_L_hair_5}
      boxExtents=<30 15 15>
      constraintSetup {angularLimitsMax=<0 90 0>, angularLimitsMin=<0 -40 0>, angularTarget=<1 0 0>}
      localJointOffset=<4 0 0>
     }]
    relativeSpaceBone {boneName=head}
   }
   n8 = Anim Dynamics - Bone: dyn_C_hair_5!(ComponentPose=@n7, bAlphaBoolEnabled:bool=True, GravityOverride:Vector="0,0,-980", AngularSpringConstant:float=20000, ExternalForce="0,0,0") -> (Pose) {
    alphaInputType=Bool
    angularDampingOverride=0.999
    angularSpringConstant=20000
    bAngularSpring=true
    bOverrideAngularDamping=true
    bUseGravityOverride=true
    gravityScale=0.35
    simulationSpace=BoneRelative
    alphaBoolBlend {blendInTime=2, blendOption=Cubic}
    boundBone {boneName=dyn_C_hair_5}
    componentAppliedLinearAccClamp=<3000 3000 3000>
    componentLinearAccScale=<0.3 0.4 0.3>
    componentLinearVelScale=<0.3 0.4 0.3>
    gravityOverride {z=-980}
    physicsBodyDefinitions [{
      boundBone {boneName=dyn_C_hair_5}
      boxExtents=<30 15 15>
      constraintSetup {angularLimitsMax=<0 90 0>, angularLimitsMin=<0 0 0>, angularTarget=<1 0 0>}
      localJointOffset=<4 0 0>
     }]
    relativeSpaceBone {boneName=head}
    simSpaceSettings {maxAngularAcceleration=8000, maxAngularVelocity=8000}
   }
   n9 = Anim Dynamics - Bone: dyn_R_hair_5!(ComponentPose=@n8, bAlphaBoolEnabled:bool=True, GravityOverride:Vector="0,0,-980", AngularSpringConstant:float=20000, ExternalForce="0,0,0") -> (Pose) {
    alphaInputType=Bool
    angularDampingOverride=0.999
    angularSpringConstant=20000
    bAngularSpring=true
    bOverrideAngularDamping=true
    bUseGravityOverride=true
    gravityScale=0.35
    simulationSpace=BoneRelative
    alphaBoolBlend {blendInTime=2, blendOption=Cubic}
    boundBone {boneName=dyn_R_hair_5}
    componentAppliedLinearAccClamp=<3000 3000 3000>
    componentLinearAccScale=<0.3 0.4 0.3>
    componentLinearVelScale=<0.3 0.4 0.3>
    gravityOverride {z=-980}
    physicsBodyDefinitions [{
      boundBone {boneName=dyn_R_hair_5}
      boxExtents=<30 15 15>
      constraintSetup {angularLimitsMax=<0 90 0>, angularLimitsMin=<0 -40 0>, angularTarget=<1 0 0>}
      localJointOffset=<4 0 0>
     }]
    relativeSpaceBone {boneName=head}
   }
   n10 = Anim Dynamics - Bone: dyn_L_hair_end_l!(ComponentPose=@n9, bAlphaBoolEnabled:bool=True, GravityOverride:Vector="0,0,-980", ExternalForce="0,0,0") -> (Pose) {
    alphaInputType=Bool
    angularDampingOverride=0.999
    angularSpringConstant=5000
    bAngularSpring=true
    bOverrideAngularDamping=true
    bUseGravityOverride=true
    gravityScale=0.35
    simulationSpace=BoneRelative
    alphaBoolBlend {blendInTime=2, blendOption=Cubic}
    boundBone {boneName=dyn_L_hair_end_l}
    componentAppliedLinearAccClamp=<3000 3000 3000>
    componentLinearAccScale=<0.3 0.4 0.3>
    componentLinearVelScale=<0.3 0.4 0.3>
    gravityOverride {z=-980}
    physicsBodyDefinitions [{
      boundBone {boneName=dyn_L_hair_end_l}
      boxExtents=<25 25 25>
      constraintSetup {angularLimitsMax=<0 40 0>, angularLimitsMin=<0 0 0>, angularTarget=<1 0 0>}
      localJointOffset=<4 0 0>
     }]
    relativeSpaceBone {boneName=head}
   }
   n11 = Anim Dynamics - Bone: dyn_L_hair_end_r!(ComponentPose=@n10, bAlphaBoolEnabled:bool=True, GravityOverride:Vector="0,0,-980", ExternalForce="0,0,0") -> (Pose) {
    alphaInputType=Bool
    angularDampingOverride=0.999
    angularSpringConstant=12000
    bAngularSpring=true
    bOverrideAngularDamping=true
    bUseGravityOverride=true
    gravityScale=0.35
    simulationSpace=BoneRelative
    alphaBoolBlend {blendInTime=2, blendOption=Cubic}
    boundBone {boneName=dyn_L_hair_end_r}
    componentAppliedLinearAccClamp=<3000 3000 3000>
    componentLinearAccScale=<0.3 0.4 0.3>
    componentLinearVelScale=<0.3 0.4 0.3>
    gravityOverride {z=-980}
    physicsBodyDefinitions [{
      boundBone {boneName=dyn_L_hair_end_r}
      boxExtents=<20 20 20>
      constraintSetup {angularLimitsMax=<0 40 0>, angularLimitsMin=<0 0 0>, angularTarget=<1 0 0>}
      localJointOffset=<4 0 0>
     }]
    relativeSpaceBone {boneName=head}
   }
   n12 = Anim Dynamics - Bone: dyn_C_hair_end_r!(ComponentPose=@n11, bAlphaBoolEnabled:bool=True, GravityOverride:Vector="0,0,-980", ExternalForce="0,0,0") -> (Pose) {
    alphaInputType=Bool
    angularDampingOverride=0.999
    angularSpringConstant=3000
    bAngularSpring=true
    bOverrideAngularDamping=true
    bUseGravityOverride=true
    gravityScale=0.35
    simulationSpace=BoneRelative
    alphaBoolBlend {blendInTime=2, blendOption=Cubic}
    boundBone {boneName=dyn_C_hair_end_r}
    componentAppliedLinearAccClamp=<3000 3000 3000>
    componentLinearAccScale=<0.3 0.4 0.3>
    componentLinearVelScale=<0.3 0.4 0.3>
    gravityOverride {z=-980}
    physicsBodyDefinitions [{
      boundBone {boneName=dyn_C_hair_end_r}
      boxExtents=<25 25 25>
      constraintSetup {angularLimitsMax=<0 40 0>, angularLimitsMin=<0 -40 0>, angularTarget=<1 0 0>}
      localJointOffset=<4 0 0>
     }]
    relativeSpaceBone {boneName=head}
   }
   n13 = Anim Dynamics - Bone: dyn_C_hair_end_l!(ComponentPose=@n12, bAlphaBoolEnabled:bool=True, GravityOverride:Vector="0,0,-980", ExternalForce="0,0,0") -> (Pose) {
    alphaInputType=Bool
    angularDampingOverride=0.999
    angularSpringConstant=15000
    bAngularSpring=true
    bOverrideAngularDamping=true
    bUseGravityOverride=true
    gravityScale=0.35
    simulationSpace=BoneRelative
    alphaBoolBlend {blendInTime=2, blendOption=Cubic}
    boundBone {boneName=dyn_C_hair_end_l}
    componentAppliedLinearAccClamp=<3000 3000 3000>
    componentLinearAccScale=<0.3 0.4 0.3>
    componentLinearVelScale=<0.3 0.4 0.3>
    gravityOverride {z=-980}
    physicsBodyDefinitions [{
      boundBone {boneName=dyn_C_hair_end_l}
      boxExtents=<25 25 25>
      constraintSetup {angularLimitsMax=<0 40 0>, angularLimitsMin=<0 0 0>, angularTarget=<1 0 0>}
      localJointOffset=<4 0 0>
     }]
    relativeSpaceBone {boneName=head}
   }
   n14 = Anim Dynamics - Bone: dyn_C_hair_end_c!(ComponentPose=@n13, bAlphaBoolEnabled:bool=True, GravityOverride:Vector="0,0,-980", ExternalForce="0,0,0") -> (Pose) {
    alphaInputType=Bool
    angularDampingOverride=0.999
    angularSpringConstant=1000
    bAngularSpring=true
    bOverrideAngularDamping=true
    bUseGravityOverride=true
    gravityScale=0.35
    simulationSpace=BoneRelative
    alphaBoolBlend {blendInTime=2, blendOption=Cubic}
    boundBone {boneName=dyn_C_hair_end_c}
    componentAppliedLinearAccClamp=<3000 3000 3000>
    componentLinearAccScale=<0.3 0.4 0.3>
    componentLinearVelScale=<0.3 0.4 0.3>
    gravityOverride {z=-980}
    physicsBodyDefinitions [{
      boundBone {boneName=dyn_C_hair_end_c}
      boxExtents=<25 25 25>
      constraintSetup {angularLimitsMax=<0 40 0>, angularLimitsMin=<0 0 0>, angularTarget=<1 0 0>}
      localJointOffset=<4 0 0>
     }]
    relativeSpaceBone {boneName=head}
   }
   n15 = Anim Dynamics - Bone: dyn_R_hair_end_r!(ComponentPose=@n14, bAlphaBoolEnabled:bool=True, GravityOverride:Vector="0,0,-980", ExternalForce="0,0,0") -> (Pose) {
    alphaInputType=Bool
    angularDampingOverride=0.999
    angularSpringConstant=2000
    bAngularSpring=true
    bOverrideAngularDamping=true
    bUseGravityOverride=true
    gravityScale=0.35
    simulationSpace=BoneRelative
    alphaBoolBlend {blendInTime=2, blendOption=Cubic}
    boundBone {boneName=dyn_R_hair_end_r}
    componentAppliedLinearAccClamp=<3000 3000 3000>
    componentLinearAccScale=<0.3 0.4 0.3>
    componentLinearVelScale=<0.3 0.4 0.3>
    gravityOverride {z=-980}
    physicsBodyDefinitions [{
      boundBone {boneName=dyn_R_hair_end_r}
      boxExtents=<25 25 25>
      constraintSetup {angularLimitsMax=<0 40 0>, angularLimitsMin=<0 0 0>, angularTarget=<1 0 0>}
      localJointOffset=<2 0 0>
     }]
    relativeSpaceBone {boneName=head}
   }
   n16 = Anim Dynamics - Bone: dyn_R_hair_end_l!(ComponentPose=@n15, bAlphaBoolEnabled:bool=True, GravityOverride:Vector="0,0,-980", ExternalForce="0,0,0") -> (Pose) {
    alphaInputType=Bool
    angularDampingOverride=0.999
    angularSpringConstant=10000
    bAngularSpring=true
    bOverrideAngularDamping=true
    bUseGravityOverride=true
    gravityScale=0.35
    simulationSpace=BoneRelative
    alphaBoolBlend {blendInTime=2, blendOption=Cubic}
    boundBone {boneName=dyn_R_hair_end_l}
    componentAppliedLinearAccClamp=<3000 3000 3000>
    componentLinearAccScale=<0.3 0.4 0.3>
    componentLinearVelScale=<0.3 0.4 0.3>
    gravityOverride {z=-980}
    physicsBodyDefinitions [{
      boundBone {boneName=dyn_R_hair_end_l}
      boxExtents=<25 25 25>
      constraintSetup {angularLimitsMax=<0 40 0>, angularLimitsMin=<0 0 0>, angularTarget=<1 0 0>}
      localJointOffset=<2 0 0>
     }]
    relativeSpaceBone {boneName=head}
   }
   n17 = Anim Dynamics - Bone: dyn_L_hair_offshoot_end!(ComponentPose=@n16, bAlphaBoolEnabled:bool=True, GravityOverride:Vector="0,0,-980", ExternalForce="0,0,0") -> (Pose) {
    alphaInputType=Bool
    angularDampingOverride=0.999
    angularSpringConstant=2000
    bAngularSpring=true
    bOverrideAngularDamping=true
    bUseGravityOverride=true
    gravityScale=0.35
    simulationSpace=BoneRelative
    alphaBoolBlend {blendInTime=2, blendOption=Cubic}
    boundBone {boneName=dyn_L_hair_offshoot_end}
    componentAppliedLinearAccClamp=<3000 3000 3000>
    componentLinearAccScale=<0.3 0.4 0.3>
    componentLinearVelScale=<0.3 0.4 0.3>
    gravityOverride {z=-980}
    physicsBodyDefinitions [{
      boundBone {boneName=dyn_L_hair_offshoot_end}
      boxExtents=<25 25 25>
      constraintSetup {angularLimitsMax=<0 0 5>, angularLimitsMin=<0 0 0>, angularTarget=<1 0 0>}
      localJointOffset=<2 0 0>
     }]
    relativeSpaceBone {boneName=head}
   }
   n18 = Anim Dynamics - Bone: dyn_R_hair_offshoot_end!(ComponentPose=@n17, bAlphaBoolEnabled:bool=True, GravityOverride:Vector="0,0,-980", ExternalForce="0,0,0") -> (Pose) {
    alphaInputType=Bool
    angularDampingOverride=0.999
    angularSpringConstant=1000
    bAngularSpring=true
    bOverrideAngularDamping=true
    bUseGravityOverride=true
    gravityScale=0.35
    simulationSpace=BoneRelative
    alphaBoolBlend {blendInTime=2, blendOption=Cubic}
    boundBone {boneName=dyn_R_hair_offshoot_end}
    componentAppliedLinearAccClamp=<3000 3000 3000>
    componentLinearAccScale=<0.3 0.4 0.3>
    componentLinearVelScale=<0.3 0.4 0.3>
    gravityOverride {z=-980}
    physicsBodyDefinitions [{
      boundBone {boneName=dyn_R_hair_offshoot_end}
      boxExtents=<25 25 25>
      constraintSetup {angularLimitsMax=<0 0 5>, angularLimitsMin=<0 0 0>, angularTarget=<1 0 0>}
      localJointOffset=<2 0 0>
     }]
    relativeSpaceBone {boneName=head}
   }
   n19 = Anim Dynamics - Bone: dyn_LF_bangs_1(ComponentPose=@n18, Alpha=1, ExternalForce="0,0,0") {
    angularDampingOverride=1
    angularSpringConstant=3000
    bAngularSpring=true
    bOverrideAngularDamping=true
    bUseGravityOverride=true
    gravityScale=0.35
    simulationSpace=BoneRelative
    boundBone {boneName=dyn_LF_bangs_1}
    componentAppliedLinearAccClamp=<1000 1000 1000>
    componentLinearAccScale=<0.3 0.4 0.3>
    componentLinearVelScale=<0.3 0.4 0.3>
    gravityOverride {z=-980}
    physicsBodyDefinitions [{
      boundBone {boneName=dyn_LF_bangs_1}
      boxExtents=<20 30 30>
      constraintSetup {angularLimitsMax=<0 0 0>, angularLimitsMin=<0 -5 -10>, angularTarget=<1 0 0>}
      localJointOffset=<2 0 0>
     }]
    relativeSpaceBone {boneName=head}
    simSpaceSettings {maxAngularAcceleration=100, maxAngularVelocity=10, simSpaceAngularAlpha=1}
   }
   n20 = Anim Dynamics - Bone: dyn_LF_bangs_2(ComponentPose=@n19, Alpha=1, ExternalForce="0,0,0") {
    angularDampingOverride=1
    angularSpringConstant=3000
    bAngularSpring=true
    bOverrideAngularDamping=true
    bUseGravityOverride=true
    gravityScale=0.35
    simulationSpace=BoneRelative
    boundBone {boneName=dyn_LF_bangs_2}
    componentAppliedLinearAccClamp=<1000 1000 1000>
    componentLinearAccScale=<0.3 0.4 0.3>
    componentLinearVelScale=<0.3 0.4 0.3>
    gravityOverride {z=-980}
    physicsBodyDefinitions [{
      boundBone {boneName=dyn_LF_bangs_2}
      boxExtents=<20 30 30>
      constraintSetup {angularLimitsMax=<0 0 0>, angularLimitsMin=<0 -5 -10>, angularTarget=<1 0 0>}
      localJointOffset=<2 0 0>
     }]
    relativeSpaceBone {boneName=head}
    simSpaceSettings {maxAngularAcceleration=100, maxAngularVelocity=10, simSpaceAngularAlpha=1}
   }
   n21 = Anim Dynamics - Bone: dyn_LF_bangs_3(ComponentPose=@n20, Alpha=1, ExternalForce="0,0,0") {
    angularDampingOverride=1
    angularSpringConstant=3000
    bAngularSpring=true
    bOverrideAngularDamping=true
    bUseGravityOverride=true
    gravityScale=0.35
    simulationSpace=BoneRelative
    boundBone {boneName=dyn_LF_bangs_3}
    componentAppliedLinearAccClamp=<1000 1000 1000>
    componentLinearAccScale=<0.3 0.4 0.3>
    componentLinearVelScale=<0.3 0.4 0.3>
    gravityOverride {z=-980}
    physicsBodyDefinitions [{
      boundBone {boneName=dyn_LF_bangs_3}
      boxExtents=<20 30 30>
      constraintSetup {angularLimitsMax=<0 0 0>, angularLimitsMin=<0 -5 -15>, angularTarget=<1 0 0>}
      localJointOffset=<2 0 0>
     }]
    relativeSpaceBone {boneName=head}
    simSpaceSettings {maxAngularAcceleration=100, maxAngularVelocity=10, simSpaceAngularAlpha=1}
   }
   n22 = Anim Dynamics - Bone: dyn_LB_bangs_1(ComponentPose=@n21, Alpha=1, ExternalForce="0,0,0") {
    angularDampingOverride=1
    angularSpringConstant=3000
    bAngularSpring=true
    bOverrideAngularDamping=true
    bUseGravityOverride=true
    gravityScale=0.35
    simulationSpace=BoneRelative
    boundBone {boneName=dyn_LB_bangs_1}
    componentAppliedLinearAccClamp=<1000 1000 1000>
    componentLinearAccScale=<0.3 0.4 0.3>
    componentLinearVelScale=<0.3 0.4 0.3>
    gravityOverride {z=-980}
    physicsBodyDefinitions [{
      boundBone {boneName=dyn_LB_bangs_1}
      boxExtents=<20 30 30>
      constraintSetup {angularLimitsMax=<0 5 0>, angularLimitsMin=<0 -5 -10>, angularTarget=<1 0 0>}
      localJointOffset=<2 0 0>
     }]
    relativeSpaceBone {boneName=head}
    simSpaceSettings {maxAngularAcceleration=100, maxAngularVelocity=10, simSpaceAngularAlpha=1}
   }
   n23 = Anim Dynamics - Bone: dyn_LB_bangs_3(ComponentPose=@n22, Alpha=1, ExternalForce="0,0,0") {
    angularDampingOverride=1
    angularSpringConstant=3000
    bAngularSpring=true
    bOverrideAngularDamping=true
    bUseGravityOverride=true
    gravityScale=0.35
    simulationSpace=BoneRelative
    boundBone {boneName=dyn_LB_bangs_3}
    componentAppliedLinearAccClamp=<1000 1000 1000>
    componentLinearAccScale=<0.3 0.4 0.3>
    componentLinearVelScale=<0.3 0.4 0.3>
    gravityOverride {z=-980}
    physicsBodyDefinitions [{
      boundBone {boneName=dyn_LB_bangs_3}
      boxExtents=<20 30 30>
      constraintSetup {angularLimitsMax=<0 5 0>, angularLimitsMin=<0 -5 -10>, angularTarget=<1 0 0>}
      localJointOffset=<2 0 0>
     }]
    relativeSpaceBone {boneName=head}
    simSpaceSettings {maxAngularAcceleration=100, maxAngularVelocity=10, simSpaceAngularAlpha=1}
   }
   n24 = Anim Dynamics - Bone: dyn_LB_bangs_2(ComponentPose=@n23, Alpha=1, ExternalForce="0,0,0") {
    angularDampingOverride=1
    angularSpringConstant=3000
    bAngularSpring=true
    bOverrideAngularDamping=true
    bUseGravityOverride=true
    gravityScale=0.35
    simulationSpace=BoneRelative
    boundBone {boneName=dyn_LB_bangs_2}
    componentAppliedLinearAccClamp=<1000 1000 1000>
    componentLinearAccScale=<0.3 0.4 0.3>
    componentLinearVelScale=<0.3 0.4 0.3>
    gravityOverride {z=-980}
    physicsBodyDefinitions [{
      boundBone {boneName=dyn_LB_bangs_2}
      boxExtents=<20 30 30>
      constraintSetup {angularLimitsMax=<0 5 0>, angularLimitsMin=<0 -5 -15>, angularTarget=<1 0 0>}
      localJointOffset=<2 0 0>
     }]
    relativeSpaceBone {boneName=head}
    simSpaceSettings {maxAngularAcceleration=100, maxAngularVelocity=10, simSpaceAngularAlpha=1}
   }
   n25 = Anim Dynamics - Bone: dyn_RB_bangs_1(ComponentPose=@n24, Alpha=1, ExternalForce="0,0,0") {
    angularDampingOverride=1
    angularSpringConstant=3000
    bAngularSpring=true
    bOverrideAngularDamping=true
    bUseGravityOverride=true
    gravityScale=0.35
    simulationSpace=BoneRelative
    boundBone {boneName=dyn_RB_bangs_1}
    componentAppliedLinearAccClamp=<1000 1000 1000>
    componentLinearAccScale=<0.3 0.4 0.3>
    componentLinearVelScale=<0.3 0.4 0.3>
    gravityOverride {z=-980}
    physicsBodyDefinitions [{
      boundBone {boneName=dyn_RB_bangs_1}
      boxExtents=<20 30 30>
      constraintSetup {angularLimitsMax=<0 5 10>, angularLimitsMin=<0 -5 0>, angularTarget=<1 0 0>}
      localJointOffset=<2 0 0>
     }]
    relativeSpaceBone {boneName=head}
    simSpaceSettings {maxAngularAcceleration=100, maxAngularVelocity=10, simSpaceAngularAlpha=1}
   }
   n26 = Anim Dynamics - Bone: dyn_RB_bangs_2(ComponentPose=@n25, Alpha=1, ExternalForce="0,0,0") {
    angularDampingOverride=1
    angularSpringConstant=3000
    bAngularSpring=true
    bOverrideAngularDamping=true
    bUseGravityOverride=true
    gravityScale=0.35
    simulationSpace=BoneRelative
    boundBone {boneName=dyn_RB_bangs_2}
    componentAppliedLinearAccClamp=<1000 1000 1000>
    componentLinearAccScale=<0.3 0.4 0.3>
    componentLinearVelScale=<0.3 0.4 0.3>
    gravityOverride {z=-980}
    physicsBodyDefinitions [{
      boundBone {boneName=dyn_RB_bangs_2}
      boxExtents=<20 30 30>
      constraintSetup {angularLimitsMax=<0 5 10>, angularLimitsMin=<0 -5 0>, angularTarget=<1 0 0>}
      localJointOffset=<2 0 0>
     }]
    relativeSpaceBone {boneName=head}
    simSpaceSettings {maxAngularAcceleration=100, maxAngularVelocity=10, simSpaceAngularAlpha=1}
   }
   n27 = Anim Dynamics - Bone: dyn_RB_bangs_3(ComponentPose=@n26, Alpha=1, ExternalForce="0,0,0") {
    angularDampingOverride=1
    angularSpringConstant=3000
    bAngularSpring=true
    bOverrideAngularDamping=true
    bUseGravityOverride=true
    gravityScale=0.35
    simulationSpace=BoneRelative
    boundBone {boneName=dyn_RB_bangs_3}
    componentAppliedLinearAccClamp=<1000 1000 1000>
    componentLinearAccScale=<0.3 0.4 0.3>
    componentLinearVelScale=<0.3 0.4 0.3>
    gravityOverride {z=-980}
    physicsBodyDefinitions [{
      boundBone {boneName=dyn_RB_bangs_3}
      boxExtents=<20 30 30>
      constraintSetup {angularLimitsMax=<0 5 15>, angularLimitsMin=<0 -5 0>, angularTarget=<1 0 0>}
      localJointOffset=<2 0 0>
     }]
    relativeSpaceBone {boneName=head}
    simSpaceSettings {maxAngularAcceleration=100, maxAngularVelocity=10, simSpaceAngularAlpha=1}
   }
   n28 = Anim Dynamics - Bone: dyn_RF_bangs_1(ComponentPose=@n27, Alpha=1, ExternalForce="0,0,0") {
    angularDampingOverride=1
    angularSpringConstant=3000
    bAngularSpring=true
    bOverrideAngularDamping=true
    bUseGravityOverride=true
    gravityScale=0.35
    simulationSpace=BoneRelative
    boundBone {boneName=dyn_RF_bangs_1}
    componentAppliedLinearAccClamp=<1000 1000 1000>
    componentLinearAccScale=<0.3 0.4 0.3>
    componentLinearVelScale=<0.3 0.4 0.3>
    gravityOverride {z=-980}
    physicsBodyDefinitions [{
      boundBone {boneName=dyn_RF_bangs_1}
      boxExtents=<20 30 30>
      constraintSetup {angularLimitsMax=<0 5 10>, angularLimitsMin=<0 -5 0>, angularTarget=<1 0 0>}
      localJointOffset=<2 0 0>
     }]
    relativeSpaceBone {boneName=head}
    simSpaceSettings {maxAngularAcceleration=100, maxAngularVelocity=10, simSpaceAngularAlpha=1}
   }
   n29 = Anim Dynamics - Bone: dyn_RF_bangs_2(ComponentPose=@n28, Alpha=1, ExternalForce="0,0,0") {
    angularDampingOverride=1
    angularSpringConstant=3000
    bAngularSpring=true
    bOverrideAngularDamping=true
    bUseGravityOverride=true
    gravityScale=0.35
    simulationSpace=BoneRelative
    boundBone {boneName=dyn_RF_bangs_2}
    componentAppliedLinearAccClamp=<1000 1000 1000>
    componentLinearAccScale=<0.3 0.4 0.3>
    componentLinearVelScale=<0.3 0.4 0.3>
    gravityOverride {z=-980}
    physicsBodyDefinitions [{
      boundBone {boneName=dyn_RF_bangs_2}
      boxExtents=<20 30 30>
      constraintSetup {angularLimitsMax=<0 5 10>, angularLimitsMin=<0 -5 0>, angularTarget=<1 0 0>}
      localJointOffset=<2 0 0>
     }]
    relativeSpaceBone {boneName=head}
    simSpaceSettings {maxAngularAcceleration=100, maxAngularVelocity=10, simSpaceAngularAlpha=1}
   }
   n30 = Anim Dynamics - Bone: dyn_RF_bangs_3(ComponentPose=@n29, Alpha=1, ExternalForce="0,0,0") {
    angularDampingOverride=1
    angularSpringConstant=3000
    bAngularSpring=true
    bOverrideAngularDamping=true
    bUseGravityOverride=true
    gravityScale=0.35
    simulationSpace=BoneRelative
    boundBone {boneName=dyn_RF_bangs_3}
    componentAppliedLinearAccClamp=<1000 1000 1000>
    componentLinearAccScale=<0.3 0.4 0.3>
    componentLinearVelScale=<0.3 0.4 0.3>
    gravityOverride {z=-980}
    physicsBodyDefinitions [{
      boundBone {boneName=dyn_RF_bangs_3}
      boxExtents=<20 30 30>
      constraintSetup {angularLimitsMax=<0 5 15>, angularLimitsMin=<0 -5 0>, angularTarget=<1 0 0>}
      localJointOffset=<2 0 0>
     }]
    relativeSpaceBone {boneName=head}
    simSpaceSettings {maxAngularAcceleration=100, maxAngularVelocity=10, simSpaceAngularAlpha=1}
   }
   n31 = Anim Dynamics - Bone: dyn_RF_bangs_8(ComponentPose=@n30, Alpha=1, ExternalForce="0,0,0") {
    angularDampingOverride=1
    angularSpringConstant=3000
    bAngularSpring=true
    bOverrideAngularDamping=true
    bUseGravityOverride=true
    gravityScale=0.35
    simulationSpace=BoneRelative
    boundBone {boneName=dyn_RF_bangs_8}
    componentAppliedLinearAccClamp=<1000 1000 1000>
    componentLinearAccScale=<0.3 0.4 0.3>
    componentLinearVelScale=<0.3 0.4 0.3>
    gravityOverride {z=-980}
    physicsBodyDefinitions [{
      boundBone {boneName=dyn_RF_bangs_8}
      boxExtents=<20 30 30>
      constraintSetup {angularLimitsMax=<0 5 15>, angularLimitsMin=<0 -5 0>, angularTarget=<1 0 0>}
      localJointOffset=<2 0 0>
     }]
    relativeSpaceBone {boneName=head}
    simSpaceSettings {maxAngularAcceleration=100, maxAngularVelocity=10, simSpaceAngularAlpha=1}
   }
   n32 = Anim Dynamics - Bone: dyn_C_hair_spike(ComponentPose=@n31, Alpha=1, ExternalForce="0,0,0") {
    angularDampingOverride=1
    angularSpringConstant=20000
    bAngularSpring=true
    bOverrideAngularDamping=true
    bUseGravityOverride=true
    gravityScale=0.35
    simulationSpace=RootRelative
    boundBone {boneName=dyn_C_hair_spike}
    componentLinearAccScale=<1 1 1>
    componentLinearVelScale=<1 1 1>
    gravityOverride {z=-980}
    physicsBodyDefinitions [{
      boundBone {boneName=dyn_C_hair_spike}
      boxExtents=<20 30 30>
      constraintSetup {angularLimitsMax=<0 0 10>, angularLimitsMin=<0 0 -10>, angularTarget=<1 0 0>}
      localJointOffset=<4 0 0>
     }]
   }
   n33 = Anim Dynamics - Bone: dyn_C_hair_spike_end(ComponentPose=@n32, Alpha=1, ExternalForce="0,0,0") {
    angularDampingOverride=1
    angularSpringConstant=20000
    bAngularSpring=true
    bOverrideAngularDamping=true
    bUseGravityOverride=true
    gravityScale=0.35
    simulationSpace=RootRelative
    boundBone {boneName=dyn_C_hair_spike_end}
    componentLinearAccScale=<1 1 1>
    componentLinearVelScale=<1 1 1>
    gravityOverride {z=-980}
    physicsBodyDefinitions [{
      boundBone {boneName=dyn_C_hair_spike_end}
      boxExtents=<20 30 30>
      constraintSetup {angularLimitsMax=<0 0 10>, angularLimitsMin=<0 0 -10>, angularTarget=<1 0 0>}
      localJointOffset=<4 0 0>
     }]
   }
   n34 = "Constraint - dyn_R_hair_spike (PPPPPPPP)"!(
    ComponentPose=@n33
    Alpha=1
    ConstraintWeights_0=0.4162
    ConstraintWeights_1:float=0.2445
    ConstraintWeights_2:float=0.1143
    ConstraintWeights_3:float=0.0821
    ConstraintWeights_4:float=0.0681
    ConstraintWeights_5:float=0.0456
    ConstraintWeights_6:float=0.0231
    ConstraintWeights_7:float=0.0062
   ) -> (Pose) {
    boneToModify {boneName=dyn_R_hair_spike}
    constraintSetup [
     {targetBone {boneName=dyn_R_hair_3}}
     {targetBone {boneName=dyn_R_hair_4}}
     {targetBone {boneName=dyn_R_hair_spike}}
     {targetBone {boneName=dyn_C_hair_3}}
     {targetBone {boneName=dyn_R_hair_2}}
     {targetBone {boneName=dyn_C_hair_4}}
     {targetBone {boneName=dyn_C_hair_2}}
     {targetBone {boneName=dyn_C_hair_5}}
    ]
    constraintWeights [1,1,1,1,1,1,1,1]
   }
   n35 = "Constraint - dyn_L_hair_spike (PPPPPPPP)"!(
    ComponentPose=@n34
    Alpha=1
    ConstraintWeights_0=0.4335
    ConstraintWeights_1:float=0.2762
    ConstraintWeights_2:float=0.1067
    ConstraintWeights_3:float=0.0636
    ConstraintWeights_4:float=0.051
    ConstraintWeights_5:float=0.0407
    ConstraintWeights_6:float=0.0233
    ConstraintWeights_7:float=0.005
   ) -> (Pose) {
    boneToModify {boneName=dyn_L_hair_spike}
    constraintSetup [
     {targetBone {boneName=dyn_L_hair_4}}
     {targetBone {boneName=dyn_L_hair_3}}
     {targetBone {boneName=dyn_C_hair_4}}
     {targetBone {boneName=dyn_L_hair_spike}}
     {targetBone {boneName=dyn_C_hair_3}}
     {targetBone {boneName=dyn_L_hair_5}}
     {targetBone {boneName=dyn_C_hair_5}}
     {targetBone {boneName=dyn_L_hair_2}}
    ]
    constraintWeights [1,1,1,1,1,1,1,1]
   }
   n36 = Anim Dynamics - Bone: dyn_R_hair_spike(ComponentPose=@n35, Alpha=1, ExternalForce="0,0,0") {
    angularDampingOverride=1
    angularSpringConstant=20000
    bAngularSpring=true
    bOverrideAngularDamping=true
    bUseGravityOverride=true
    gravityScale=0.35
    simulationSpace=RootRelative
    boundBone {boneName=dyn_R_hair_spike}
    componentLinearAccScale=<1 1 1>
    componentLinearVelScale=<1 1 1>
    gravityOverride {z=-980}
    physicsBodyDefinitions [{
      boundBone {boneName=dyn_R_hair_spike}
      boxExtents=<20 30 30>
      constraintSetup {angularLimitsMax=<0 20 15>, angularLimitsMin=<0 0 -15>, angularTarget=<1 0 0>}
      localJointOffset=<4 0 0>
     }]
   }
   n37 = Anim Dynamics - Bone: dyn_L_hair_spike(ComponentPose=@n36, Alpha=1, ExternalForce="0,0,0") {
    angularDampingOverride=1
    angularSpringConstant=20000
    bAngularSpring=true
    bOverrideAngularDamping=true
    bUseGravityOverride=true
    gravityScale=0.35
    simulationSpace=RootRelative
    boundBone {boneName=dyn_L_hair_spike}
    componentLinearAccScale=<1 1 1>
    componentLinearVelScale=<1 1 1>
    gravityOverride {z=-980}
    physicsBodyDefinitions [{
      boundBone {boneName=dyn_L_hair_spike}
      boxExtents=<20 30 30>
      constraintSetup {angularLimitsMax=<0 20 0>, angularLimitsMin=<0 0 -15>, angularTarget=<1 0 0>}
      localJointOffset=<4 0 0>
     }]
   }
   n38 = "Constraint - dyn_L_hair_5 (P)"(ComponentPose=@n37, Alpha=1, ConstraintWeights_0=1) {boneToModify {boneName=dyn_L_hair_5}, constraintSetup [{targetBone {boneName=dyn_C_hair_5}}], constraintWeights [1]}
   n39 = "Constraint - dyn_R_hair_5 (P)"(ComponentPose=@n38, Alpha=1, ConstraintWeights_0=1) {boneToModify {boneName=dyn_R_hair_5}, constraintSetup [{targetBone {boneName=dyn_C_hair_5}}], constraintWeights [1]}
   n40 = "Constraint - dyn_L_hair_4 (P)"(ComponentPose=@n39, Alpha=1, ConstraintWeights_0=1) {boneToModify {boneName=dyn_L_hair_4}, constraintSetup [{targetBone {boneName=dyn_C_hair_4}}], constraintWeights [1]}
   n41 = "Constraint - dyn_R_hair_4 (P)"(ComponentPose=@n40, Alpha=1, ConstraintWeights_0=1) {boneToModify {boneName=dyn_R_hair_4}, constraintSetup [{targetBone {boneName=dyn_C_hair_4}}], constraintWeights [1]}
   n42 = "Constraint - dyn_R_hair_offshoot (P)"(ComponentPose=@n41, Alpha=1, ConstraintWeights_0=1) {boneToModify {boneName=dyn_R_hair_offshoot}, constraintSetup [{targetBone {boneName=dyn_R_hair_4}}], constraintWeights [1]}
   Save cached pose 'High Inert'(Pose=Component To Local(ComponentPose=@n42)) {cacheName=High Inert}
   n43 = Anim Dynamics - Bone: dyn_L_hair_2!(
    ComponentPose=Local To Component(LocalPose=Use cached pose 'Cntrl Rig'())
    bAlphaBoolEnabled:bool=True
    GravityOverride:Vector="0,0,-980"
    AngularSpringConstant:float=20000
    ExternalForce="0,0,0"
   ) -> (Pose) {
    alphaInputType=Bool
    angularDampingOverride=0.9999
    angularSpringConstant=20000
    bAngularSpring=true
    bOverrideAngularDamping=true
    bUseGravityOverride=true
    gravityScale=0.35
    simulationSpace=BoneRelative
    alphaBoolBlend {blendInTime=2, blendOption=Cubic}
    boundBone {boneName=dyn_L_hair_2}
    componentAppliedLinearAccClamp=<10000 10000 10000>
    componentLinearAccScale=<0.3 0.4 0.3>
    componentLinearVelScale=<0.3 0.4 0.3>
    gravityOverride {z=-980}
    physicsBodyDefinitions [{
      boundBone {boneName=dyn_L_hair_2}
      boxExtents=<25 25 25>
      constraintSetup {angularLimitsMax=<0 25 0>, angularLimitsMin=<0 0 -2>, angularTarget=<1 0 0>}
      localJointOffset=<4 0 0>
     }]
    relativeSpaceBone {boneName=spine_05}
    simSpaceSettings {maxAngularAcceleration=100, maxAngularVelocity=10, simSpaceAngularAlpha=0.4}
   }
   n44 = Anim Dynamics - Bone: dyn_C_hair_2!(ComponentPose=@n43, bAlphaBoolEnabled:bool=True, GravityOverride:Vector="0,0,-980", AngularSpringConstant:float=20000, ExternalForce="0,0,0") -> (Pose) {
    alphaInputType=Bool
    angularDampingOverride=0.9999
    angularSpringConstant=20000
    bAngularSpring=true
    bOverrideAngularDamping=true
    bUseGravityOverride=true
    gravityScale=0.35
    simulationSpace=BoneRelative
    alphaBoolBlend {blendInTime=2, blendOption=Cubic}
    boundBone {boneName=dyn_C_hair_2}
    componentAppliedLinearAccClamp=<10000 10000 10000>
    componentLinearAccScale=<0.3 0.4 0.3>
    componentLinearVelScale=<0.3 0.4 0.3>
    gravityOverride {z=-980}
    physicsBodyDefinitions [{
      boundBone {boneName=dyn_C_hair_2}
      boxExtents=<25 25 25>
      constraintSetup {angularLimitsMax=<0 25 2>, angularLimitsMin=<0 0 -2>, angularTarget=<1 0 0>}
      localJointOffset=<4 0 0>
     }]
    relativeSpaceBone {boneName=spine_05}
    simSpaceSettings {maxAngularAcceleration=100, maxAngularVelocity=10, simSpaceAngularAlpha=0.4}
   }
   n45 = Anim Dynamics - Bone: dyn_R_hair_2!(ComponentPose=@n44, bAlphaBoolEnabled:bool=True, GravityOverride:Vector="0,0,-980", AngularSpringConstant:float=20000, ExternalForce="0,0,0") -> (Pose) {
    alphaInputType=Bool
    angularDampingOverride=0.9999
    angularSpringConstant=20000
    bAngularSpring=true
    bOverrideAngularDamping=true
    bUseGravityOverride=true
    gravityScale=0.35
    simulationSpace=BoneRelative
    alphaBoolBlend {blendInTime=2, blendOption=Cubic}
    boundBone {boneName=dyn_R_hair_2}
    componentAppliedLinearAccClamp=<10000 10000 10000>
    componentLinearAccScale=<0.3 0.4 0.3>
    componentLinearVelScale=<0.3 0.4 0.3>
    gravityOverride {z=-980}
    physicsBodyDefinitions [{
      boundBone {boneName=dyn_R_hair_2}
      boxExtents=<25 25 25>
      constraintSetup {angularLimitsMax=<0 25 2>, angularLimitsMin=<0 0 0>, angularTarget=<1 0 0>}
      localJointOffset=<4 0 0>
     }]
    relativeSpaceBone {boneName=spine_05}
    simSpaceSettings {maxAngularAcceleration=100, maxAngularVelocity=10, simSpaceAngularAlpha=0.4}
   }
   n46 = Anim Dynamics - Bone: dyn_L_hair_end_l!(ComponentPose=@n45, bAlphaBoolEnabled=True, GravityOverride="0,0,-980", AngularSpringConstant:float=5000, ExternalForce="0,0,0") -> (Pose) {
    alphaInputType=Bool
    angularDampingOverride=0.9999
    angularSpringConstant=5000
    bAngularSpring=true
    bOverrideAngularDamping=true
    bUseGravityOverride=true
    gravityScale=0.35
    simulationSpace=BoneRelative
    alphaBoolBlend {blendInTime=2, blendOption=Cubic}
    boundBone {boneName=dyn_L_hair_end_l}
    componentAppliedLinearAccClamp=<10000 10000 10000>
    componentLinearAccScale=<0.3 0.4 0.3>
    componentLinearVelScale=<0.3 0.4 0.3>
    gravityOverride {z=-980}
    physicsBodyDefinitions [{
      boundBone {boneName=dyn_L_hair_end_l}
      boxExtents=<25 25 25>
      constraintSetup {angularLimitsMax=<0 40 10>, angularLimitsMin=<0 0 -10>, angularTarget=<1 0 0>}
      localJointOffset=<4 0 0>
     }]
    simSpaceSettings {maxAngularAcceleration=100, maxAngularVelocity=10, simSpaceAngularAlpha=0.4}
   }
   n47 = Anim Dynamics - Bone: dyn_L_hair_end_r!(ComponentPose=@n46, bAlphaBoolEnabled=True, GravityOverride="0,0,-980", AngularSpringConstant:float=12000, ExternalForce="0,0,0") -> (Pose) {
    alphaInputType=Bool
    angularDampingOverride=0.9999
    angularSpringConstant=12000
    bAngularSpring=true
    bOverrideAngularDamping=true
    bUseGravityOverride=true
    gravityScale=0.35
    simulationSpace=BoneRelative
    alphaBoolBlend {blendInTime=2, blendOption=Cubic}
    boundBone {boneName=dyn_L_hair_end_r}
    componentAppliedLinearAccClamp=<10000 10000 10000>
    componentLinearAccScale=<0.3 0.4 0.3>
    componentLinearVelScale=<0.3 0.4 0.3>
    gravityOverride {z=-980}
    physicsBodyDefinitions [{
      boundBone {boneName=dyn_L_hair_end_r}
      boxExtents=<20 20 20>
      constraintSetup {angularLimitsMax=<0 40 5>, angularLimitsMin=<0 0 -5>, angularTarget=<1 0 0>}
      localJointOffset=<4 0 0>
     }]
    relativeSpaceBone {boneName=root}
    simSpaceSettings {maxAngularAcceleration=100, maxAngularVelocity=10, simSpaceAngularAlpha=0.4}
   }
   n48 = Anim Dynamics - Bone: dyn_C_hair_end_r!(ComponentPose=@n47, bAlphaBoolEnabled=True, GravityOverride="0,0,-980", AngularSpringConstant:float=3000, ExternalForce="0,0,0") -> (Pose) {
    alphaInputType=Bool
    angularDampingOverride=0.9999
    angularSpringConstant=3000
    bAngularSpring=true
    bOverrideAngularDamping=true
    bUseGravityOverride=true
    gravityScale=0.35
    simulationSpace=BoneRelative
    alphaBoolBlend {blendInTime=2, blendOption=Cubic}
    boundBone {boneName=dyn_C_hair_end_r}
    componentAppliedLinearAccClamp=<10000 10000 10000>
    componentLinearAccScale=<0.3 0.4 0.3>
    componentLinearVelScale=<0.3 0.4 0.3>
    gravityOverride {z=-980}
    physicsBodyDefinitions [{
      boundBone {boneName=dyn_C_hair_end_r}
      boxExtents=<25 25 25>
      constraintSetup {angularLimitsMax=<0 0 10>, angularLimitsMin=<0 0 -10>, angularTarget=<1 0 0>}
      localJointOffset=<4 0 0>
     }]
    relativeSpaceBone {boneName=root}
    simSpaceSettings {maxAngularAcceleration=100, maxAngularVelocity=10, simSpaceAngularAlpha=0.4}
   }
   n49 = Anim Dynamics - Bone: dyn_L_hair_3!(ComponentPose=@n48, bAlphaBoolEnabled:bool=True, GravityOverride:Vector="0,0,-980", AngularSpringConstant:float=20000, ExternalForce="0,0,0") -> (Pose) {
    alphaInputType=Bool
    angularDampingOverride=0.9999
    angularSpringConstant=20000
    bAngularSpring=true
    bOverrideAngularDamping=true
    bUseGravityOverride=true
    gravityScale=0.35
    simulationSpace=BoneRelative
    alphaBoolBlend {blendInTime=2, blendOption=Cubic}
    boundBone {boneName=dyn_L_hair_3}
    componentAppliedLinearAccClamp=<10000 10000 10000>
    componentLinearAccScale=<0.3 0.4 0.3>
    componentLinearVelScale=<0.3 0.4 0.3>
    gravityOverride {z=-980}
    physicsBodyDefinitions [{
      boundBone {boneName=dyn_L_hair_3}
      boxExtents=<25 25 25>
      constraintSetup {angularLimitsMax=<0 35 0>, angularLimitsMin=<0 0 -2>, angularTarget=<1 0 0>}
      localJointOffset=<4 0 0>
     }]
    relativeSpaceBone {boneName=spine_05}
    simSpaceSettings {maxAngularAcceleration=100, maxAngularVelocity=10, simSpaceAngularAlpha=0.4}
   }
   n50 = Anim Dynamics - Bone: dyn_C_hair_3!(ComponentPose=@n49, bAlphaBoolEnabled:bool=True, GravityOverride:Vector="0,0,-980", AngularSpringConstant:float=20000, ExternalForce="0,0,0") -> (Pose) {
    alphaInputType=Bool
    angularDampingOverride=0.9999
    angularSpringConstant=20000
    bAngularSpring=true
    bOverrideAngularDamping=true
    bUseGravityOverride=true
    gravityScale=0.35
    simulationSpace=BoneRelative
    alphaBoolBlend {blendInTime=2, blendOption=Cubic}
    boundBone {boneName=dyn_C_hair_3}
    componentAppliedLinearAccClamp=<10000 10000 10000>
    componentLinearAccScale=<0.3 0.4 0.3>
    componentLinearVelScale=<0.3 0.4 0.3>
    gravityOverride {z=-980}
    physicsBodyDefinitions [{
      boundBone {boneName=dyn_C_hair_3}
      boxExtents=<25 25 25>
      constraintSetup {angularLimitsMax=<0 35 2>, angularLimitsMin=<0 0 -2>, angularTarget=<1 0 0>}
      localJointOffset=<4 0 0>
     }]
    relativeSpaceBone {boneName=spine_05}
    simSpaceSettings {maxAngularAcceleration=100, maxAngularVelocity=10, simSpaceAngularAlpha=0.4}
   }
   n51 = Anim Dynamics - Bone: dyn_R_hair_3!(ComponentPose=@n50, bAlphaBoolEnabled:bool=True, GravityOverride:Vector="0,0,-980", AngularSpringConstant:float=20000, ExternalForce="0,0,0") -> (Pose) {
    alphaInputType=Bool
    angularDampingOverride=0.9999
    angularSpringConstant=20000
    bAngularSpring=true
    bOverrideAngularDamping=true
    bUseGravityOverride=true
    gravityScale=0.35
    simulationSpace=BoneRelative
    alphaBoolBlend {blendInTime=2, blendOption=Cubic}
    boundBone {boneName=dyn_R_hair_3}
    componentAppliedLinearAccClamp=<10000 10000 10000>
    componentLinearAccScale=<0.3 0.4 0.3>
    componentLinearVelScale=<0.3 0.4 0.3>
    gravityOverride {z=-980}
    physicsBodyDefinitions [{
      boundBone {boneName=dyn_R_hair_3}
      boxExtents=<25 25 25>
      constraintSetup {angularLimitsMax=<0 35 2>, angularLimitsMin=<0 0 0>, angularTarget=<1 0 0>}
      localJointOffset=<4 0 0>
     }]
    relativeSpaceBone {boneName=spine_05}
    simSpaceSettings {maxAngularAcceleration=100, maxAngularVelocity=10, simSpaceAngularAlpha=0.4}
   }
   n52 = Anim Dynamics - Bone: dyn_C_hair_end_l!(ComponentPose=@n51, bAlphaBoolEnabled=True, GravityOverride="0,0,-980", AngularSpringConstant:float=15000, ExternalForce="0,0,0") -> (Pose) {
    alphaInputType=Bool
    angularDampingOverride=0.9999
    angularSpringConstant=15000
    bAngularSpring=true
    bOverrideAngularDamping=true
    bUseGravityOverride=true
    gravityScale=0.35
    simulationSpace=BoneRelative
    alphaBoolBlend {blendInTime=2, blendOption=Cubic}
    boundBone {boneName=dyn_C_hair_end_l}
    componentAppliedLinearAccClamp=<10000 10000 10000>
    componentLinearAccScale=<0.3 0.4 0.3>
    componentLinearVelScale=<0.3 0.4 0.3>
    gravityOverride {z=-980}
    physicsBodyDefinitions [{
      boundBone {boneName=dyn_C_hair_end_l}
      boxExtents=<25 25 25>
      constraintSetup {angularLimitsMax=<0 0 10>, angularLimitsMin=<0 0 -10>, angularTarget=<1 0 0>}
      localJointOffset=<4 0 0>
     }]
    relativeSpaceBone {boneName=root}
    simSpaceSettings {maxAngularAcceleration=100, maxAngularVelocity=10, simSpaceAngularAlpha=0.4}
   }
   n53 = Anim Dynamics - Bone: dyn_C_hair_end_c!(ComponentPose=@n52, bAlphaBoolEnabled=True, GravityOverride="0,0,-980", AngularSpringConstant:float=1000, ExternalForce="0,0,0") -> (Pose) {
    alphaInputType=Bool
    angularDampingOverride=0.9999
    angularSpringConstant=1000
    bAngularSpring=true
    bOverrideAngularDamping=true
    bUseGravityOverride=true
    gravityScale=0.35
    simulationSpace=RootRelative
    alphaBoolBlend {blendInTime=2, blendOption=Cubic}
    boundBone {boneName=dyn_C_hair_end_c}
    componentAppliedLinearAccClamp=<10000 10000 10000>
    componentLinearAccScale=<0.3 0.4 0.3>
    componentLinearVelScale=<0.3 0.4 0.3>
    gravityOverride {z=-980}
    physicsBodyDefinitions [{
      boundBone {boneName=dyn_C_hair_end_c}
      boxExtents=<25 25 25>
      constraintSetup {angularLimitsMax=<0 0 5>, angularLimitsMin=<0 0 -5>, angularTarget=<1 0 0>}
      localJointOffset=<4 0 0>
     }]
    relativeSpaceBone {boneName=root}
    simSpaceSettings {maxAngularAcceleration=100, maxAngularVelocity=10, simSpaceAngularAlpha=0.4}
   }
   n54 = Anim Dynamics - Bone: dyn_R_hair_end_r!(ComponentPose=@n53, bAlphaBoolEnabled=True, GravityOverride="0,0,-980", AngularSpringConstant:float=2000, ExternalForce="0,0,0") -> (Pose) {
    alphaInputType=Bool
    angularDampingOverride=0.9999
    angularSpringConstant=2000
    bAngularSpring=true
    bOverrideAngularDamping=true
    bUseGravityOverride=true
    gravityScale=0.35
    simulationSpace=RootRelative
    alphaBoolBlend {blendInTime=2, blendOption=Cubic}
    boundBone {boneName=dyn_R_hair_end_r}
    componentAppliedLinearAccClamp=<10000 10000 10000>
    componentLinearAccScale=<0.3 0.4 0.3>
    componentLinearVelScale=<0.3 0.4 0.3>
    gravityOverride {z=-980}
    physicsBodyDefinitions [{
      boundBone {boneName=dyn_R_hair_end_r}
      boxExtents=<25 25 25>
      constraintSetup {angularLimitsMax=<0 40 10>, angularLimitsMin=<0 0 -10>, angularTarget=<1 0 0>}
      localJointOffset=<2 0 0>
     }]
    relativeSpaceBone {boneName=root}
    simSpaceSettings {maxAngularAcceleration=100, maxAngularVelocity=10, simSpaceAngularAlpha=0.4}
   }
   n55 = Anim Dynamics - Bone: dyn_L_hair_4(ComponentPose=@n54, bAlphaBoolEnabled=True, GravityOverride="0,0,-980", AngularSpringConstant=20000, ExternalForce="0,0,0") {
    alphaInputType=Bool
    angularDampingOverride=0.9999
    angularSpringConstant=20000
    bAngularSpring=true
    bOverrideAngularDamping=true
    bUseGravityOverride=true
    gravityScale=0.35
    simulationSpace=RootRelative
    alphaBoolBlend {blendInTime=2, blendOption=Cubic}
    boundBone {boneName=dyn_L_hair_4}
    componentAppliedLinearAccClamp=<10000 10000 10000>
    componentLinearAccScale=<0.3 0.4 0.3>
    componentLinearVelScale=<0.3 0.4 0.3>
    gravityOverride {z=-980}
    physicsBodyDefinitions [{
      boundBone {boneName=dyn_L_hair_4}
      boxExtents=<30 15 15>
      constraintSetup {angularLimitsMax=<0 45 5>, angularLimitsMin=<0 0 -5>, angularTarget=<1 0 0>}
      localJointOffset=<4 0 0>
     }]
    relativeSpaceBone {boneName=root}
    simSpaceSettings {maxAngularAcceleration=100, maxAngularVelocity=10, simSpaceAngularAlpha=0.4}
   }
   n56 = Anim Dynamics - Bone: dyn_C_hair_4(ComponentPose=@n55, bAlphaBoolEnabled=True, GravityOverride="0,0,-980", AngularSpringConstant=20000, ExternalForce="0,0,0") {
    alphaInputType=Bool
    angularDampingOverride=0.9999
    angularSpringConstant=20000
    bAngularSpring=true
    bOverrideAngularDamping=true
    bUseGravityOverride=true
    gravityScale=0.35
    simulationSpace=RootRelative
    alphaBoolBlend {blendInTime=2, blendOption=Cubic}
    boundBone {boneName=dyn_C_hair_4}
    componentAppliedLinearAccClamp=<10000 10000 10000>
    componentLinearAccScale=<0.3 0.4 0.3>
    componentLinearVelScale=<0.3 0.4 0.3>
    gravityOverride {z=-980}
    physicsBodyDefinitions [{
      boundBone {boneName=dyn_C_hair_4}
      boxExtents=<30 15 15>
      constraintSetup {angularLimitsMax=<0 45 5>, angularLimitsMin=<0 0 -5>, angularTarget=<1 0 0>}
      localJointOffset=<4 0 0>
     }]
    relativeSpaceBone {boneName=root}
    simSpaceSettings {maxAngularAcceleration=100, maxAngularVelocity=10, simSpaceAngularAlpha=0.4}
   }
   n57 = Anim Dynamics - Bone: dyn_R_hair_4(ComponentPose=@n56, bAlphaBoolEnabled=True, GravityOverride="0,0,-980", AngularSpringConstant=20000, ExternalForce="0,0,0") {
    alphaInputType=Bool
    angularDampingOverride=0.9999
    angularSpringConstant=20000
    bAngularSpring=true
    bOverrideAngularDamping=true
    bUseGravityOverride=true
    gravityScale=0.35
    simulationSpace=RootRelative
    alphaBoolBlend {blendInTime=2, blendOption=Cubic}
    boundBone {boneName=dyn_R_hair_4}
    componentAppliedLinearAccClamp=<10000 10000 10000>
    componentLinearAccScale=<0.3 0.4 0.3>
    componentLinearVelScale=<0.3 0.4 0.3>
    gravityOverride {z=-980}
    physicsBodyDefinitions [{
      boundBone {boneName=dyn_R_hair_4}
      boxExtents=<30 15 15>
      constraintSetup {angularLimitsMax=<0 45 5>, angularLimitsMin=<0 0 0>, angularTarget=<1 0 0>}
      localJointOffset=<4 0 0>
     }]
    relativeSpaceBone {boneName=root}
    simSpaceSettings {maxAngularAcceleration=100, maxAngularVelocity=10, simSpaceAngularAlpha=0.4}
   }
   n58 = Anim Dynamics - Bone: dyn_R_hair_end_l!(ComponentPose=@n57, bAlphaBoolEnabled=True, GravityOverride="0,0,-980", AngularSpringConstant:float=10000, ExternalForce="0,0,0") -> (Pose) {
    alphaInputType=Bool
    angularDampingOverride=0.9999
    angularSpringConstant=10000
    bAngularSpring=true
    bOverrideAngularDamping=true
    bUseGravityOverride=true
    gravityScale=0.35
    simulationSpace=RootRelative
    alphaBoolBlend {blendInTime=2, blendOption=Cubic}
    boundBone {boneName=dyn_R_hair_end_l}
    componentAppliedLinearAccClamp=<10000 10000 10000>
    componentLinearAccScale=<0.3 0.4 0.3>
    componentLinearVelScale=<0.3 0.4 0.3>
    gravityOverride {z=-980}
    physicsBodyDefinitions [{
      boundBone {boneName=dyn_R_hair_end_l}
      boxExtents=<25 25 25>
      constraintSetup {angularLimitsMax=<0 40 0>, angularLimitsMin=<0 0 0>, angularTarget=<1 0 0>}
      localJointOffset=<2 0 0>
     }]
    relativeSpaceBone {boneName=root}
    simSpaceSettings {maxAngularAcceleration=100, maxAngularVelocity=10, simSpaceAngularAlpha=0.4}
   }
   n59 = Anim Dynamics - Bone: dyn_L_hair_offshoot!(ComponentPose=@n58, bAlphaBoolEnabled:bool=True, GravityOverride:Vector="0,0,-980", AngularSpringConstant:float=20000, ExternalForce="0,0,0") -> (Pose) {
    alphaInputType=Bool
    angularDampingOverride=0.9999
    angularSpringConstant=20000
    bAngularSpring=true
    bOverrideAngularDamping=true
    bUseGravityOverride=true
    gravityScale=0.35
    simulationSpace=RootRelative
    alphaBoolBlend {blendInTime=2, blendOption=Cubic}
    boundBone {boneName=dyn_L_hair_offshoot}
    componentAppliedLinearAccClamp=<10000 10000 10000>
    componentLinearAccScale=<0.3 0.4 0.3>
    componentLinearVelScale=<0.3 0.4 0.3>
    gravityOverride {z=-980}
    physicsBodyDefinitions [{
      boundBone {boneName=dyn_L_hair_offshoot}
      boxExtents=<25 25 25>
      constraintSetup {angularLimitsMax=<0 20 0>, angularLimitsMin=<0 0 -15>, angularTarget=<1 0 0>}
      localJointOffset=<2 0 0>
     }]
    relativeSpaceBone {boneName=root}
    simSpaceSettings {maxAngularAcceleration=100, maxAngularVelocity=10, simSpaceAngularAlpha=0.4}
   }
   n60 = Anim Dynamics - Bone: dyn_R_hair_offshoot!(ComponentPose=@n59, bAlphaBoolEnabled:bool=True, GravityOverride:Vector="0,0,-980", AngularSpringConstant:float=1000, ExternalForce="0,0,0") -> (Pose) {
    alphaInputType=Bool
    angularDampingOverride=0.9999
    angularSpringConstant=1000
    bAngularSpring=true
    bOverrideAngularDamping=true
    bUseGravityOverride=true
    gravityScale=0.35
    simulationSpace=RootRelative
    alphaBoolBlend {blendInTime=2, blendOption=Cubic}
    boundBone {boneName=dyn_R_hair_offshoot}
    componentAppliedLinearAccClamp=<10000 10000 10000>
    componentLinearAccScale=<0.3 0.4 0.3>
    componentLinearVelScale=<0.3 0.4 0.3>
    gravityOverride {z=-980}
    physicsBodyDefinitions [{
      boundBone {boneName=dyn_R_hair_offshoot}
      boxExtents=<25 25 25>
      constraintSetup {angularLimitsMax=<0 0 10>, angularLimitsMin=<0 0 0>, angularTarget=<1 0 0>}
      localJointOffset=<4 0 0>
     }]
    relativeSpaceBone {boneName=root}
    simSpaceSettings {maxAngularAcceleration=100, maxAngularVelocity=10, simSpaceAngularAlpha=0.4}
   }
   n61 = Anim Dynamics - Bone: dyn_L_hair_5(ComponentPose=@n60, bAlphaBoolEnabled=True, GravityOverride="0,0,-980", AngularSpringConstant=20000, ExternalForce="0,0,0") {
    alphaInputType=Bool
    angularDampingOverride=0.9999
    angularSpringConstant=20000
    bAngularSpring=true
    bOverrideAngularDamping=true
    bUseGravityOverride=true
    gravityScale=0.35
    simulationSpace=RootRelative
    alphaBoolBlend {blendInTime=2, blendOption=Cubic}
    boundBone {boneName=dyn_L_hair_5}
    componentAppliedLinearAccClamp=<10000 10000 10000>
    componentLinearAccScale=<0.3 0.4 0.3>
    componentLinearVelScale=<0.3 0.4 0.3>
    gravityOverride {z=-980}
    physicsBodyDefinitions [{
      boundBone {boneName=dyn_L_hair_5}
      boxExtents=<30 15 15>
      constraintSetup {angularLimitsMax=<0 55 10>, angularLimitsMin=<0 -40 -10>, angularTarget=<1 0 0>}
      localJointOffset=<4 0 0>
     }]
    relativeSpaceBone {boneName=root}
    simSpaceSettings {maxAngularAcceleration=100, maxAngularVelocity=10, simSpaceAngularAlpha=0.4}
   }
   n62 = Anim Dynamics - Bone: dyn_C_hair_5(ComponentPose=@n61, bAlphaBoolEnabled=True, GravityOverride="0,0,-980", AngularSpringConstant=20000, ExternalForce="0,0,0") {
    alphaInputType=Bool
    angularDampingOverride=0.9999
    angularSpringConstant=20000
    bAngularSpring=true
    bOverrideAngularDamping=true
    bUseGravityOverride=true
    gravityScale=0.35
    simulationSpace=RootRelative
    alphaBoolBlend {blendInTime=2, blendOption=Cubic}
    boundBone {boneName=dyn_C_hair_5}
    componentAppliedLinearAccClamp=<10000 10000 10000>
    componentLinearAccScale=<0.3 0.4 0.3>
    componentLinearVelScale=<0.3 0.4 0.3>
    gravityOverride {z=-980}
    physicsBodyDefinitions [{
      boundBone {boneName=dyn_C_hair_5}
      boxExtents=<30 15 15>
      constraintSetup {angularLimitsMax=<0 55 5>, angularLimitsMin=<0 0 5>, angularTarget=<1 0 0>}
      localJointOffset=<4 0 0>
     }]
    relativeSpaceBone {boneName=root}
    simSpaceSettings {maxAngularAcceleration=100, maxAngularVelocity=10, simSpaceAngularAlpha=0.4}
   }
   n63 = Anim Dynamics - Bone: dyn_R_hair_5(ComponentPose=@n62, bAlphaBoolEnabled=True, GravityOverride="0,0,-980", AngularSpringConstant=20000, ExternalForce="0,0,0") {
    alphaInputType=Bool
    angularDampingOverride=0.9999
    angularSpringConstant=20000
    bAngularSpring=true
    bOverrideAngularDamping=true
    bUseGravityOverride=true
    gravityScale=0.35
    simulationSpace=RootRelative
    alphaBoolBlend {blendInTime=2, blendOption=Cubic}
    boundBone {boneName=dyn_R_hair_5}
    componentAppliedLinearAccClamp=<10000 10000 10000>
    componentLinearAccScale=<0.3 0.4 0.3>
    componentLinearVelScale=<0.3 0.4 0.3>
    gravityOverride {z=-980}
    physicsBodyDefinitions [{
      boundBone {boneName=dyn_R_hair_5}
      boxExtents=<30 15 15>
      constraintSetup {angularLimitsMax=<0 55 10>, angularLimitsMin=<0 -40 0>, angularTarget=<1 0 0>}
      localJointOffset=<4 0 0>
     }]
    relativeSpaceBone {boneName=root}
    simSpaceSettings {maxAngularAcceleration=100, maxAngularVelocity=10, simSpaceAngularAlpha=0.4}
   }
   n64 = Anim Dynamics - Bone: dyn_L_hair_offshoot_end!(ComponentPose=@n63, bAlphaBoolEnabled=True, GravityOverride="0,0,-980", AngularSpringConstant:float=2000, ExternalForce="0,0,0") -> (Pose) {
    alphaInputType=Bool
    angularDampingOverride=0.9999
    angularSpringConstant=2000
    bAngularSpring=true
    bOverrideAngularDamping=true
    bUseGravityOverride=true
    gravityScale=0.35
    simulationSpace=RootRelative
    alphaBoolBlend {blendInTime=2, blendOption=Cubic}
    boundBone {boneName=dyn_L_hair_offshoot_end}
    componentAppliedLinearAccClamp=<10000 10000 10000>
    componentLinearAccScale=<0.3 0.4 0.3>
    componentLinearVelScale=<0.3 0.4 0.3>
    gravityOverride {z=-980}
    physicsBodyDefinitions [{
      boundBone {boneName=dyn_L_hair_offshoot_end}
      boxExtents=<25 25 25>
      constraintSetup {angularLimitsMax=<0 20 15>, angularLimitsMin=<0 0 0>, angularTarget=<1 0 0>}
      localJointOffset=<2 0 0>
     }]
    relativeSpaceBone {boneName=root}
    simSpaceSettings {maxAngularAcceleration=100, maxAngularVelocity=10, simSpaceAngularAlpha=0.4}
   }
   n65 = Anim Dynamics - Bone: dyn_R_hair_offshoot_end!(ComponentPose=@n64, bAlphaBoolEnabled=True, GravityOverride="0,0,-980", AngularSpringConstant:float=1000, ExternalForce="0,0,0") -> (Pose) {
    alphaInputType=Bool
    angularDampingOverride=0.9999
    angularSpringConstant=1000
    bAngularSpring=true
    bOverrideAngularDamping=true
    bUseGravityOverride=true
    gravityScale=0.35
    simulationSpace=RootRelative
    alphaBoolBlend {blendInTime=2, blendOption=Cubic}
    boundBone {boneName=dyn_R_hair_offshoot_end}
    componentAppliedLinearAccClamp=<10000 10000 10000>
    componentLinearAccScale=<0.3 0.4 0.3>
    componentLinearVelScale=<0.3 0.4 0.3>
    gravityOverride {z=-980}
    physicsBodyDefinitions [{
      boundBone {boneName=dyn_R_hair_offshoot_end}
      boxExtents=<25 25 25>
      constraintSetup {angularLimitsMax=<0 0 20>, angularLimitsMin=<0 0 0>, angularTarget=<1 0 0>}
      localJointOffset=<4 0 0>
     }]
    relativeSpaceBone {boneName=root}
    simSpaceSettings {maxAngularAcceleration=100, maxAngularVelocity=10, simSpaceAngularAlpha=0.4}
   }
   n66 = Anim Dynamics - Bone: dyn_C_hair_6!(ComponentPose=@n65, bAlphaBoolEnabled:bool=True, GravityOverride:Vector="0,0,-980", AngularSpringConstant:float=20000, ExternalForce="0,0,0") -> (Pose) {
    alphaInputType=Bool
    angularDampingOverride=0.9999
    angularSpringConstant=20000
    bAngularSpring=true
    bOverrideAngularDamping=true
    bUseGravityOverride=true
    gravityScale=0.35
    simulationSpace=RootRelative
    alphaBoolBlend {blendInTime=2, blendOption=Cubic}
    boundBone {boneName=dyn_C_hair_6}
    componentAppliedLinearAccClamp=<10000 10000 10000>
    componentLinearAccScale=<0.3 0.4 0.3>
    componentLinearVelScale=<0.3 0.4 0.3>
    gravityOverride {z=-980}
    physicsBodyDefinitions [{
      boundBone {boneName=dyn_C_hair_6}
      boxExtents=<30 15 15>
      constraintSetup {angularLimitsMax=<0 65 0>, angularLimitsMin=<0 0 0>, angularTarget=<1 0 0>}
      localJointOffset=<4 0 0>
     }]
    relativeSpaceBone {boneName=root}
    simSpaceSettings {maxAngularAcceleration=100, maxAngularVelocity=10, simSpaceAngularAlpha=0.4}
   }
   n67 = Anim Dynamics - Bone: dyn_LF_bangs_1(ComponentPose=@n66, Alpha=1, ExternalForce="0,0,0") {
    angularDampingOverride=1
    angularSpringConstant=20000
    bAngularSpring=true
    bOverrideAngularDamping=true
    bUseGravityOverride=true
    gravityScale=0.35
    simulationSpace=BoneRelative
    boundBone {boneName=dyn_LF_bangs_1}
    componentLinearAccScale=<1 1 1>
    componentLinearVelScale=<1 1 1>
    gravityOverride {z=-980}
    physicsBodyDefinitions [{
      boundBone {boneName=dyn_LF_bangs_1}
      boxExtents=<20 30 30>
      constraintSetup {angularLimitsMax=<0 0 0>, angularLimitsMin=<0 -5 -10>, angularTarget=<1 0 0>}
      localJointOffset=<4 0 0>
     }]
    simSpaceSettings {maxAngularAcceleration=1000, maxAngularVelocity=1000, simSpaceAngularAlpha=0.2}
   }
   n68 = Anim Dynamics - Bone: dyn_LF_bangs_2(ComponentPose=@n67, Alpha=1, ExternalForce="0,0,0") {
    angularDampingOverride=1
    angularSpringConstant=20000
    bAngularSpring=true
    bOverrideAngularDamping=true
    bUseGravityOverride=true
    gravityScale=0.35
    simulationSpace=BoneRelative
    boundBone {boneName=dyn_LF_bangs_2}
    componentLinearAccScale=<1 1 1>
    componentLinearVelScale=<1 1 1>
    gravityOverride {z=-980}
    physicsBodyDefinitions [{
      boundBone {boneName=dyn_LF_bangs_2}
      boxExtents=<20 30 30>
      constraintSetup {angularLimitsMax=<0 0 0>, angularLimitsMin=<0 -5 -10>, angularTarget=<1 0 0>}
      localJointOffset=<4 0 0>
     }]
    simSpaceSettings {maxAngularAcceleration=1000, maxAngularVelocity=1000, simSpaceAngularAlpha=0.2}
   }
   n69 = Anim Dynamics - Bone: dyn_LF_bangs_3(ComponentPose=@n68, Alpha=1, ExternalForce="0,0,0") {
    angularDampingOverride=1
    angularSpringConstant=20000
    bAngularSpring=true
    bOverrideAngularDamping=true
    bUseGravityOverride=true
    gravityScale=0.35
    simulationSpace=BoneRelative
    boundBone {boneName=dyn_LF_bangs_3}
    componentLinearAccScale=<1 1 1>
    componentLinearVelScale=<1 1 1>
    gravityOverride {z=-980}
    physicsBodyDefinitions [{
      boundBone {boneName=dyn_LF_bangs_3}
      boxExtents=<20 30 30>
      constraintSetup {angularLimitsMax=<0 0 0>, angularLimitsMin=<0 -5 -15>, angularTarget=<1 0 0>}
      localJointOffset=<4 0 0>
     }]
    simSpaceSettings {maxAngularAcceleration=1000, maxAngularVelocity=1000, simSpaceAngularAlpha=0.2}
   }
   n70 = Anim Dynamics - Bone: dyn_RB_bangs_1(ComponentPose=@n69, Alpha=1, ExternalForce="0,0,0") {
    angularDampingOverride=1
    angularSpringConstant=20000
    bAngularSpring=true
    bOverrideAngularDamping=true
    bUseGravityOverride=true
    gravityScale=0.35
    simulationSpace=BoneRelative
    boundBone {boneName=dyn_RB_bangs_1}
    componentLinearAccScale=<1 1 1>
    componentLinearVelScale=<1 1 1>
    gravityOverride {z=-980}
    physicsBodyDefinitions [{
      boundBone {boneName=dyn_RB_bangs_1}
      boxExtents=<20 30 30>
      constraintSetup {angularLimitsMax=<0 5 10>, angularLimitsMin=<0 -5 0>, angularTarget=<1 0 0>}
      localJointOffset=<4 0 0>
     }]
    simSpaceSettings {maxAngularAcceleration=1000, maxAngularVelocity=1000, simSpaceAngularAlpha=0.2}
   }
   n71 = Anim Dynamics - Bone: dyn_RB_bangs_2(ComponentPose=@n70, Alpha=1, ExternalForce="0,0,0") {
    angularDampingOverride=1
    angularSpringConstant=20000
    bAngularSpring=true
    bOverrideAngularDamping=true
    bUseGravityOverride=true
    gravityScale=0.35
    simulationSpace=BoneRelative
    boundBone {boneName=dyn_RB_bangs_2}
    componentLinearAccScale=<1 1 1>
    componentLinearVelScale=<1 1 1>
    gravityOverride {z=-980}
    physicsBodyDefinitions [{
      boundBone {boneName=dyn_RB_bangs_2}
      boxExtents=<20 30 30>
      constraintSetup {angularLimitsMax=<0 5 10>, angularLimitsMin=<0 -5 0>, angularTarget=<1 0 0>}
      localJointOffset=<4 0 0>
     }]
    simSpaceSettings {maxAngularAcceleration=1000, maxAngularVelocity=1000, simSpaceAngularAlpha=0.2}
   }
   n72 = Anim Dynamics - Bone: dyn_RB_bangs_3(ComponentPose=@n71, Alpha=1, ExternalForce="0,0,0") {
    angularDampingOverride=1
    angularSpringConstant=20000
    bAngularSpring=true
    bOverrideAngularDamping=true
    bUseGravityOverride=true
    gravityScale=0.35
    simulationSpace=BoneRelative
    boundBone {boneName=dyn_RB_bangs_3}
    componentLinearAccScale=<1 1 1>
    componentLinearVelScale=<1 1 1>
    gravityOverride {z=-980}
    physicsBodyDefinitions [{
      boundBone {boneName=dyn_RB_bangs_3}
      boxExtents=<20 30 30>
      constraintSetup {angularLimitsMax=<0 5 15>, angularLimitsMin=<0 -5 0>, angularTarget=<1 0 0>}
      localJointOffset=<4 0 0>
     }]
    simSpaceSettings {maxAngularAcceleration=1000, maxAngularVelocity=1000, simSpaceAngularAlpha=0.2}
   }
   n73 = Anim Dynamics - Bone: dyn_LB_bangs_1(ComponentPose=@n72, Alpha=1, ExternalForce="0,0,0") {
    angularDampingOverride=1
    angularSpringConstant=20000
    bAngularSpring=true
    bOverrideAngularDamping=true
    bUseGravityOverride=true
    gravityScale=0.35
    simulationSpace=BoneRelative
    boundBone {boneName=dyn_LB_bangs_1}
    componentLinearAccScale=<1 1 1>
    componentLinearVelScale=<1 1 1>
    gravityOverride {z=-980}
    physicsBodyDefinitions [{
      boundBone {boneName=dyn_LB_bangs_1}
      boxExtents=<20 30 30>
      constraintSetup {angularLimitsMax=<0 5 0>, angularLimitsMin=<0 -5 -10>, angularTarget=<1 0 0>}
      localJointOffset=<4 0 0>
     }]
    simSpaceSettings {maxAngularAcceleration=1000, maxAngularVelocity=1000, simSpaceAngularAlpha=0.2}
   }
   n74 = Anim Dynamics - Bone: dyn_LB_bangs_3(ComponentPose=@n73, Alpha=1, ExternalForce="0,0,0") {
    angularDampingOverride=1
    angularSpringConstant=20000
    bAngularSpring=true
    bOverrideAngularDamping=true
    bUseGravityOverride=true
    gravityScale=0.35
    simulationSpace=BoneRelative
    boundBone {boneName=dyn_LB_bangs_3}
    componentLinearAccScale=<1 1 1>
    componentLinearVelScale=<1 1 1>
    gravityOverride {z=-980}
    physicsBodyDefinitions [{
      boundBone {boneName=dyn_LB_bangs_3}
      boxExtents=<20 30 30>
      constraintSetup {angularLimitsMax=<0 5 0>, angularLimitsMin=<0 -5 -10>, angularTarget=<1 0 0>}
      localJointOffset=<4 0 0>
     }]
    simSpaceSettings {maxAngularAcceleration=1000, maxAngularVelocity=1000, simSpaceAngularAlpha=0.2}
   }
   n75 = Anim Dynamics - Bone: dyn_LB_bangs_2(ComponentPose=@n74, Alpha=1, ExternalForce="0,0,0") {
    angularDampingOverride=1
    angularSpringConstant=20000
    bAngularSpring=true
    bOverrideAngularDamping=true
    bUseGravityOverride=true
    gravityScale=0.35
    simulationSpace=BoneRelative
    boundBone {boneName=dyn_LB_bangs_2}
    componentLinearAccScale=<1 1 1>
    componentLinearVelScale=<1 1 1>
    gravityOverride {z=-980}
    physicsBodyDefinitions [{
      boundBone {boneName=dyn_LB_bangs_2}
      boxExtents=<20 30 30>
      constraintSetup {angularLimitsMax=<0 5 0>, angularLimitsMin=<0 -5 -15>, angularTarget=<1 0 0>}
      localJointOffset=<4 0 0>
     }]
    simSpaceSettings {maxAngularAcceleration=1000, maxAngularVelocity=1000, simSpaceAngularAlpha=0.2}
   }
   n76 = Anim Dynamics - Bone: dyn_RF_bangs_1(ComponentPose=@n75, Alpha=1, ExternalForce="0,0,0") {
    angularDampingOverride=1
    angularSpringConstant=20000
    bAngularSpring=true
    bOverrideAngularDamping=true
    bUseGravityOverride=true
    gravityScale=0.35
    simulationSpace=BoneRelative
    boundBone {boneName=dyn_RF_bangs_1}
    componentLinearAccScale=<1 1 1>
    componentLinearVelScale=<1 1 1>
    gravityOverride {z=-980}
    physicsBodyDefinitions [{
      boundBone {boneName=dyn_RF_bangs_1}
      boxExtents=<20 30 30>
      constraintSetup {angularLimitsMax=<0 5 10>, angularLimitsMin=<0 -5 0>, angularTarget=<1 0 0>}
      localJointOffset=<4 0 0>
     }]
    simSpaceSettings {maxAngularAcceleration=1000, maxAngularVelocity=1000, simSpaceAngularAlpha=0.2}
   }
   n77 = Anim Dynamics - Bone: dyn_RF_bangs_2(ComponentPose=@n76, Alpha=1, ExternalForce="0,0,0") {
    angularDampingOverride=1
    angularSpringConstant=20000
    bAngularSpring=true
    bOverrideAngularDamping=true
    bUseGravityOverride=true
    gravityScale=0.35
    simulationSpace=BoneRelative
    boundBone {boneName=dyn_RF_bangs_2}
    componentLinearAccScale=<1 1 1>
    componentLinearVelScale=<1 1 1>
    gravityOverride {z=-980}
    physicsBodyDefinitions [{
      boundBone {boneName=dyn_RF_bangs_2}
      boxExtents=<20 30 30>
      constraintSetup {angularLimitsMax=<0 5 10>, angularLimitsMin=<0 -5 0>, angularTarget=<1 0 0>}
      localJointOffset=<4 0 0>
     }]
    simSpaceSettings {maxAngularAcceleration=1000, maxAngularVelocity=1000, simSpaceAngularAlpha=0.2}
   }
   n78 = Anim Dynamics - Bone: dyn_RF_bangs_3(ComponentPose=@n77, Alpha=1, ExternalForce="0,0,0") {
    angularDampingOverride=1
    angularSpringConstant=20000
    bAngularSpring=true
    bOverrideAngularDamping=true
    bUseGravityOverride=true
    gravityScale=0.35
    simulationSpace=BoneRelative
    boundBone {boneName=dyn_RF_bangs_3}
    componentLinearAccScale=<1 1 1>
    componentLinearVelScale=<1 1 1>
    gravityOverride {z=-980}
    physicsBodyDefinitions [{
      boundBone {boneName=dyn_RF_bangs_3}
      boxExtents=<20 30 30>
      constraintSetup {angularLimitsMax=<0 5 15>, angularLimitsMin=<0 -5 0>, angularTarget=<1 0 0>}
      localJointOffset=<4 0 0>
     }]
    simSpaceSettings {maxAngularAcceleration=1000, maxAngularVelocity=1000, simSpaceAngularAlpha=0.2}
   }
   n79 = Anim Dynamics - Bone: dyn_RF_bangs_8(ComponentPose=@n78, Alpha=1, ExternalForce="0,0,0") {
    angularDampingOverride=1
    angularSpringConstant=20000
    bAngularSpring=true
    bOverrideAngularDamping=true
    bUseGravityOverride=true
    gravityScale=0.35
    simulationSpace=BoneRelative
    boundBone {boneName=dyn_RF_bangs_8}
    componentLinearAccScale=<1 1 1>
    componentLinearVelScale=<1 1 1>
    gravityOverride {z=-980}
    physicsBodyDefinitions [{
      boundBone {boneName=dyn_RF_bangs_8}
      boxExtents=<20 30 30>
      constraintSetup {angularLimitsMax=<0 5 15>, angularLimitsMin=<0 -5 0>, angularTarget=<1 0 0>}
      localJointOffset=<4 0 0>
     }]
    simSpaceSettings {maxAngularAcceleration=1000, maxAngularVelocity=1000, simSpaceAngularAlpha=0.2}
   }
   n80 = Anim Dynamics - Bone: dyn_C_hair_spike(ComponentPose=@n79, Alpha=1, ExternalForce="0,0,0") {
    angularDampingOverride=1
    angularSpringConstant=20000
    bAngularSpring=true
    bOverrideAngularDamping=true
    bUseGravityOverride=true
    gravityScale=0.35
    simulationSpace=RootRelative
    boundBone {boneName=dyn_C_hair_spike}
    componentLinearAccScale=<1 1 1>
    componentLinearVelScale=<1 1 1>
    gravityOverride {z=-980}
    physicsBodyDefinitions [{
      boundBone {boneName=dyn_C_hair_spike}
      boxExtents=<20 30 30>
      constraintSetup {angularLimitsMax=<0 0 10>, angularLimitsMin=<0 0 -10>, angularTarget=<1 0 0>}
      localJointOffset=<4 0 0>
     }]
   }
   n81 = Anim Dynamics - Bone: dyn_C_hair_spike_end(ComponentPose=@n80, Alpha=1, ExternalForce="0,0,0") {
    angularDampingOverride=1
    angularSpringConstant=20000
    bAngularSpring=true
    bOverrideAngularDamping=true
    bUseGravityOverride=true
    gravityScale=0.35
    simulationSpace=RootRelative
    boundBone {boneName=dyn_C_hair_spike_end}
    componentLinearAccScale=<1 1 1>
    componentLinearVelScale=<1 1 1>
    gravityOverride {z=-980}
    physicsBodyDefinitions [{
      boundBone {boneName=dyn_C_hair_spike_end}
      boxExtents=<20 30 30>
      constraintSetup {angularLimitsMax=<0 0 10>, angularLimitsMin=<0 0 -10>, angularTarget=<1 0 0>}
      localJointOffset=<4 0 0>
     }]
   }
   n82 = "Constraint - dyn_R_hair_spike (PPPPPPPP)"(
    ComponentPose=@n81
    Alpha=1
    ConstraintWeights_0=0.4162
    ConstraintWeights_1=0.2445
    ConstraintWeights_2=0.1143
    ConstraintWeights_3=0.0821
    ConstraintWeights_4=0.0681
    ConstraintWeights_5=0.0456
    ConstraintWeights_6=0.0231
    ConstraintWeights_7=0.0062
   ) {
    boneToModify {boneName=dyn_R_hair_spike}
    constraintSetup [
     {targetBone {boneName=dyn_R_hair_3}}
     {targetBone {boneName=dyn_R_hair_4}}
     {targetBone {boneName=dyn_R_hair_spike}}
     {targetBone {boneName=dyn_C_hair_3}}
     {targetBone {boneName=dyn_R_hair_2}}
     {targetBone {boneName=dyn_C_hair_4}}
     {targetBone {boneName=dyn_C_hair_2}}
     {targetBone {boneName=dyn_C_hair_5}}
    ]
    constraintWeights [1,1,1,1,1,1,1,1]
   }
   n83 = "Constraint - dyn_L_hair_spike (PPPPPPPP)"(
    ComponentPose=@n82
    Alpha=1
    ConstraintWeights_0=0.4335
    ConstraintWeights_1=0.2762
    ConstraintWeights_2=0.1067
    ConstraintWeights_3=0.0636
    ConstraintWeights_4=0.051
    ConstraintWeights_5=0.0407
    ConstraintWeights_6=0.0233
    ConstraintWeights_7=0.005
   ) {
    boneToModify {boneName=dyn_L_hair_spike}
    constraintSetup [
     {targetBone {boneName=dyn_L_hair_4}}
     {targetBone {boneName=dyn_L_hair_3}}
     {targetBone {boneName=dyn_C_hair_4}}
     {targetBone {boneName=dyn_L_hair_spike}}
     {targetBone {boneName=dyn_C_hair_3}}
     {targetBone {boneName=dyn_L_hair_5}}
     {targetBone {boneName=dyn_C_hair_5}}
     {targetBone {boneName=dyn_L_hair_2}}
    ]
    constraintWeights [1,1,1,1,1,1,1,1]
   }
   n84 = Anim Dynamics - Bone: dyn_R_hair_spike(ComponentPose=@n83, Alpha=1, ExternalForce="0,0,0") {
    angularDampingOverride=1
    angularSpringConstant=20000
    bAngularSpring=true
    bOverrideAngularDamping=true
    bUseGravityOverride=true
    gravityScale=0.35
    simulationSpace=RootRelative
    boundBone {boneName=dyn_R_hair_spike}
    componentLinearAccScale=<1 1 1>
    componentLinearVelScale=<1 1 1>
    gravityOverride {z=-980}
    physicsBodyDefinitions [{
      boundBone {boneName=dyn_R_hair_spike}
      boxExtents=<20 30 30>
      constraintSetup {angularLimitsMax=<0 20 15>, angularLimitsMin=<0 0 -15>, angularTarget=<1 0 0>}
      localJointOffset=<4 0 0>
     }]
   }
   n85 = Anim Dynamics - Bone: dyn_L_hair_spike(ComponentPose=@n84, Alpha=1, ExternalForce="0,0,0") {
    angularDampingOverride=1
    angularSpringConstant=20000
    bAngularSpring=true
    bOverrideAngularDamping=true
    bUseGravityOverride=true
    gravityScale=0.35
    simulationSpace=RootRelative
    boundBone {boneName=dyn_L_hair_spike}
    componentLinearAccScale=<1 1 1>
    componentLinearVelScale=<1 1 1>
    gravityOverride {z=-980}
    physicsBodyDefinitions [{
      boundBone {boneName=dyn_L_hair_spike}
      boxExtents=<20 30 30>
      constraintSetup {angularLimitsMax=<0 20 0>, angularLimitsMin=<0 0 -15>, angularTarget=<1 0 0>}
      localJointOffset=<4 0 0>
     }]
   }
   n86 = "Constraint - dyn_L_hair_5 (P)"(ComponentPose=@n85, Alpha=1, ConstraintWeights_0=1) {boneToModify {boneName=dyn_L_hair_5}, constraintSetup [{targetBone {boneName=dyn_C_hair_5}}], constraintWeights [1]}
   n87 = "Constraint - dyn_R_hair_5 (P)"(ComponentPose=@n86, Alpha=1, ConstraintWeights_0=1) {boneToModify {boneName=dyn_R_hair_5}, constraintSetup [{targetBone {boneName=dyn_C_hair_5}}], constraintWeights [1]}
   n88 = "Constraint - dyn_R_hair_offshoot (P)"(ComponentPose=@n87, Alpha=1, ConstraintWeights_0=1) {boneToModify {boneName=dyn_R_hair_offshoot}, constraintSetup [{targetBone {boneName=dyn_R_hair_4}}], constraintWeights [1]}
   n89 = "Transform (Modify) Bone - Bone: dyn_R_hair_offshoot"(ComponentPose=@n88, Alpha=1, Translation="0,0,0", Rotation="20,10,0", Scale="1,1,1") {
    rotationMode=BMM_Additive
    rotationSpace=BCS_BoneSpace
    translationMode=BMM_Additive
    translationSpace=BCS_BoneSpace
    boneToModify {boneName=dyn_R_hair_offshoot}
    rotation {pitch=20, yaw=10}
   }
   Save cached pose 'HairDyn'(Pose=Component To Local(ComponentPose=@n89)) {cacheName=HairDyn}
   n90 = Anim Dynamics - Bone: dyn_C_hair_5(
    ComponentPose=Local To Component(LocalPose=Use cached pose 'Cntrl Rig'())
    bAlphaBoolEnabled=True
    GravityOverride="0,0,-980"
    AngularSpringConstant=20000
    ExternalForce="0,0,0"
   ) {
    alphaInputType=Bool
    angularDampingOverride=1
    angularSpringConstant=20000
    bAngularSpring=true
    bOverrideAngularDamping=true
    bUseGravityOverride=true
    bUseSphericalLimits=true
    gravityScale=0.35
    simulationSpace=RootRelative
    alphaBoolBlend {blendInTime=2, blendOption=Cubic}
    boundBone {boneName=dyn_C_hair_5}
    componentAppliedLinearAccClamp=<400000 400000 400000>
    componentLinearAccScale=<0.3 0.4 0.3>
    componentLinearVelScale=<0.3 0.4 3>
    gravityOverride {z=-980}
    physicsBodyDefinitions [{
      boundBone {boneName=dyn_C_hair_5}
      boxExtents=<25 25 25>
      constraintSetup {angularLimitsMax=<0 10 10>, angularLimitsMin=<0 0 -10>, angularTarget=<1 0 0>}
      localJointOffset=<7 0 0>
     }]
    relativeSpaceBone {boneName=root}
    sphericalLimits [{limitRadius=15, drivingBone {boneName=pelvis}, sphereLocalOffset=<0 0 0>}]
   }
   n91 = Anim Dynamics - Bone: dyn_L_hair_5(ComponentPose=@n90, bAlphaBoolEnabled=True, GravityOverride="0,0,-980", AngularSpringConstant=20000, ExternalForce="0,0,0") {
    alphaInputType=Bool
    angularDampingOverride=1
    angularSpringConstant=20000
    bAngularSpring=true
    bOverrideAngularDamping=true
    bUseGravityOverride=true
    bUseSphericalLimits=true
    gravityScale=0.35
    simulationSpace=RootRelative
    alphaBoolBlend {blendInTime=2, blendOption=Cubic}
    boundBone {boneName=dyn_L_hair_5}
    componentAppliedLinearAccClamp=<400000 400000 400000>
    componentLinearAccScale=<0.3 0.4 0.3>
    componentLinearVelScale=<0.3 0.4 3>
    gravityOverride {z=-980}
    physicsBodyDefinitions [{
      boundBone {boneName=dyn_L_hair_5}
      boxExtents=<25 25 25>
      constraintSetup {angularLimitsMax=<0 10 10>, angularLimitsMin=<0 0 -10>, angularTarget=<1 0 0>}
      localJointOffset=<7 0 0>
     }]
    relativeSpaceBone {boneName=root}
    sphericalLimits [{limitRadius=15, drivingBone {boneName=pelvis}, sphereLocalOffset=<0 0 0>}]
   }
   n92 = Anim Dynamics - Bone: dyn_R_hair_5(ComponentPose=@n91, bAlphaBoolEnabled=True, GravityOverride="0,0,-980", AngularSpringConstant=20000, ExternalForce="0,0,0") {
    alphaInputType=Bool
    angularDampingOverride=1
    angularSpringConstant=20000
    bAngularSpring=true
    bOverrideAngularDamping=true
    bUseGravityOverride=true
    bUseSphericalLimits=true
    gravityScale=0.35
    simulationSpace=RootRelative
    alphaBoolBlend {blendInTime=2, blendOption=Cubic}
    boundBone {boneName=dyn_R_hair_5}
    componentAppliedLinearAccClamp=<400000 400000 400000>
    componentLinearAccScale=<0.3 0.4 0.3>
    componentLinearVelScale=<0.3 0.4 3>
    gravityOverride {z=-980}
    physicsBodyDefinitions [{
      boundBone {boneName=dyn_R_hair_5}
      boxExtents=<25 25 25>
      constraintSetup {angularLimitsMax=<0 10 10>, angularLimitsMin=<0 0 -10>, angularTarget=<1 0 0>}
      localJointOffset=<7 0 0>
     }]
    relativeSpaceBone {boneName=root}
    sphericalLimits [{limitRadius=15, drivingBone {boneName=pelvis}, sphereLocalOffset=<0 0 0>}]
   }
   n93 = Anim Dynamics - Bone: dyn_C_hair_6(ComponentPose=@n92, bAlphaBoolEnabled=True, GravityOverride="0,0,-980", AngularSpringConstant=20000, ExternalForce="0,0,0") {
    alphaInputType=Bool
    angularDampingOverride=1
    angularSpringConstant=20000
    bAngularSpring=true
    bOverrideAngularDamping=true
    bUseGravityOverride=true
    bUseSphericalLimits=true
    gravityScale=0.35
    simulationSpace=RootRelative
    alphaBoolBlend {blendInTime=2, blendOption=Cubic}
    boundBone {boneName=dyn_C_hair_6}
    componentAppliedLinearAccClamp=<400000 400000 400000>
    componentLinearAccScale=<0.3 0.4 0.3>
    componentLinearVelScale=<0.3 0.4 3>
    gravityOverride {z=-980}
    physicsBodyDefinitions [{
      boundBone {boneName=dyn_C_hair_6}
      boxExtents=<25 25 25>
      constraintSetup {angularLimitsMax=<0 0 10>, angularLimitsMin=<0 -45 -10>, angularTarget=<1 0 0>}
      localJointOffset=<7 0 0>
     }]
    relativeSpaceBone {boneName=root}
    sphericalLimits [{limitRadius=15, drivingBone {boneName=pelvis}, sphereLocalOffset=<0 0 0>}]
   }
   n94 = Anim Dynamics - Bone: dyn_C_hair_end_r(ComponentPose=@n93, bAlphaBoolEnabled=True, GravityOverride="0,0,-980", AngularSpringConstant=20000, ExternalForce="0,0,0") {
    alphaInputType=Bool
    angularDampingOverride=1
    angularSpringConstant=20000
    bAngularSpring=true
    bOverrideAngularDamping=true
    bUseGravityOverride=true
    bUseSphericalLimits=true
    gravityScale=0.35
    simulationSpace=RootRelative
    alphaBoolBlend {blendInTime=2, blendOption=Cubic}
    boundBone {boneName=dyn_C_hair_end_r}
    componentAppliedLinearAccClamp=<400000 400000 400000>
    componentLinearAccScale=<0.3 0.4 0.3>
    componentLinearVelScale=<0.3 0.4 3>
    gravityOverride {z=-980}
    physicsBodyDefinitions [{
      boundBone {boneName=dyn_C_hair_end_r}
      boxExtents=<25 25 25>
      constraintSetup {angularLimitsMax=<0 45 10>, angularLimitsMin=<0 0 -10>, angularTarget=<1 0 0>}
      localJointOffset=<7 0 0>
     }]
    relativeSpaceBone {boneName=root}
    sphericalLimits [{limitRadius=15, drivingBone {boneName=pelvis}, sphereLocalOffset=<0 0 0>}]
   }
   n95 = Anim Dynamics - Bone: dyn_C_hair_end_l(ComponentPose=@n94, bAlphaBoolEnabled=True, GravityOverride="0,0,-980", AngularSpringConstant=20000, ExternalForce="0,0,0") {
    alphaInputType=Bool
    angularDampingOverride=1
    angularSpringConstant=20000
    bAngularSpring=true
    bOverrideAngularDamping=true
    bUseGravityOverride=true
    bUseSphericalLimits=true
    gravityScale=0.35
    simulationSpace=RootRelative
    alphaBoolBlend {blendInTime=2, blendOption=Cubic}
    boundBone {boneName=dyn_C_hair_end_l}
    componentAppliedLinearAccClamp=<400000 400000 400000>
    componentLinearAccScale=<0.3 0.4 0.3>
    componentLinearVelScale=<0.3 0.4 3>
    gravityOverride {z=-980}
    physicsBodyDefinitions [{
      boundBone {boneName=dyn_C_hair_end_l}
      boxExtents=<25 25 25>
      constraintSetup {angularLimitsMax=<0 45 10>, angularLimitsMin=<0 0 -10>, angularTarget=<1 0 0>}
      localJointOffset=<7 0 0>
     }]
    relativeSpaceBone {boneName=root}
    sphericalLimits [{limitRadius=15, drivingBone {boneName=pelvis}, sphereLocalOffset=<0 0 0>}]
   }
   n96 = Anim Dynamics - Bone: dyn_C_hair_end_c(ComponentPose=@n95, bAlphaBoolEnabled=True, GravityOverride="0,0,-980", AngularSpringConstant=20000, ExternalForce="0,0,0") {
    alphaInputType=Bool
    angularDampingOverride=1
    angularSpringConstant=20000
    bAngularSpring=true
    bOverrideAngularDamping=true
    bUseGravityOverride=true
    bUseSphericalLimits=true
    gravityScale=0.35
    simulationSpace=RootRelative
    alphaBoolBlend {blendInTime=2, blendOption=Cubic}
    boundBone {boneName=dyn_C_hair_end_c}
    componentAppliedLinearAccClamp=<400000 400000 400000>
    componentLinearAccScale=<0.3 0.4 0.3>
    componentLinearVelScale=<0.3 0.4 3>
    gravityOverride {z=-980}
    physicsBodyDefinitions [{
      boundBone {boneName=dyn_C_hair_end_c}
      boxExtents=<25 25 25>
      constraintSetup {angularLimitsMax=<0 45 10>, angularLimitsMin=<0 0 -10>, angularTarget=<1 0 0>}
      localJointOffset=<7 0 0>
     }]
    relativeSpaceBone {boneName=root}
    sphericalLimits [{limitRadius=15, drivingBone {boneName=pelvis}, sphereLocalOffset=<0 0 0>}]
   }
   n97 = Anim Dynamics - Bone: dyn_L_hair_offshoot_end(ComponentPose=@n96, bAlphaBoolEnabled=True, GravityOverride="0,0,-980", AngularSpringConstant=20000, ExternalForce="0,0,0") {
    alphaInputType=Bool
    angularDampingOverride=1
    angularSpringConstant=20000
    bAngularSpring=true
    bOverrideAngularDamping=true
    bUseGravityOverride=true
    bUseSphericalLimits=true
    gravityScale=0.35
    simulationSpace=RootRelative
    alphaBoolBlend {blendInTime=2, blendOption=Cubic}
    boundBone {boneName=dyn_L_hair_offshoot_end}
    componentAppliedLinearAccClamp=<400000 400000 400000>
    componentLinearAccScale=<0.3 0.4 0.3>
    componentLinearVelScale=<0.3 0.4 3>
    gravityOverride {z=-980}
    physicsBodyDefinitions [{
      boundBone {boneName=dyn_L_hair_offshoot_end}
      boxExtents=<25 25 25>
      constraintSetup {angularLimitsMax=<0 45 10>, angularLimitsMin=<0 0 -10>, angularTarget=<1 0 0>}
      localJointOffset=<7 0 0>
     }]
    relativeSpaceBone {boneName=root}
    sphericalLimits [{limitRadius=15, drivingBone {boneName=pelvis}, sphereLocalOffset=<0 0 0>}]
   }
   n98 = Anim Dynamics - Bone: dyn_L_hair_end_l(ComponentPose=@n97, bAlphaBoolEnabled=True, GravityOverride="0,0,-980", AngularSpringConstant=20000, ExternalForce="0,0,0") {
    alphaInputType=Bool
    angularDampingOverride=1
    angularSpringConstant=20000
    bAngularSpring=true
    bOverrideAngularDamping=true
    bUseGravityOverride=true
    bUseSphericalLimits=true
    gravityScale=0.35
    simulationSpace=RootRelative
    alphaBoolBlend {blendInTime=2, blendOption=Cubic}
    boundBone {boneName=dyn_L_hair_end_l}
    componentAppliedLinearAccClamp=<400000 400000 400000>
    componentLinearAccScale=<0.3 0.4 0.3>
    componentLinearVelScale=<0.3 0.4 3>
    gravityOverride {z=-980}
    physicsBodyDefinitions [{
      boundBone {boneName=dyn_L_hair_end_l}
      boxExtents=<25 25 25>
      constraintSetup {angularLimitsMax=<0 45 10>, angularLimitsMin=<0 0 -10>, angularTarget=<1 0 0>}
      localJointOffset=<7 0 0>
     }]
    relativeSpaceBone {boneName=root}
    sphericalLimits [{limitRadius=15, drivingBone {boneName=pelvis}, sphereLocalOffset=<0 0 0>}]
   }
   n99 = Anim Dynamics - Bone: dyn_L_hair_end_r(ComponentPose=@n98, bAlphaBoolEnabled=True, GravityOverride="0,0,-980", AngularSpringConstant=20000, ExternalForce="0,0,0") {
    alphaInputType=Bool
    angularDampingOverride=1
    angularSpringConstant=20000
    bAngularSpring=true
    bOverrideAngularDamping=true
    bUseGravityOverride=true
    bUseSphericalLimits=true
    gravityScale=0.35
    simulationSpace=RootRelative
    alphaBoolBlend {blendInTime=2, blendOption=Cubic}
    boundBone {boneName=dyn_L_hair_end_r}
    componentAppliedLinearAccClamp=<400000 400000 400000>
    componentLinearAccScale=<0.3 0.4 0.3>
    componentLinearVelScale=<0.3 0.4 3>
    gravityOverride {z=-980}
    physicsBodyDefinitions [{
      boundBone {boneName=dyn_L_hair_end_r}
      boxExtents=<25 25 25>
      constraintSetup {angularLimitsMax=<0 45 10>, angularLimitsMin=<0 0 -10>, angularTarget=<1 0 0>}
      localJointOffset=<7 0 0>
     }]
    relativeSpaceBone {boneName=root}
    sphericalLimits [{limitRadius=15, drivingBone {boneName=pelvis}, sphereLocalOffset=<0 0 0>}]
   }
   n100 = Anim Dynamics - Bone: dyn_R_hair_end_r(ComponentPose=@n99, bAlphaBoolEnabled=True, GravityOverride="0,0,-980", AngularSpringConstant=20000, ExternalForce="0,0,0") {
    alphaInputType=Bool
    angularDampingOverride=1
    angularSpringConstant=20000
    bAngularSpring=true
    bOverrideAngularDamping=true
    bUseGravityOverride=true
    bUseSphericalLimits=true
    gravityScale=0.35
    simulationSpace=RootRelative
    alphaBoolBlend {blendInTime=2, blendOption=Cubic}
    boundBone {boneName=dyn_R_hair_end_r}
    componentAppliedLinearAccClamp=<400000 400000 400000>
    componentLinearAccScale=<0.3 0.4 0.3>
    componentLinearVelScale=<0.3 0.4 3>
    gravityOverride {z=-980}
    physicsBodyDefinitions [{
      boundBone {boneName=dyn_R_hair_end_r}
      boxExtents=<25 25 25>
      constraintSetup {angularLimitsMax=<0 45 10>, angularLimitsMin=<0 0 -10>, angularTarget=<1 0 0>}
      localJointOffset=<7 0 0>
     }]
    relativeSpaceBone {boneName=root}
    sphericalLimits [{limitRadius=15, drivingBone {boneName=pelvis}, sphereLocalOffset=<0 0 0>}]
   }
   n101 = Anim Dynamics - Bone: dyn_R_hair_end_l(ComponentPose=@n100, bAlphaBoolEnabled=True, GravityOverride="0,0,-980", AngularSpringConstant=20000, ExternalForce="0,0,0") {
    alphaInputType=Bool
    angularDampingOverride=1
    angularSpringConstant=20000
    bAngularSpring=true
    bOverrideAngularDamping=true
    bUseGravityOverride=true
    bUseSphericalLimits=true
    gravityScale=0.35
    simulationSpace=RootRelative
    alphaBoolBlend {blendInTime=2, blendOption=Cubic}
    boundBone {boneName=dyn_R_hair_end_l}
    componentAppliedLinearAccClamp=<400000 400000 400000>
    componentLinearAccScale=<0.3 0.4 0.3>
    componentLinearVelScale=<0.3 0.4 3>
    gravityOverride {z=-980}
    physicsBodyDefinitions [{
      boundBone {boneName=dyn_R_hair_end_l}
      boxExtents=<25 25 25>
      constraintSetup {angularLimitsMax=<0 45 10>, angularLimitsMin=<0 0 -10>, angularTarget=<1 0 0>}
      localJointOffset=<7 0 0>
     }]
    relativeSpaceBone {boneName=root}
    sphericalLimits [{limitRadius=15, drivingBone {boneName=pelvis}, sphereLocalOffset=<0 0 0>}]
   }
   n102 = Anim Dynamics - Bone: dyn_R_hair_offshoot_end(ComponentPose=@n101, bAlphaBoolEnabled=True, GravityOverride="0,0,-980", AngularSpringConstant=20000, ExternalForce="0,0,0") {
    alphaInputType=Bool
    angularDampingOverride=1
    angularSpringConstant=20000
    bAngularSpring=true
    bOverrideAngularDamping=true
    bUseGravityOverride=true
    bUseSphericalLimits=true
    gravityScale=0.35
    simulationSpace=RootRelative
    alphaBoolBlend {blendInTime=2, blendOption=Cubic}
    boundBone {boneName=dyn_R_hair_offshoot_end}
    componentAppliedLinearAccClamp=<400000 400000 400000>
    componentLinearAccScale=<0.3 0.4 0.3>
    componentLinearVelScale=<0.3 0.4 3>
    gravityOverride {z=-980}
    physicsBodyDefinitions [{
      boundBone {boneName=dyn_R_hair_offshoot_end}
      boxExtents=<25 25 25>
      constraintSetup {angularLimitsMax=<0 45 10>, angularLimitsMin=<0 0 -10>, angularTarget=<1 0 0>}
      localJointOffset=<7 0 0>
     }]
    relativeSpaceBone {boneName=root}
    sphericalLimits [{limitRadius=15, drivingBone {boneName=pelvis}, sphereLocalOffset=<0 0 0>}]
   }
   n103 = Anim Dynamics - Bone: dyn_LF_bangs_1(ComponentPose=@n102, Alpha=1, ExternalForce="0,0,0") {
    angularDampingOverride=1
    angularSpringConstant=3000
    bAngularSpring=true
    bOverrideAngularDamping=true
    bUseGravityOverride=true
    gravityScale=0.35
    simulationSpace=BoneRelative
    boundBone {boneName=dyn_LF_bangs_1}
    componentAppliedLinearAccClamp=<1000 1000 1000>
    componentLinearAccScale=<0.3 0.4 0.3>
    componentLinearVelScale=<0.3 0.4 0.3>
    gravityOverride {z=-980}
    physicsBodyDefinitions [{
      boundBone {boneName=dyn_LF_bangs_1}
      boxExtents=<20 30 30>
      constraintSetup {angularLimitsMax=<0 0 0>, angularLimitsMin=<0 -5 -10>, angularTarget=<1 0 0>}
      localJointOffset=<2 0 0>
     }]
    relativeSpaceBone {boneName=head}
    simSpaceSettings {maxAngularAcceleration=100, maxAngularVelocity=10, simSpaceAngularAlpha=1}
   }
   n104 = Anim Dynamics - Bone: dyn_LF_bangs_2(ComponentPose=@n103, Alpha=1, ExternalForce="0,0,0") {
    angularDampingOverride=1
    angularSpringConstant=3000
    bAngularSpring=true
    bOverrideAngularDamping=true
    bUseGravityOverride=true
    gravityScale=0.35
    simulationSpace=BoneRelative
    boundBone {boneName=dyn_LF_bangs_2}
    componentAppliedLinearAccClamp=<1000 1000 1000>
    componentLinearAccScale=<0.3 0.4 0.3>
    componentLinearVelScale=<0.3 0.4 0.3>
    gravityOverride {z=-980}
    physicsBodyDefinitions [{
      boundBone {boneName=dyn_LF_bangs_2}
      boxExtents=<20 30 30>
      constraintSetup {angularLimitsMax=<0 0 0>, angularLimitsMin=<0 -5 -10>, angularTarget=<1 0 0>}
      localJointOffset=<2 0 0>
     }]
    relativeSpaceBone {boneName=head}
    simSpaceSettings {maxAngularAcceleration=100, maxAngularVelocity=10, simSpaceAngularAlpha=1}
   }
   n105 = Anim Dynamics - Bone: dyn_LF_bangs_3(ComponentPose=@n104, Alpha=1, ExternalForce="0,0,0") {
    angularDampingOverride=1
    angularSpringConstant=3000
    bAngularSpring=true
    bOverrideAngularDamping=true
    bUseGravityOverride=true
    gravityScale=0.35
    simulationSpace=BoneRelative
    boundBone {boneName=dyn_LF_bangs_3}
    componentAppliedLinearAccClamp=<1000 1000 1000>
    componentLinearAccScale=<0.3 0.4 0.3>
    componentLinearVelScale=<0.3 0.4 0.3>
    gravityOverride {z=-980}
    physicsBodyDefinitions [{
      boundBone {boneName=dyn_LF_bangs_3}
      boxExtents=<20 30 30>
      constraintSetup {angularLimitsMax=<0 0 0>, angularLimitsMin=<0 -5 -15>, angularTarget=<1 0 0>}
      localJointOffset=<2 0 0>
     }]
    relativeSpaceBone {boneName=head}
    simSpaceSettings {maxAngularAcceleration=100, maxAngularVelocity=10, simSpaceAngularAlpha=1}
   }
   n106 = Anim Dynamics - Bone: dyn_LB_bangs_1(ComponentPose=@n105, Alpha=1, ExternalForce="0,0,0") {
    angularDampingOverride=1
    angularSpringConstant=3000
    bAngularSpring=true
    bOverrideAngularDamping=true
    bUseGravityOverride=true
    gravityScale=0.35
    simulationSpace=BoneRelative
    boundBone {boneName=dyn_LB_bangs_1}
    componentAppliedLinearAccClamp=<1000 1000 1000>
    componentLinearAccScale=<0.3 0.4 0.3>
    componentLinearVelScale=<0.3 0.4 0.3>
    gravityOverride {z=-980}
    physicsBodyDefinitions [{
      boundBone {boneName=dyn_LB_bangs_1}
      boxExtents=<20 30 30>
      constraintSetup {angularLimitsMax=<0 5 0>, angularLimitsMin=<0 -5 -10>, angularTarget=<1 0 0>}
      localJointOffset=<2 0 0>
     }]
    relativeSpaceBone {boneName=head}
    simSpaceSettings {maxAngularAcceleration=100, maxAngularVelocity=10, simSpaceAngularAlpha=1}
   }
   n107 = Anim Dynamics - Bone: dyn_LB_bangs_3(ComponentPose=@n106, Alpha=1, ExternalForce="0,0,0") {
    angularDampingOverride=1
    angularSpringConstant=3000
    bAngularSpring=true
    bOverrideAngularDamping=true
    bUseGravityOverride=true
    gravityScale=0.35
    simulationSpace=BoneRelative
    boundBone {boneName=dyn_LB_bangs_3}
    componentAppliedLinearAccClamp=<1000 1000 1000>
    componentLinearAccScale=<0.3 0.4 0.3>
    componentLinearVelScale=<0.3 0.4 0.3>
    gravityOverride {z=-980}
    physicsBodyDefinitions [{
      boundBone {boneName=dyn_LB_bangs_3}
      boxExtents=<20 30 30>
      constraintSetup {angularLimitsMax=<0 5 0>, angularLimitsMin=<0 -5 -10>, angularTarget=<1 0 0>}
      localJointOffset=<2 0 0>
     }]
    relativeSpaceBone {boneName=head}
    simSpaceSettings {maxAngularAcceleration=100, maxAngularVelocity=10, simSpaceAngularAlpha=1}
   }
   n108 = Anim Dynamics - Bone: dyn_LB_bangs_2(ComponentPose=@n107, Alpha=1, ExternalForce="0,0,0") {
    angularDampingOverride=1
    angularSpringConstant=3000
    bAngularSpring=true
    bOverrideAngularDamping=true
    bUseGravityOverride=true
    gravityScale=0.35
    simulationSpace=BoneRelative
    boundBone {boneName=dyn_LB_bangs_2}
    componentAppliedLinearAccClamp=<1000 1000 1000>
    componentLinearAccScale=<0.3 0.4 0.3>
    componentLinearVelScale=<0.3 0.4 0.3>
    gravityOverride {z=-980}
    physicsBodyDefinitions [{
      boundBone {boneName=dyn_LB_bangs_2}
      boxExtents=<20 30 30>
      constraintSetup {angularLimitsMax=<0 5 0>, angularLimitsMin=<0 -5 -15>, angularTarget=<1 0 0>}
      localJointOffset=<2 0 0>
     }]
    relativeSpaceBone {boneName=head}
    simSpaceSettings {maxAngularAcceleration=100, maxAngularVelocity=10, simSpaceAngularAlpha=1}
   }
   n109 = Anim Dynamics - Bone: dyn_RB_bangs_1(ComponentPose=@n108, Alpha=1, ExternalForce="0,0,0") {
    angularDampingOverride=1
    angularSpringConstant=3000
    bAngularSpring=true
    bOverrideAngularDamping=true
    bUseGravityOverride=true
    gravityScale=0.35
    simulationSpace=BoneRelative
    boundBone {boneName=dyn_RB_bangs_1}
    componentAppliedLinearAccClamp=<1000 1000 1000>
    componentLinearAccScale=<0.3 0.4 0.3>
    componentLinearVelScale=<0.3 0.4 0.3>
    gravityOverride {z=-980}
    physicsBodyDefinitions [{
      boundBone {boneName=dyn_RB_bangs_1}
      boxExtents=<20 30 30>
      constraintSetup {angularLimitsMax=<0 5 10>, angularLimitsMin=<0 -5 0>, angularTarget=<1 0 0>}
      localJointOffset=<2 0 0>
     }]
    relativeSpaceBone {boneName=head}
    simSpaceSettings {maxAngularAcceleration=100, maxAngularVelocity=10, simSpaceAngularAlpha=1}
   }
   n110 = Anim Dynamics - Bone: dyn_RB_bangs_2(ComponentPose=@n109, Alpha=1, ExternalForce="0,0,0") {
    angularDampingOverride=1
    angularSpringConstant=3000
    bAngularSpring=true
    bOverrideAngularDamping=true
    bUseGravityOverride=true
    gravityScale=0.35
    simulationSpace=BoneRelative
    boundBone {boneName=dyn_RB_bangs_2}
    componentAppliedLinearAccClamp=<1000 1000 1000>
    componentLinearAccScale=<0.3 0.4 0.3>
    componentLinearVelScale=<0.3 0.4 0.3>
    gravityOverride {z=-980}
    physicsBodyDefinitions [{
      boundBone {boneName=dyn_RB_bangs_2}
      boxExtents=<20 30 30>
      constraintSetup {angularLimitsMax=<0 5 10>, angularLimitsMin=<0 -5 0>, angularTarget=<1 0 0>}
      localJointOffset=<2 0 0>
     }]
    relativeSpaceBone {boneName=head}
    simSpaceSettings {maxAngularAcceleration=100, maxAngularVelocity=10, simSpaceAngularAlpha=1}
   }
   n111 = Anim Dynamics - Bone: dyn_RB_bangs_3(ComponentPose=@n110, Alpha=1, ExternalForce="0,0,0") {
    angularDampingOverride=1
    angularSpringConstant=3000
    bAngularSpring=true
    bOverrideAngularDamping=true
    bUseGravityOverride=true
    gravityScale=0.35
    simulationSpace=BoneRelative
    boundBone {boneName=dyn_RB_bangs_3}
    componentAppliedLinearAccClamp=<1000 1000 1000>
    componentLinearAccScale=<0.3 0.4 0.3>
    componentLinearVelScale=<0.3 0.4 0.3>
    gravityOverride {z=-980}
    physicsBodyDefinitions [{
      boundBone {boneName=dyn_RB_bangs_3}
      boxExtents=<20 30 30>
      constraintSetup {angularLimitsMax=<0 5 15>, angularLimitsMin=<0 -5 0>, angularTarget=<1 0 0>}
      localJointOffset=<2 0 0>
     }]
    relativeSpaceBone {boneName=head}
    simSpaceSettings {maxAngularAcceleration=100, maxAngularVelocity=10, simSpaceAngularAlpha=1}
   }
   n112 = Anim Dynamics - Bone: dyn_RF_bangs_1(ComponentPose=@n111, Alpha=1, ExternalForce="0,0,0") {
    angularDampingOverride=1
    angularSpringConstant=3000
    bAngularSpring=true
    bOverrideAngularDamping=true
    bUseGravityOverride=true
    gravityScale=0.35
    simulationSpace=BoneRelative
    boundBone {boneName=dyn_RF_bangs_1}
    componentAppliedLinearAccClamp=<1000 1000 1000>
    componentLinearAccScale=<0.3 0.4 0.3>
    componentLinearVelScale=<0.3 0.4 0.3>
    gravityOverride {z=-980}
    physicsBodyDefinitions [{
      boundBone {boneName=dyn_RF_bangs_1}
      boxExtents=<20 30 30>
      constraintSetup {angularLimitsMax=<0 5 10>, angularLimitsMin=<0 -5 0>, angularTarget=<1 0 0>}
      localJointOffset=<2 0 0>
     }]
    relativeSpaceBone {boneName=head}
    simSpaceSettings {maxAngularAcceleration=100, maxAngularVelocity=10, simSpaceAngularAlpha=1}
   }
   n113 = Anim Dynamics - Bone: dyn_RF_bangs_2(ComponentPose=@n112, Alpha=1, ExternalForce="0,0,0") {
    angularDampingOverride=1
    angularSpringConstant=3000
    bAngularSpring=true
    bOverrideAngularDamping=true
    bUseGravityOverride=true
    gravityScale=0.35
    simulationSpace=BoneRelative
    boundBone {boneName=dyn_RF_bangs_2}
    componentAppliedLinearAccClamp=<1000 1000 1000>
    componentLinearAccScale=<0.3 0.4 0.3>
    componentLinearVelScale=<0.3 0.4 0.3>
    gravityOverride {z=-980}
    physicsBodyDefinitions [{
      boundBone {boneName=dyn_RF_bangs_2}
      boxExtents=<20 30 30>
      constraintSetup {angularLimitsMax=<0 5 10>, angularLimitsMin=<0 -5 0>, angularTarget=<1 0 0>}
      localJointOffset=<2 0 0>
     }]
    relativeSpaceBone {boneName=head}
    simSpaceSettings {maxAngularAcceleration=100, maxAngularVelocity=10, simSpaceAngularAlpha=1}
   }
   n114 = Anim Dynamics - Bone: dyn_RF_bangs_3(ComponentPose=@n113, Alpha=1, ExternalForce="0,0,0") {
    angularDampingOverride=1
    angularSpringConstant=3000
    bAngularSpring=true
    bOverrideAngularDamping=true
    bUseGravityOverride=true
    gravityScale=0.35
    simulationSpace=BoneRelative
    boundBone {boneName=dyn_RF_bangs_3}
    componentAppliedLinearAccClamp=<1000 1000 1000>
    componentLinearAccScale=<0.3 0.4 0.3>
    componentLinearVelScale=<0.3 0.4 0.3>
    gravityOverride {z=-980}
    physicsBodyDefinitions [{
      boundBone {boneName=dyn_RF_bangs_3}
      boxExtents=<20 30 30>
      constraintSetup {angularLimitsMax=<0 5 15>, angularLimitsMin=<0 -5 0>, angularTarget=<1 0 0>}
      localJointOffset=<2 0 0>
     }]
    relativeSpaceBone {boneName=head}
    simSpaceSettings {maxAngularAcceleration=100, maxAngularVelocity=10, simSpaceAngularAlpha=1}
   }
   n115 = Anim Dynamics - Bone: dyn_RF_bangs_8(ComponentPose=@n114, Alpha=1, ExternalForce="0,0,0") {
    angularDampingOverride=1
    angularSpringConstant=3000
    bAngularSpring=true
    bOverrideAngularDamping=true
    bUseGravityOverride=true
    gravityScale=0.35
    simulationSpace=BoneRelative
    boundBone {boneName=dyn_RF_bangs_8}
    componentAppliedLinearAccClamp=<1000 1000 1000>
    componentLinearAccScale=<0.3 0.4 0.3>
    componentLinearVelScale=<0.3 0.4 0.3>
    gravityOverride {z=-980}
    physicsBodyDefinitions [{
      boundBone {boneName=dyn_RF_bangs_8}
      boxExtents=<20 30 30>
      constraintSetup {angularLimitsMax=<0 5 15>, angularLimitsMin=<0 -5 0>, angularTarget=<1 0 0>}
      localJointOffset=<2 0 0>
     }]
    relativeSpaceBone {boneName=head}
    simSpaceSettings {maxAngularAcceleration=100, maxAngularVelocity=10, simSpaceAngularAlpha=1}
   }
   Save cached pose 'Backpack'(Pose=Component To Local(ComponentPose=@n115)) {cacheName=Backpack}
  ]}
 EventGraph {nodes [
   n1 = Event BlueprintProcessAnimInputEvent()
   > n2 = Branch(Condition=IsNamed(self=@n1.AnimInputEvent, InEventName=aie.Player.Zipline.Entered))
   > Set isOnZipline(isOnZipline:bool=1) -> (Output_Get:bool)
   Branch(execute=@n2.else, Condition=IsNamed(self=@n1.AnimInputEvent, InEventName=aie.Player.Zipline.Exited))
   > Set isOnZipline(isOnZipline=false)
   Event OnCharacterCustomizationLoaded()
   n3 = CheckForBackbling()
   > Set isWearingPack(isWearingPack:bool=OR Boolean!(A=@n3, B=@n3.bIsBackblingACape, C:bool=@n3.bIsBackblingLong) -> (ReturnValue)) -> (Output_Get:bool)
   > Set isWearingCape(isWearingCape=@n3.bIsBackblingLong)
   Event BlueprintUpdateAnimation()
   > n4 = Sequence!(execute) -> (then_0, then_1, then_2:Exec, then_3:Exec)
   > n5 = Gravity Override State Function DEPRECATED(CharacterPartAnimInstance=Self-Reference(), GameTimeInSeconds=GetGameTimeInSeconds(), GravityOverrideParams=Get Gravity Override Params())
   > Set GravityOverride(GravityOverride=@n5)
   > n6 = Set GravityOverride AnimDynamics(GravityOverride AnimDynamics:Vector="0, 0, 0") -> (Output_Get:Vector)
   n7 = Get bIsSkyDiving()
   n8 = Get bIsJumping()
   n9 = Get bIsFalling()
   > Branch(Condition=AND Boolean(A=OR Boolean!(A=@n8, B=@n9) -> (ReturnValue), B=NOT Boolean(A=@n7)))
   n10 = Remap Float(Input=Get PawnSpeed2D(), OldRange=(X=0,Y=550), NewRange=(X=0,Y=1), Clamp=true) -> (Output=~)
   n11 = "Clamp (Float)"(Value=Get PawnVelocityZ(), Min=-980, Max=0)
   n12 = float + float(A=Lerp(A=1300, B=float * float(A=@n11, B=-2) {is=title K2Node_PromotableOperator}, Alpha=@n10), B=Lerp(A=0, B=800, Alpha=@n10))
   > Set GravityOverride AnimDynamics(GravityOverride AnimDynamics=vector + vector(A=@n6, B=MakeVector(Y=0, Z=@n12)))
   Set disableUpperLimbsCorrection(
    execute=@n4.then_1
    disableUpperLimbsCorrection:bool=OR Boolean!(A=Get bIsSkydiveDiveMode(), B=Get bIsSkydiveFloating(), C:bool=Get bIsParachuteOpened()) -> (ReturnValue)
   ) -> (Output_Get:bool)
   > Set toggleAnimDynamics(toggleAnimDynamics:bool=NOT Boolean(A=Get isOnZipline() -> (isOnZipline:bool))) -> (Output_Get:bool)
   n13 = Get bIsTacticalSprinting()
   n14 = AND Boolean(A=Get bIsJumping(), B=@n13)
   > Set isTacticalJumping(isTacticalJumping:bool=@n14) -> (Output_Get:bool)
   n15 = Get bIsFalling()
   > Set isTacticalJumpingFalling(isTacticalJumpingFalling:bool=OR Boolean!(A=@n14, B=AND Boolean(A=@n13, B=@n15)) -> (ReturnValue)) -> (Output_Get:bool)
   n16 = Branch(execute=@n4.then_2, Condition=OR Boolean(A=Get bIsSkydiveDiveMode(), B=Get bIsSkydiveFloating()))
   > Set angularSpringConstant(angularSpringConstant:double=10000) -> (Output_Get:double)
   n17 = Branch(execute=@n16.else, Condition=Get bIsSprinting())
   > n18 = Branch(Condition=Get bIsTacticalSprinting())
   > Set angularSpringConstant(angularSpringConstant=15000)
   Set angularSpringConstant(execute=@n18.else, angularSpringConstant=17000)
   Set angularSpringConstant(execute=@n17.else, angularSpringConstant=8000)
   n19 = Get isTacticalJumpingFalling() -> (isTacticalJumpingFalling:bool)
   n20 = Get bIsSliding()
   n21 = Get isTacticalJumping() -> (isTacticalJumping:bool)
   n22 = Get bIsTacticalSprinting()
   Set highintert(execute=@n4.then_3, highintert:double="To Float (Boolean)"(InBool=OR Boolean!(A=@n21, B=@n19, C:bool=@n20, D:bool=@n22) -> (ReturnValue))) -> (Output_Get:double)
   n23 = NOT Boolean(A=Get isTacticalJumping())
   n24 = NOT Boolean(A=Get isTacticalJumpingFalling())
   n25 = NOT Boolean(A=Get bIsSliding())
   n26 = NOT Boolean(A=Get bIsTacticalSprinting())
   > Set NOThighintert(NOThighintert:double="To Float (Boolean)"(InBool=OR Boolean(A=@n23, B=@n24, C=@n25, D=@n26))) -> (Output_Get:double)
   n27 = Get bIsPlayingEmote()
   > Set FE(FE:bool=AND Boolean(A=Get bIsInFrontEnd(), B=NOT Boolean(A=@n27))) -> (Output_Get:bool)
   > Set HighInert(HighInert:bool=OR Boolean(A=Get bIsPlayingMeleeAnim(), B=Get bIsPlayingEmote(), C=Get isTacticalJumping(), D=Get bIsBaseballSliding())) -> (Output_Get:bool)
   n28 = Get bIsTargeting()
   n29 = Get Weapon() -> (Weapon:bool)
   n30 = Get bIsPlayingEmote()
   n31 = Get bIsPlayingMeleeAnim()
   n32 = Get bIsMoving2D()
   > Set push hair(push hair:double="To Float (Boolean)"(InBool=OR Boolean!(A=@n32, B=@n28, C=@n29, D=@n30, E:bool=@n31) -> (ReturnValue))) -> (Output_Get:double)
   n33 = Get bIsPlayingMeleeAnim()
   n34 = Get isTacticalJumping()
   n35 = Get bIsKneeSliding()
   > Set TSprint(TSprint:double="To Float (Boolean)"(InBool=OR Boolean!(A=@n34, B=@n33, C=@n35) -> (ReturnValue))) -> (Output_Get:double)
   > Set Diving(Diving:double="To Float (Boolean)"(InBool=Get bIsSkydiveDiveMode())) -> (Output_Get:double)
   Event BlueprintInitializeAnimation()
   > Set Skeletal_Mesh(Skeletal_Mesh:Skeletal Mesh Component Object Reference=GetOwnerSkeletalMeshComponent()) -> (Output_Get:Skeletal Mesh Component Object Reference)
   Event BlueprintBeginPlay() {is=title K2Node_Event}
   > Set disableControlRig(disableControlRig:bool=false) -> (Output_Get:bool)
   Event OnWeaponChanged()
   > n36 = Sequence!(execute) -> (then_0, then_1)
   > Set isAssaultRifleEquipped float(isAssaultRifleEquipped float:double=0) -> (Output_Get:double)
   > n37 = Switch on EFortWeaponCoreAnimation!(execute, Selection:EFortWeaponCoreAnimation Enum=Get WeaponCoreAnim!() -> (WeaponCoreAnim)) -> (
    Melee:Exec
    Pistol:Exec
    Shotgun:Exec
    PaperBlueprint:Exec
    Rifle:Exec
    MeleeOneHand:Exec
    MachinePistol:Exec
    RocketLauncher:Exec
    GrenadeLauncher:Exec
    GoingCommando:Exec
    AssaultRifle:Exec
    TacticalShotgun:Exec
    SniperRifle:Exec
    TrapPlacement:Exec
    ShoulderLauncher:Exec
    AbilityDecoTool:Exec
    Crossbow:Exec
    C4:Exec
    RemoteControl:Exec
    DualWield:Exec
    AR_BullPup:Exec
    AR_ForwardGrip:Exec
    MedPackPaddles:Exec
    SMG_P90:Exec
    AR_DrumGun:Exec
    Consumable_Small:Exec
    Consumable_Large:Exec
    Balloon:Exec
    MountedTurret:Exec
    CreativeTool:Exec
    ExplosiveBow:Exec
    AshtonIndigo:Exec
    AshtonChicago:Exec
    MeleeDualWield:Exec
    Unarmed:Exec
   )
   Set isAssaultRifleEquipped float(execute=@n37.AssaultRifle, isAssaultRifleEquipped float=1)
   Set Weapon(execute=@n36.then_1, Weapon:bool=false) -> (Output_Get:bool)
   > n38 = Switch on EFortWeaponCoreAnimation(Selection=Get WeaponCoreAnim())
   Set Weapon(
    execute=[
     @n38.AR_BullPup
     @n38.AR_DrumGun
     @n38.AR_ForwardGrip
     @n38.AbilityDecoTool
     @n38.AshtonChicago
     @n38.AshtonIndigo
     @n38.AssaultRifle
     @n38.Balloon
     @n38.C4
     @n38.Consumable_Large
     @n38.Consumable_Small
     @n38.CreativeTool
     @n38.Crossbow
     @n38.DualWield
     @n38.ExplosiveBow
     @n38.GoingCommando
     @n38.GrenadeLauncher
     @n38.MachinePistol
     @n38.MedPackPaddles
     @n38.Melee
     @n38.MeleeDualWield
     @n38.MeleeOneHand
     @n38.MountedTurret
     @n38.PaperBlueprint
     @n38.Pistol
     @n38.RemoteControl
     @n38.Rifle
     @n38.RocketLauncher
     @n38.SMG_P90
     @n38.Shotgun
     @n38.ShoulderLauncher
     @n38.SniperRifle
     @n38.TacticalShotgun
     @n38.TrapPlacement
    ]
    Weapon=true
   )
   Get bIsSliding()
  ]}
]
variables [
 alphaHairRigidBody float {type=number}
 angularSpringConstant {type=number, default=8000}
 disableControlRig {type=boolean}
 disableUpperLimbsCorrection {type=boolean}
 Diving {type=number}
 FE {type=boolean}
 Gravity Override Params {
  type=GravityOverrideParamsStruct
  default {
   onGround_Falling {gravityOverride=<0 0 -10000>}
   onGround_Moving {gravityOverride=<0 0 -10000>}
   tactical_Sprint {gravityOverride=<0 -450 -980>}
  }
 }
 GravityOverride {type=Vector}
 GravityOverride AnimDynamics {type=Vector}
 HighInert {type=boolean}
 highintert {type=number}
 isAssaultRifleEquipped float {type=number}
 isOnZipline {type=boolean}
 isTacticalJumping {type=boolean}
 isTacticalJumpingFalling {type=boolean}
 isWearingCape {type=boolean}
 isWearingPack {type=boolean}
 NOThighintert {type=number, default=1}
 push hair {type=number}
 Skeletal_Mesh {type=SkeletalMeshComponent}
 toggleAnimDynamics {type=boolean, default=true}
 TSprint {type=number}
 Weapon {type=boolean}
]
