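# Physics asset for the TameEagle female medium character's FaceAcc part, defining capsule bodies and joints for its dynamic hair bones.
# Center, left and right hair chains hang from spine_04 on free-rotation root joints, then branch into end and offshoot bones with tight swing and twist limits.
# Pelvis, spine and arm capsules cover the body, linked by free-rotation constraints.
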PhysicsAsset /BRCosmetics/Characters/Player/Female/Medium/Bodies/F_MED_TameEagle/Meshes/Parts/F_MED_TameEagle_FaceAcc_Physics
bodies {
 dyn_C_hair_4 [
  {shapeType=Capsule, length=2, radius=1, shapeName=dyn_C_hair_1_capsule, rotation=<rot 0 0 90>}
  {shapeType=Capsule, length=3, radius=5, shapeName=dyn_C_hair_4_capsule, center=<5 0 0>, rotation=<rot 90 0 0>}
 ]
 dyn_C_hair_5 [{shapeType=Capsule, length=5.5538, radius=5, shapeName=dyn_C_hair_1_capsule, center=<6 0 -0.8503>, rotation=<rot 90 0 0>}]
 dyn_C_hair_6 [{shapeType=Capsule, length=5.6721, radius=4.5, shapeName=dyn_C_hair_1_capsule, center=<6.0196 0 -1.5901>, rotation=<rot 90 0 0>}]
 dyn_C_hair_end_c [{
   shapeType=Capsule
   length=4.2384
   radius=2.7551
   shapeName=dyn_C_hair_end_c_capsule
   center=<5.3899 2.6403 -0.7996>
   rotation=<rot 0 -48.4615 -67.96>
  }]
 dyn_C_hair_end_l [{
   shapeType=Capsule
   length=3.7153
   radius=1.2315
   shapeName=dyn_C_hair_end_l_capsule
   center=<5.5271 0.8725 -1.7115>
   rotation=<rot 0 -69.5198 -64.9785>
  }]
 dyn_C_hair_end_r [{
   shapeType=Capsule
   length=3.5311
   radius=1.8966
   shapeName=dyn_C_hair_end_r_capsule
   center=<3.4577 2.4994 -1.364>
   rotation=<rot 0 -52.0832 -63.5901>
  }]
 dyn_L_hair_4 [
  {shapeType=Capsule, length=2, radius=1, shapeName=dyn_C_hair_1_capsule, rotation=<rot 0 0 90>}
  {shapeType=Capsule, length=3, radius=5, shapeName=dyn_L_hair_4_capsule, center=<5 0 0>, rotation=<rot 90 0 0>}
 ]
 dyn_L_hair_5 [{shapeType=Capsule, length=5.5538, radius=5, shapeName=dyn_C_hair_1_capsule, center=<6 0 -0.8503>, rotation=<rot 90 0 0>}]
 dyn_L_hair_END [{shapeType=Capsule, length=6.9636, radius=2.1788, shapeName=dyn_L_hair_END_capsule, center=<5.2125 0.5379 0>, rotation=<rot 90 16.875 0>}]
 dyn_L_hair_end_l [{
   shapeType=Capsule
   length=4.1415
   radius=1.9759
   shapeName=dyn_L_hair_end_l_capsule
   center=<4.748 -3.0055 -2.4534>
   rotation=<rot 0 -139.0545 -66.9509>
  }]
 dyn_L_hair_end_r [{
   shapeType=Capsule
   length=4.9298
   radius=2.2909
   shapeName=dyn_L_hair_end_r_capsule
   center=<6.3931 0.0712 -1.8333>
   rotation=<rot 0 -72.522 -70.9668>
  }]
 dyn_L_hair_offshoot [{
   shapeType=Capsule
   length=4.51
   radius=1.6383
   shapeName=dyn_L_hair_offshoot_capsule
   center=<4.8783 -1.8695 1.7798>
   rotation=<rot 0 74.5289 -71.3818>
  }]
 dyn_R_hair_4 [
  {shapeType=Capsule, length=2, radius=1, shapeName=dyn_C_hair_1_capsule, rotation=<rot 0 0 90>}
  {shapeType=Capsule, length=3, radius=5, shapeName=dyn_R_hair_4_capsule, center=<5 0 0>, rotation=<rot 90 0 0>}
 ]
 dyn_R_hair_5 [{shapeType=Capsule, length=5.5538, radius=5, shapeName=dyn_C_hair_1_capsule, center=<6 0 -0.8503>, rotation=<rot 90 0 0>}]
 dyn_R_hair_END [{
   shapeType=Capsule
   length=4.6829
   radius=1.5475
   shapeName=dyn_R_hair_END_capsule
   center=<4.8517 -2.5113 -1.3407>
   rotation=<rot 0 -126.8854 -71.3478>
  }]
 dyn_R_hair_end_l [{
   shapeType=Capsule
   length=5.4725
   radius=2.5357
   shapeName=dyn_R_hair_end_l_capsule
   center=<5.2781 1.8023 -1.2398>
   rotation=<rot 0 -72.0289 -82.1081>
  }]
 dyn_R_hair_end_r [{
   shapeType=Capsule
   length=4.4548
   radius=1.8274
   shapeName=dyn_R_hair_end_r_capsule
   center=<6.0204 0.6431 -0.542>
   rotation=<rot 0 -67.6499 -78.9669>
  }]
 dyn_R_hair_offshoot [{
   shapeType=Capsule
   length=5.5331
   radius=1.2518
   shapeName=dyn_R_hair_offshoot_capsule
   center=<4.3726 0.8887 1.6613>
   rotation=<rot 0 96.5602 -83.0499>
  }]
 lowerarm_l [{shapeType=Capsule, length=16.4463, radius=3.8719, shapeName=lowerarm_r_capsule, center=<11.498 0 0>, rotation=<rot 90 0 0>}]
 lowerarm_r [{shapeType=Capsule, length=16.4463, radius=3.8719, shapeName=lowerarm_r_capsule, center=<-11.498 0 0>, rotation=<rot 90 0 0>}]
 pelvis [{shapeType=Capsule, length=12.648, radius=12.0693, shapeName=pelvis_capsule, center=<-4.3439 -2.7153 0>, rotation=<rot 0 -90 0>}]
 spine_01 [{shapeType=Capsule, length=8.3135, radius=9.6685, shapeName=spine_01_capsule, center=<0.9493 1.2061 0>, rotation=<rot 0 -90 0>}]
 spine_02 [{shapeType=Capsule, length=4.887, radius=9.8602, shapeName=spine_02_capsule, center=<-0.2722 3.3307 -0.3025>, rotation=<rot 0 -90 0>}]
 spine_03 [{shapeType=Capsule, length=6.4618, radius=8.7935, shapeName=spine_03_capsule, center=<-1.3718 2.4404 0>, rotation=<rot 0 -90 0>}]
 spine_04 [{shapeType=Capsule, length=3.967, radius=12.0063, shapeName=spine_04_capsule, center=<0 3.4519 0>, rotation=<rot 0 -90 0>}]
 upperarm_l [{shapeType=Capsule, length=22.745, radius=4.8788, shapeName=upperarm_l_capsule, center=<13 0 0>, rotation=<rot -90 180 180>}]
 upperarm_r [{shapeType=Capsule, length=22.745, radius=4.8788, shapeName=upperarm_l_capsule, center=<-13 0 0>, rotation=<rot -90 180 180>}]
}
constraints [
 {
  bone1Name=spine_01
  bone2Name=pelvis
  swing1LimitDegrees=45
  swing1Motion=Free
  swing2LimitDegrees=45
  swing2Motion=Free
  twistLimitDegrees=45
  twistMotion=Free
 }
 {
  bone1Name=spine_02
  bone2Name=spine_01
  swing1LimitDegrees=45
  swing1Motion=Free
  swing2LimitDegrees=45
  swing2Motion=Free
  twistLimitDegrees=45
  twistMotion=Free
 }
 {
  bone1Name=spine_03
  bone2Name=spine_02
  swing1LimitDegrees=45
  swing1Motion=Free
  swing2LimitDegrees=45
  swing2Motion=Free
  twistLimitDegrees=45
  twistMotion=Free
 }
 {
  bone1Name=spine_04
  bone2Name=spine_03
  swing1LimitDegrees=45
  swing1Motion=Free
  swing2LimitDegrees=45
  swing2Motion=Free
  twistLimitDegrees=45
  twistMotion=Free
 }
 {
  bone1Name=dyn_C_hair_4
  bone2Name=spine_04
  swing1LimitDegrees=45
  swing1Motion=Free
  swing2LimitDegrees=45
  swing2Motion=Free
  twistLimitDegrees=45
  twistMotion=Free
 }
 {bone1Name=dyn_C_hair_5, bone2Name=dyn_C_hair_4, swing1LimitDegrees=5, swing2LimitDegrees=25, twistLimitDegrees=0}
 {bone1Name=dyn_C_hair_6, bone2Name=dyn_C_hair_5, swing1LimitDegrees=5, swing2LimitDegrees=20, twistLimitDegrees=0}
 {
  bone1Name=dyn_L_hair_4
  bone2Name=spine_04
  swing1LimitDegrees=45
  swing1Motion=Free
  swing2LimitDegrees=45
  swing2Motion=Free
  twistLimitDegrees=45
  twistMotion=Free
 }
 {bone1Name=dyn_L_hair_5, bone2Name=dyn_L_hair_4, swing1LimitDegrees=5, swing2LimitDegrees=25, twistLimitDegrees=0}
 {
  bone1Name=dyn_R_hair_4
  bone2Name=spine_04
  swing1LimitDegrees=45
  swing1Motion=Free
  swing2LimitDegrees=45
  swing2Motion=Free
  twistLimitDegrees=45
  twistMotion=Free
 }
 {bone1Name=dyn_R_hair_5, bone2Name=dyn_R_hair_4, swing1LimitDegrees=5, swing2LimitDegrees=25, twistLimitDegrees=0}
 {
  bone1Name=upperarm_l
  bone2Name=spine_04
  swing1LimitDegrees=45
  swing1Motion=Free
  swing2LimitDegrees=45
  swing2Motion=Free
  twistLimitDegrees=45
  twistMotion=Free
 }
 {
  bone1Name=lowerarm_l
  bone2Name=upperarm_l
  swing1LimitDegrees=45
  swing1Motion=Free
  swing2LimitDegrees=45
  swing2Motion=Free
  twistLimitDegrees=45
  twistMotion=Free
 }
 {
  bone1Name=upperarm_r
  bone2Name=spine_04
  swing1LimitDegrees=45
  swing1Motion=Free
  swing2LimitDegrees=45
  swing2Motion=Free
  twistLimitDegrees=45
  twistMotion=Free
 }
 {
  bone1Name=lowerarm_r
  bone2Name=upperarm_r
  swing1LimitDegrees=45
  swing1Motion=Free
  swing2LimitDegrees=45
  swing2Motion=Free
  twistLimitDegrees=45
  twistMotion=Free
 }
 {bone1Name=dyn_L_hair_END, bone2Name=dyn_L_hair_offshoot, swing1LimitDegrees=10, swing2LimitDegrees=20, twistLimitDegrees=2}
 {bone1Name=dyn_L_hair_offshoot, bone2Name=dyn_L_hair_4, swing1LimitDegrees=10, swing2LimitDegrees=10, twistLimitDegrees=2}
 {bone1Name=dyn_R_hair_offshoot, bone2Name=dyn_R_hair_4, swing1LimitDegrees=10, swing2LimitDegrees=10, twistLimitDegrees=2}
 {bone1Name=dyn_R_hair_END, bone2Name=dyn_R_hair_offshoot, swing1LimitDegrees=10, swing2LimitDegrees=10, twistLimitDegrees=2}
 {bone1Name=dyn_R_hair_end_r, bone2Name=dyn_R_hair_5, swing1LimitDegrees=10, swing2LimitDegrees=10, twistLimitDegrees=2}
 {bone1Name=dyn_R_hair_end_l, bone2Name=dyn_R_hair_5, swing1LimitDegrees=5, swing2LimitDegrees=10, twistLimitDegrees=2}
 {bone1Name=dyn_L_hair_end_r, bone2Name=dyn_L_hair_5, swing1LimitDegrees=5, swing2LimitDegrees=10, twistLimitDegrees=2}
 {bone1Name=dyn_L_hair_end_l, bone2Name=dyn_L_hair_5, swing1LimitDegrees=5, swing2LimitDegrees=10, twistLimitDegrees=2}
 {bone1Name=dyn_C_hair_end_r, bone2Name=dyn_C_hair_6, swing1LimitDegrees=10, swing2LimitDegrees=10, twistLimitDegrees=2}
 {bone1Name=dyn_C_hair_end_l, bone2Name=dyn_C_hair_6, swing1LimitDegrees=10, swing2LimitDegrees=10, twistLimitDegrees=2}
 {bone1Name=dyn_C_hair_end_c, bone2Name=dyn_C_hair_6, swing1LimitDegrees=10, swing2LimitDegrees=10, twistLimitDegrees=2}
]
